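# Translucent Niagara sprite material that draws a radial progress indicator, apparently for a container effect, as a ring or diamond.
# The UV angle, offset by particle random and rounded, splits color and alpha between particle color and a dynamic parameter.
# Two subtracted sphere masks or a diamond texture shape the opacity, with optional radial lines.
# A blendable hologram layer adds flickering world-aligned noise, scanlines, fresnel, depth fade, custom-depth occlusion and strobe.
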Material /MidMatchAssignedGameplay/Items/Environmental/VFX/ContainerFX/Materials/M_Aug_ContainterProgress
graph {
 properties {
  bUseMaterialAttributes=true
  bUsedWithNiagaraSprites=true
  blendMode=BLEND_Translucent
  pixelDepthOffsetMode=PDOM_Legacy
  usedShadingModels=Default Lit
 }
 nodes [
  n1 = MaterialFunctionCall() {materialFunction=&VectorToRadialValue}
  n2 = Add(A=LinearInterpolate(Alpha=ParticleRandom()) {constA=0.5, constB=-0.5}, B=@n1.Vector Converted to Angle)
  n3 = LinearInterpolate(A=ParticleColor().RGBA, B=DynamicParameter() {defaultValue=<rgba 0 1 1 1>}.RGBA, Alpha=OneMinus(None=Round(None=@n2)))
  n4 = ComponentMask(None=@n3) {a=true}
  n5 = Multiply(A=@n1.Vector Converted to Angle) {constB=100}
  n6 = LinearInterpolate(B=Round(None=Frac(None=@n5)), Alpha=ScalarParameter() {parameterName=Use Radial Lines}) {constA=1}
  n7 = TextureObject() {samplerType=SAMPLERTYPE_LinearColor, texture=&T_NoisePack_LinearColor}
  n8 = Divide(A=ObjectRadius(), B=ScalarParameter() {defaultValue=0.1, group=Noise, parameterName=NoiseSize})
  n9 = Multiply(A=VectorParameter() {group=Noise, parameterName=NoiseTileMult, defaultValue=<rgba 1 1 0.1 1>}.RGB, B=@n8)
  n10 = Time()
  n11 = VectorParameter() {group=Noise, parameterName=NoisePanSpeed, defaultValue=<rgba -0.08 0.07 -0.2 0>}
  n12 = Add(A=WorldPosition().XYZ, B=Multiply(A=@n8, B=Multiply(A=@n10, B=@n11.RGB)))
  n13 = Multiply(
   A=Power(Base=MaterialFunctionCall(TextureObject=@n7, TextureSize=@n9, WorldPosition=@n12) {materialFunction=&WorldAlignedTexture}.XY Texture) {constExponent=4}
   B=ScalarParameter() {defaultValue=2, group=Noise, parameterName=NoiseMaskMult}
  ) {constB=20}
  n14 = ActorPositionWS()
  n15 = MaterialFunctionCall(Float2=Round(None=Divide(A=@n14) {constB=2000})) {materialFunction=&BreakOutFloat2Components}
  n16 = Add(A=Time(), B=Add(A=@n15.R, B=@n15.R))
  n17 = Multiply(A=ScalarParameter() {defaultValue=0.5, parameterName=FlickerSpeed}, B=@n16)
  n18 = Sine(None=@n17) {period=1.6}
  n19 = Sine(None=@n17) {period=1.42}
  n20 = Ceil(None=AppendVector(A=AppendVector(A=@n18, B=@n19), B=Sine(None=@n17) {period=1.51}))
  n21 = ScalarParameter() {defaultValue=0.5, group=Noise, parameterName=NoiseFrequency}
  n22 = Add(A=Sine(None=@n17) {period=0.72}, B=Multiply(B=@n21) {constA=0.75}) {constB=0.75}
  n23 = Multiply(A=Add(A=Sine(None=@n17) {period=2.1}, B=Multiply(B=@n21) {constA=0.25}), B=@n22)
  n24 = Multiply(B=@n21) {constA=0.9, constB=0.9}
  n25 = Sine(None=@n17) {period=1.25}
  n26 = Multiply(A=@n22, B=Ceil(None=Add(A=@n24, B=@n25) {constB=0.9}))
  n27 = Sine(None=@n16) {period=0.075}
  n28 = Multiply(A=Saturate(None=Multiply(A=@n23, B=@n26)), B=LinearInterpolate(Alpha=ConstantBiasScale(None=@n27)) {constA=0.7})
  n29 = Add(
   A=ScalarParameter() {group=Noise, parameterName=NoiseAdd}
   B=Multiply(A=Multiply(A=@n20, B=@n28), B=ScalarParameter() {defaultValue=2, group=Noise, parameterName=NoiseMult})
  )
  n30 = Desaturation(None=Saturate(None=Multiply(A=@n13, B=@n29)))
  n31 = WorldPosition()
  n32 = ComponentMask(None=MaterialFunctionCall() {materialFunction=&ObjectPivotPoint}.Object Pivot Location) {b=true}
  n33 = Add(A=@n32, B=ScalarParameter() {defaultValue=300, parameterName=PulseHeight}) {constB=1600}
  n34 = ScalarParameter() {defaultValue=1, parameterName=TurnOn}
  n35 = ScalarParameter() {defaultValue=33, parameterName=PulseRadius}
  n36 = Multiply(A=@n35) {constB=0.1}
  n37 = Multiply(
   A=SphereMask(
    A=ComponentMask(None=@n31.XYZ) {b=true}
    B=LinearInterpolate(A=@n32, B=@n33, Alpha=@n34) {constA=-0.25, constB=1.2}
    Radius=LinearInterpolate(A=@n35, B=@n36, Alpha=@n34) {constA=500, constB=100}
   ) {hardnessPercent=0}
   B=OneMinus(None=@n34)
  )
  n38 = Multiply(A=@n37) {constB=2}
  n39 = QualitySwitch(
   Default=MaterialFunctionCall(In=Add(A=@n30, B=@n38), Contrast=ScalarParameter() {defaultValue=0.2, group=Noise, parameterName=NoiseContrast}) {materialFunction=&CheapContrast}.Result
   Low=Constant()
  )
  n40 = LinearInterpolate(B=ScalarParameter() {defaultValue=0.6, group=Noise, parameterName=NoiseFresnelXPMult}, Alpha=@n39) {constA=1, constB=0.5}
  n41 = Multiply(A=ScalarParameter() {defaultValue=1, parameterName=FresnelExp}, B=@n40)
  n42 = TextureSampleParameter2D() {group=Texture, parameterName=2DHologramTexture, texture=/Engine/EngineResources/WhiteSquareTexture}
  n43 = Saturate(None=@n39)
  n44 = Multiply(A=ScalarParameter() {defaultValue=-0.4, group=Noise, parameterName=NoiseBrightness}, B=@n43) {constB=0.1}
  n45 = LinearInterpolate(
   A=ScalarParameter() {parameterName=MaskMin}
   Alpha=ChannelMaskParameter(None=TextureSampleParameter2D() {group=Texture, parameterName=MaskTexture, texture=&WhiteSquareTexture}.RGB) {maskChannel=Green, parameterName=MaskChannel, defaultValue=<rgba 0 1 0 0>}.R
  )
  n46 = CameraPositionWS()
  n47 = ActorPositionWS()
  n48 = Normalize(VectorInput=Multiply(A=Constant3Vector() {constant=<rgba 1 1 0 0>}, B=Subtract(A=@n46, B=@n47)))
  n49 = Constant3Vector() {constant=<rgba 0 1 0 0>}
  n50 = Divide(A=MaterialFunctionCall!(Normalized V1=@n48, Normalized V2=Normalize(VectorInput=@n49)) {materialFunction=/Engine/Functions/Engine_MaterialFunctions02/Math/AngleBetweenVectors}.Angle In Degrees) {constB=360}
  n51 = Normalize(VectorInput=Constant3Vector() {constant=<rgba 1 0 0 0>})
  n52 = Multiply(A=Sign(None=DotProduct(A=@n48, B=@n51))) {constB=-1}
  n53 = StaticSwitchParameter(
   True=MaterialFunctionCall(Rotation Axis=Constant3Vector() {constant=<rgba 0 0 1 0>}, Rotation Angle=Multiply(A=@n50, B=@n52)) {materialFunction=&FixRotateAboutAxisNormals}.New World Space Vertex Normal
   False=PixelNormalWS()
  ) {parameterName=FaceCamera}
  n54 = Fresnel(ExponentIn=Multiply(A=@n40, B=ScalarParameter() {defaultValue=1, parameterName=EdgeFresnelXP}), Normal=@n53) {baseReflectFraction=0, exponent=1}
  n55 = Multiply(
   A=Fresnel(ExponentIn=@n41, BaseReflectFractionIn=ScalarParameter() {defaultValue=0.5, parameterName=FresnelMin}, Normal=@n53) {baseReflectFraction=0, exponent=0.5}
   B=Saturate(None=LinearInterpolate(Alpha=@n54) {constA=5, constB=-0.4})
  )
  n56 = ScalarParameter() {defaultValue=0.7, parameterName=ScanlineColorMinInfluence}
  n57 = Ceil(None=@n43)
  n58 = WorldPosition()
  n59 = ScalarParameter() {defaultValue=20000, parameterName=ScaleDistance}
  n60 = CameraPositionWS()
  n61 = ActorPositionWS()
  n62 = Saturate(None=Divide(A=Distance(A=@n60, B=@n61), B=@n59) {constB=1000})
  n63 = Multiply(
   A=Multiply(A=ScalarParameter() {defaultValue=200, parameterName=ScanlineTile}, B=LinearInterpolate(Alpha=@n57) {constA=1, constB=1.25})
   B=Divide(A=ComponentMask(None=@n58.XYZ) {b=true}, B=LinearInterpolate(B=@n59, Alpha=@n62) {constA=50, constB=100}) {constB=20}
  ) {constA=1}
  n64 = DotProduct(A=VertexNormalWS(), B=Constant3Vector() {constant=<rgba 0 0 1 0>})
  n65 = LinearInterpolate(A=ConstantBiasScale(None=Sine(None=@n63)), Alpha=Power(Base=Abs(None=@n64)) {constExponent=5}) {constB=0.25}
  n66 = Multiply(A=Add(A=Power(Base=@n55), B=@n44), B=Saturate(None=Add(A=@n56, B=@n65)))
  n67 = ComponentMask(None=Subtract(A=WorldPosition().XYZ, B=ObjectPositionWS())) {b=true}
  n68 = ObjectRadius()
  n69 = Saturate(None=ConstantBiasScale(None=Divide(A=@n67, B=@n68)))
  n70 = LinearInterpolate(A=@n69, B=OneMinus(None=@n69), Alpha=ScalarParameter() {parameterName=FlipGradient})
  n71 = ScalarParameter() {defaultValue=0.4, parameterName=VerticalGradientInflience}
  n72 = Add(A=Multiply(A=ScalarParameter() {defaultValue=1, parameterName=TurnOn}, B=LinearInterpolate(A=@n66, B=@n70, Alpha=@n71)), B=@n37)
  n73 = StaticSwitchParameter(
   True=Add(A=Power(Base=@n42.G, Exp=@n41), B=Multiply(A=@n42.G, B=@n44))
   False=DepthFade(Opacity=Multiply(A=@n45, B=@n72), FadeDistance=ScalarParameter() {defaultValue=5, parameterName=FadeDistance})
  ) {parameterName=2DHologram}
  n74 = ScalarParameter() {defaultValue=0.6, parameterName=MinScanlineOpacity}
  n75 = Multiply(A=@n39) {constB=1.5}
  n76 = LinearInterpolate(A=Clamp(None=Subtract(A=@n74, B=@n75)) {minDefault=-0.25}, B=Add(A=@n75), Alpha=Power(Base=@n65) {constExponent=1.5}) {constA=0.5}
  n77 = Multiply(A=Saturate(None=Multiply(A=@n73, B=@n76)), B=ScalarParameter() {defaultValue=1.25, parameterName=OpacityMult})
  n78 = Constant()
  n79 = If(
   A=Add(A=PixelDepth().R, B=ScalarParameter() {defaultValue=-0.25, parameterName=BiasSorting_Far})
   B=ComponentMask(None=SceneTexture() {sceneTextureId=PPI_CustomDepth}.Color) {r=true}
   "A > B"=@n78
   "A == B"=@n78
   "A < B"=Constant() {r=1}
  )
  n80 = Constant() {r=1}
  n81 = FeatureLevelSwitch(Default=@n79, ES2=@n80, ES3_1=@n80, SM4_REMOVED=@n79)
  n82 = Multiply(A=Time(), B=ScalarParameter() {defaultValue=5, parameterName=StrobeSpeed})
  n83 = Clamp(None=Multiply(A=Sine(None=@n82), B=ScalarParameter() {defaultValue=2, parameterName=StrobeHardness}))
  n84 = ScalarParameter() {parameterName=Strobe}
  n85 = LinearInterpolate(
   B=Multiply(A=Multiply(A=@n77, B=@n81), B=LinearInterpolate(B=@n83, Alpha=@n84) {constA=1})
   Alpha=ScalarParameter() {defaultValue=1, parameterName=use digital lines}
  ) {constA=1}
  n86 = TextureCoordinate()
  n87 = Constant() {r=0.5}
  n88 = SphereMask(
   A=@n86
   B=@n87
   Radius=ScalarParameter() {defaultValue=0.5, group=Ring Prams, parameterName=Outer Rad}
   Hardness=ScalarParameter() {defaultValue=0.9, group=Ring Prams, parameterName=Hardness Radiner}
  )
  n89 = SphereMask(
   A=@n86
   B=@n87
   Radius=ScalarParameter() {defaultValue=0.35, group=Ring Prams, parameterName=Inner Rad}
   Hardness=ScalarParameter() {defaultValue=0.9, group=Ring Prams, parameterName=Hardness Rad outer}
  ) {attenuationRadius=0.35}
  n90 = Power(Base=LinearInterpolate(
    A=Subtract(A=@n88, B=@n89)
    B=TextureSample() {samplerType=SAMPLERTYPE_LinearGrayscale, texture=/MidMatchAssignedGameplay/Items/Environmental/VFX/ContainerFX/T_Loading_Diamond}.R
    Alpha=ScalarParameter() {parameterName=Use Diamond}
   )) {constExponent=2.2}
  MakeMaterialAttributes!(
   BaseColor
   Metallic
   Specular
   Roughness
   Anisotropy
   EmissiveColor=ComponentMask(None=@n3) {b=true, g=true, r=true}
   Opacity=Multiply(A=Multiply(A=@n4, B=@n6), B=Multiply(A=@n85, B=@n90))
   OpacityMask
   Normal
   Tangent
   WorldPositionOffset
   SubsurfaceColor
   ClearCoat
   ClearCoatRoughness
   AmbientOcclusion
   Refraction
   CustomizedUVs_0
   CustomizedUVs_1
   CustomizedUVs_2
   CustomizedUVs_3
   CustomizedUVs_4
   CustomizedUVs_5
   CustomizedUVs_6
   CustomizedUVs_7
   PixelDepthOffset
   ShadingModel
   Displacement
  )
  MaterialFunctionCall() {materialFunction=&DebugTimeSine}
  MaterialFunctionCall() {materialFunction=&DebugTimeSine}
  n91 = Constant() {r=5}
  Round(None=Multiply(A=MaterialFunctionCall!(Falloff=@n91) {materialFunction=/Engine/Functions/Engine_MaterialFunctions01/Gradient/DiamondGradient}.DiamondGradient))
 ]
}
