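# Niagara effect for an augment container's portal rift teleport, layering crack meshes, shockwave, layered flare sprites, ground decal-style sprite and smoke poof.
# GPU glass shards and sparks burst in five loops: shards use BoneFlow skeletal sampling by loop index, sparks take color and lifetime from a caustics texture.
# A fading light carries post-process, with visibility set by LocalPostProcessEffect.
# bIsFrontEndPreview suppresses smoke; the Ref emitter is disabled.
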NiagaraSystem /MidMatchAssignedGameplay/Items/Environmental/VFX/ContainerFX/NS_AugContainer_Portal_Rift_Teleport
dependencies {
 usedDataInterfaces [
  &NiagaraDataInterfaceCurve
  &NiagaraDataInterfaceSkeletalMesh
  &NiagaraDataInterfaceVectorCurve
  /Script/Niagara.NiagaraDataInterfaceTexture
  &NiagaraDataInterfaceColorCurve
  &NiagaraDataInterfaceVector2DCurve
 ]
 usedDynamicInputs []
 usedModules [&SystemState]
}
emitters {
 CrackSpline {
  emitterUpdateScript [
   &EmitterState(
    Inactive Response=::NewEnumerator1
    Life Cycle Mode=::NewEnumerator1
    Loop Behavior=::NewEnumerator1
    Loop Delay=0.05
    Loop Duration=2
    UseLoopDelay=true
   )
   &SpawnBurst_Instantaneous(Use Spawn Probability=false, UseLoopCountLimit=false)
   &SpawnBurst_Instantaneous(Spawn Time=0.05, Use Spawn Probability=false, UseLoopCountLimit=false) {module=SpawnBurst_Instantaneous001}
  ]
  particleSpawnScript [
   &InitializeParticle(
    Lifetime Max=0.7
    Lifetime Mode=::NewEnumerator1
    Mesh Scale=&Multiply_VectorByFloat(Vector=&RandomRangeVector(Minimum=<2 2 3>, Maximum=<4 4 5>), Float=3)
    Mesh Scale Mode=::NewEnumerator4
    UsePositionOffset=false
    Write Material Random=true
   ) {module=InitializeParticle001}
   &Color(Color=&MakeLinearColorFromVectorAndFloat("Vector (RGB)"=<1.5 1.5 1.5>, "Float (Alpha)"=&ClampFloat(Float=1)))
   &CameraOffset(Camera Offset Amount=-25)
   &InitialMeshOrientation(Use Rotation Vector=false)
  ]
  particleUpdateScript [
   &ParticleState()
   &ScaleMeshSize(Scale Factor=&VectorFromFloat(Value=&FloatFromCurve(
      FloatCurve=DataInterface() {propertyValues {Curve {keys [
          {
           arriveTangent=6.059
           arriveTangentWeight=0.9566
           interpMode=RCIM_Cubic
           leaveTangent=6.059
           leaveTangentWeight=0.9566
           tangentMode=RCTM_User
           tangentWeightMode=RCTWM_WeightedBoth
           time=0.1
           value=0
          }
          {arriveTangentWeight=1.1463, interpMode=RCIM_Cubic, leaveTangent=7.9844, leaveTangentWeight=1.1463, time=0.5, value=1}
         ]}}}
      Scale Curve=&RandomRangeFloat(Minimum=0.8, Maximum=1.2)
     )))
   &ScaleColor(
    Scale Alpha=&FloatFromCurve(FloatCurve=DataInterface() {propertyValues {Curve {keys [<at 0.77 1 RCIM_Cubic>,<at 1 0>]}}})
    Scale RGB=&VectorFromCurve(
     VectorCurve=DataInterface() {propertyValues {
       XCurve {keys [
         {arriveTangent=-6.9225, interpMode=RCIM_Cubic, time=0, value=4.5}
         {arriveTangent=-3.6158, interpMode=RCIM_Cubic, leaveTangent=-3.6158, time=0.65, value=1}
         <at 1 0.8842 RCIM_Cubic>
        ]}
       YCurve {keys [
         {arriveTangent=-15.8226, interpMode=RCIM_Cubic, time=0, value=6}
         {arriveTangent=-5.6986, interpMode=RCIM_Cubic, leaveTangent=-5.6986, time=0.65, value=1}
         {interpMode=RCIM_Cubic, leaveTangent=-5.843, time=1, value=0.3014}
        ]}
       ZCurve {keys [
         {arriveTangent=-15.2224, interpMode=RCIM_Cubic, time=0, value=6}
         {arriveTangent=-5, interpMode=RCIM_Cubic, leaveTangent=-5, time=0.65, value=1}
         <at 1 1 RCIM_Cubic>
        ]}
      }}
     Scale Curve=<0.12 0.12 0.12>
    )
    ScaleA=true
    ScaleRGB=true
   )
   &DynamicMaterialParameters(
    Index 0 Param 1=&FloatFromCurve(FloatCurve=DataInterface() {propertyValues {Curve {keys [
         <at 0 0 RCIM_Cubic>
         <at 0.1 1>
         {arriveTangent=-1.1111, interpMode=RCIM_Cubic, leaveTangent=-1.1111, time=0.3329, value=0.616}
         {interpMode=RCIM_Cubic, leaveTangent=-0.3905, time=1, value=0}
        ]}}})
    Index 0 Param 2=&FloatFromCurve(
     FloatCurve=DataInterface() {propertyValues {Curve {keys [
         {arriveTangent=-2.0225, interpMode=RCIM_Cubic, time=0, value=0}
         <at 0.1 0.6>
         {arriveTangent=-0.6667, interpMode=RCIM_Cubic, leaveTangent=-0.6667, time=0.65, value=0.25}
         <at 1 0 RCIM_Cubic>
        ]}}}
     Scale Curve=1.5
    )
    Index 0 Param 3=&FloatFromCurve(FloatCurve=DataInterface() {propertyValues {Curve {keys [{arriveTangent=1.0145, interpMode=RCIM_Cubic, time=0, value=-0.1}, <at 0.25 0>]}}})
    Index 0 Param 4=&FloatFromCurve(FloatCurve=DataInterface() {propertyValues {Curve {keys [<at 0 0 RCIM_Cubic>,<at 1 0>]}}})
    Module.Param0WriteEnabled=true
    Module.Param1WriteEnabled=true
    Module.Param2WriteEnabled=true
    Module.Param3WriteEnabled=true
   )
  ]
  rendererClasses [&NiagaraMeshRendererProperties]
  renderers [{rendererClass=&NiagaraMeshRendererProperties, rendererIndex=0}]
 }
 Glass {
  simTarget=GPUComputeSim
  emitterUpdateScript [
   &EmitterState(
    Delay First Loop Only=true
    Life Cycle Mode=::NewEnumerator1
    Loop Behavior=::NewEnumerator2
    Loop Count=5
    Loop Delay=0.6
    Loop Duration=0.1
    Recalculate Duration Each Loop=true
    UseLoopDelay=true
   )
   &SpawnBurst_Instantaneous(
    Spawn Count=&RandomRangeInt(Minimum=15, Maximum=50, Evaluation Type=::NewEnumerator1, Random Seed=$Emitter.LoopCount, Fixed Random Seed=true)
    Use Spawn Probability=false
   )
  ]
  particleSpawnScript [
   /SetVariables(Particles.SpawnGroup=$Emitter.LoopCount) {isSetParameters=true, module=SetVariables_#1}
   &InitializeParticle(
    Lifetime Min=0.2
    Lifetime Mode=::NewEnumerator1
    Mesh Scale Mode=::NewEnumerator1
    Mesh Uniform Scale Max=0.35
    Mesh Uniform Scale Min=0.15
    Sprite Size Mode=::NewEnumerator3
    UsePositionOffset=false
    Write Material Random=true
   )
   /Game/Effects/Niagara/Modules/BoneFlow/FN_BoneFlow_v4/FN_BoneFlow(
    Angle:NiagaraFloat=&RandomRangeFloat(Minimum=0, Maximum=1)
    BoneID:NiagaraInt32=$Particles.SpawnGroup
    Height:NiagaraFloat=&RandomRangeFloat(Minimum=0, Maximum=1)
    Interpolate Mesh Sampling=false
    Radius:NiagaraFloat=&RandomRangeFloat(Minimum=10, Maximum=20)
    SkeletalMesh=DataInterface() {propertyValues {
      DefaultMesh=null
      ExcludeBoneName=root
      OptionalComponentClass=null
      PreviewMesh=/Game/Characters/Player/Male/Male_Avg_Base/M_Med_Base_Skeleton
      SkinningMode=SkinOnTheFly
      SoftSourceActor=null
      SourceMode=Default
      UvSetIndex=0
      WholeMeshLOD=0
      bExcludeBone=false
      bReadDeformedGeometry=true
      bRequireCurrentFrameData=true
      ComponentTags []
      FilteredBones [upperarm_l,lowerarm_l,hand_l,upperarm_r,lowerarm_r,hand_r,thigh_l,calf_l,foot_l,thigh_r,calf_r,foot_r,neck_02,spine_02,pelvis]
      FilteredSockets []
      MeshUserParameter {parameter {name=None, typeDefHandle {registeredTypeIndex=1}, varData []}}
      SamplingRegions []
     }}
    Write Position Attribute=true
   ) {module=FN_BoneFlow001}
   /SetVariables(Particles.MeshIndex=&RandomRangeInt(Minimum=0, Maximum=2)) {isSetParameters=true}
   &InitialMeshOrientation(Orientation Vector=&RandomVector(Vector Scale=1), Use Orientation Vector=true, Use Rotation Vector=false)
   &MeshRotationForce(Rotation=&RandomVector(Vector Scale=10))
   &CurlNoiseForce(Noise Frequency=150, Noise Strength=30, Pan Noise=false)
   &ApplyInitialForces()
   &AddVelocity(Velocity=&Multiply_VectorByFloat(
     Vector=&NormalizeVector(Vector To Normalize=$Output.FN_BoneFlow001.RotatedBinormal)
     Float=&Multiply_Float(
      A=&RandomRangeFloat(Minimum=200, Maximum=250)
      B=/Game/Effects/Niagara/DynamicInputs/Random/FN_SeededRandomRangeFloat(Minimum:NiagaraFloat=0.3, Maximum:NiagaraFloat=1, Random Seed:NiagaraInt32=$Particles.SpawnGroup)
     )
    ))
   &SystemLocation(
    MaskSpawnGroup=false
    Offset=&RandomVector(Vector Scale=&RandomRangeFloat(Minimum=0, Maximum=75))
   )
  ]
  particleUpdateScript [
   &ParticleState()
   &CurlNoiseForce(
    Noise Frequency=20
    Noise Strength=&FloatFromCurve(
     FloatCurve=DataInterface() {propertyValues {Curve {keys [{arriveTangent=-3.2987, interpMode=RCIM_Cubic, leaveTangent=-3.2987, tangentMode=RCTM_User, time=0, value=1}, <at 1 0>]}}}
     Scale Curve=300
    )
    Pan Noise=false
   ) {module=CurlNoiseForce001}
   &GravityForce(Gravity=<100 0 -10>)
   &Drag(Drag=2, Rotational Drag=0.01, Use Linear Drag=true, Use Rotational Drag=true)
   &SolveForcesAndVelocity(Clamp Velocity=false, Limit Acceleration=false)
   &ScaleMeshSize(Scale Factor=&VectorFromFloat(Value=&FloatFromCurve(FloatCurve=DataInterface() {propertyValues {Curve {keys [<at 0 1 RCIM_Cubic>,<at 1 0 RCIM_Cubic>]}}})))
   &ScaleColor(
    Scale RGB=/Niagara/DynamicInputs/ScaleByCurve/ScaleVectorByCurve(
     InitialValue:Vector3f=<1 1 1>
     ScaleCurve:NiagaraDataInterfaceVectorCurve=DataInterface() {propertyValues {
       XCurve {keys [{arriveTangent=-4.3916, interpMode=RCIM_Cubic, leaveTangent=-4.3916, tangentMode=RCTM_User, time=0.0032, value=8.5279}, <at 0.9366 3.0287>]}
       YCurve {keys [{arriveTangent=-4.8434, interpMode=RCIM_Cubic, leaveTangent=-4.8434, tangentMode=RCTM_User, time=0.0072, value=8.4747}, <at 0.9366 3.0322>]}
       ZCurve {keys [{arriveTangent=-5.0879, interpMode=RCIM_Cubic, leaveTangent=-5.0879, tangentMode=RCTM_User, time=-0.0049, value=8.7936}, <at 0.9407 2.9968>]}
      }}
     CurveIndex:NiagaraFloat=$Particles.NormalizedAge
    )
    ScaleA=true
    ScaleRGB=true
   )
  ]
  rendererClasses [&NiagaraMeshRendererProperties]
  renderers [{rendererClass=&NiagaraMeshRendererProperties, rendererIndex=0}]
 }
 Light {
  emitterUpdateScript [
   &EmitterState()
   &SpawnBurst_Instantaneous(Spawn Count=1, Spawn Time=0.01, Use Spawn Probability=false)
  ]
  particleSpawnScript [
   &InitializeParticle(
    Color Mode=::NewEnumerator0
    Lifetime=1.5
    Sprite Size Mode=::NewEnumerator3
    Uniform Sprite Size=400
    UsePositionOffset=false
    Write Material Random=true
   )
   /SetVariables(Particles.PostProcessVisibility:NiagaraInt32=&MakeCustomIntFromBool(Bool=$User.LocalPostProcessEffect, True INT=1, False INT=0)) {isSetParameters=true}
  ]
  particleUpdateScript [
   &ParticleState()
   &FN_Light_Attributes(
    Color=<rgba 1 2 10 1>
    Create Color Attribute=true
    Create Position Attribute=true
    Create Radius Attribute=true
    Enable Visibility Tag=true
    Position=&Add_Vector(
     A=Linked() {linkedVariable {name=Particles.Position, type=NiagaraPosition}}
     B=&TransformVector(Vector=<0 0 50>, Source Space=::World, Destination Space=::Local)
    )
    Radius=&FloatFromCurve(
     FloatCurve=DataInterface() {propertyValues {Curve {keys [{arriveTangent=-2.2238, interpMode=RCIM_Cubic, leaveTangent=-2.2238, tangentMode=RCTM_User, time=0, value=1}, <at 0.9922 0.4552>]}}}
     Scale Curve=1000
    )
    Scale Color=&FloatFromCurve(
     FloatCurve=DataInterface() {propertyValues {Curve {keys [
         <at -0.0014 0.984 RCIM_Cubic>
         <at 0.0884 0.6894>
         <at 0.2427 0.8309>
         {arriveTangent=-0.8435, interpMode=RCIM_Cubic, leaveTangent=-0.8435, time=0.3969, value=0.4849}
         <at 0.597 0.5321>
         <at 0.7804 0.0444>
         <at 1 0>
        ]}}}
     Scale Curve=0.45
    )
    VisibilityTag=HlslExpression() {hlslExpression="Particles.UniqueID && 1"}
   )
   &FN_Gen5_Light(
    Attenutation:NiagaraFloat=&FloatFromCurve(
     FloatCurve=DataInterface() {propertyValues {Curve {keys [{arriveTangent=-1.1596, interpMode=RCIM_Cubic, leaveTangent=-1.1596, tangentMode=RCTM_User, time=0, value=1}, <at 1 0>]}}}
     Scale Curve=1000
    )
    Color:LinearColor=<rgba 500.0001 999.9993 5000 1>
    Intensity:NiagaraFloat=&FloatFromCurve(
     FloatCurve=DataInterface() {propertyValues {Curve {keys [{arriveTangent=-2.5178, interpMode=RCIM_Cubic, leaveTangent=-2.5178, tangentMode=RCTM_User, time=0, value=1}, <at 1 0>]}}}
     Scale Curve=6000
    )
    Position Offset:Vector3f=&TransformVector(Vector=<0 0 50>, Source Space=::World, Destination Space=::Local)
   )
   /Game/Effects/Niagara/Modules/ComponentRenderer/FN_PostProcess(Blend:NiagaraFloat=&FloatFromCurve(FloatCurve=DataInterface() {propertyValues {Curve {keys [{arriveTangent=-1.9521, interpMode=RCIM_Cubic, leaveTangent=-1.9521, tangentMode=RCTM_User, time=0, value=1}, <at 1 0>]}}}))
  ]
  rendererClasses [&NiagaraLightRendererProperties, &NiagaraComponentRendererProperties]
  renderers [
   {rendererClass=&NiagaraLightRendererProperties, rendererIndex=0}
   {rendererClass=&NiagaraComponentRendererProperties, rendererIndex=1}
   {rendererClass=&NiagaraComponentRendererProperties, rendererIndex=2}
  ]
 }
 Particle_Decale {
  emitterUpdateScript [
   &EmitterState(Life Cycle Mode=::NewEnumerator1, Loop Behavior=::NewEnumerator1, Loop Duration=1, UseLoopDelay=false)
   &SpawnBurst_Instantaneous(Use Spawn Probability=false, UseLoopCountLimit=false)
  ]
  particleSpawnScript [
   &InitializeParticle(
    Color=<rgba 3.75 18.625 25 1>
    Lifetime=0.9
    Position Offset=<0 0 -2>
    Sprite Size=<75 75>
    Sprite Size Mode=::NewEnumerator4
    UsePositionOffset=true
    Write Material Random=true
   )
   &CameraOffset(Camera Offset Amount=40)
  ]
  particleUpdateScript [
   &ParticleState()
   &ScaleSpriteSize(
    Uniform Curve Index=$Particles.Age
    Uniform Curve Sprite Scale=DataInterface() {propertyValues {Curve {keys [
        {tangentMode=RCTM_User, time=0, value=0}
        {arriveTangentWeight=0.0667, interpMode=RCIM_Cubic, leaveTangentWeight=0.0667, tangentMode=RCTM_User, time=0.2, value=0}
        {arriveTangent=2.8114, arriveTangentWeight=0.0667, interpMode=RCIM_Cubic, leaveTangent=2.8114, tangentMode=RCTM_User, time=0.4, value=1}
       ]}}}
   )
   &ScaleColor(
    Scale Alpha=&FloatFromCurve(FloatCurve=DataInterface() {propertyValues {Curve {keys [<at -0.0046 1.0113 RCIM_Cubic>,<at 1.0011 0.002 RCIM_Cubic>]}}})
    ScaleA=true
    ScaleRGB=true
   )
   &DynamicMaterialParameters(
    Index 0 Param 1=1
    Index 0 Param 2=1
    Index 0 Param 3=1
    Index 0 Param 4=&RandomRangeFloat(Minimum=0, Maximum=1)
    Index 1 Param 1=69.5807
    Index 1 Param 2=25
    Module.Param0WriteEnabled=true
    Module.Param0WriteEnabled1=true
    Module.Param1WriteEnabled=true
    Module.Param1WriteEnabled1=true
    Module.Param2WriteEnabled=true
    Module.Param2WriteEnabled1=true
    Module.Param3WriteEnabled=true
    Module.Param3WriteEnabled1=true
    Write Parameter Index 1=true
   )
   /SetVariables(Particles.Offset Position:NiagaraPosition=&AddVectorToPosition(
     Position=$Particles.Position
     Vector=&MakeVector(
      X=0
      Y=0
      Z=&Lerp_Float(
       A=-35.0433
       B=2.8221
       Alpha=&FloatFromCurve(
        FloatCurve=DataInterface() {propertyValues {Curve {keys [<at 0.2 0 RCIM_Cubic>, {arriveTangent=2.8114, interpMode=RCIM_Cubic, leaveTangent=2.8114, tangentMode=RCTM_User, time=0.4, value=1}]}}}
        CurveIndex=$Particles.Age
       )
      )
     )
    )) {isSetParameters=true}
  ]
  rendererClasses [&NiagaraSpriteRendererProperties]
  renderers [{rendererClass=&NiagaraSpriteRendererProperties}]
 }
 Ref {
  bEnabled=false
  emitterUpdateScript [
   &EmitterState(Life Cycle Mode=::NewEnumerator1, UseLoopDelay=false)
   &SpawnBurst_Instantaneous(Use Spawn Probability=false, UseLoopCountLimit=false)
  ]
  particleSpawnScript [&InitializeParticle(Color Mode=::NewEnumerator0, Lifetime=5, Position Offset=<0 0 -35>, UsePositionOffset=true, Write Material Random=true)]
  particleUpdateScript [
   &ParticleState()
   /SetVariables(Particles.New Vector:Vector3f=&VectorFromFloat(Value=&FloatFromCurve(
      FloatCurve=DataInterface() {propertyValues {Curve {keys [
          {interpMode=RCIM_Cubic, leaveTangentWeight=0.0667, tangentMode=RCTM_User, time=0.2, value=0}
          {arriveTangent=2.8114, arriveTangentWeight=0.0667, interpMode=RCIM_Cubic, leaveTangent=2.8114, tangentMode=RCTM_User, time=0.4, value=1}
         ]}}}
      CurveIndex=$Emitter.LoopedAge
     ))) {isSetParameters=true}
  ]
  rendererClasses [
   &NiagaraSpriteRendererProperties
   &NiagaraComponentRendererProperties
   &NiagaraMeshRendererProperties
  ]
  renderers [
   {rendererClass=&NiagaraSpriteRendererProperties}
   {rendererClass=&NiagaraComponentRendererProperties, rendererIndex=1}
   {rendererClass=&NiagaraMeshRendererProperties, rendererIndex=2}
  ]
 }
 ShockWave {
  emitterUpdateScript [
   &EmitterState()
   &SpawnBurst_Instantaneous(Spawn Time=0.2, Use Spawn Probability=false, UseLoopCountLimit=false)
  ]
  particleSpawnScript [
   &InitializeParticle(
    Color=<rgba 1 1 1 0.035>
    Lifetime=0.5
    Mesh Scale Mode=::NewEnumerator3
    Mesh Uniform Scale=1.25
    UsePositionOffset=false
    Write Material Random=true
   ) {module=InitializeParticle001}
   &Color(Color=&MakeLinearColorFromVectorAndFloat("Vector (RGB)"=<1 1 1>, "Float (Alpha)"=&ClampFloat(Float=0.1))) {enabled=false}
  ]
  particleUpdateScript [
   &ParticleState()
   &ScaleColor(
    Scale Alpha=&FloatFromCurve(FloatCurve=DataInterface() {propertyValues {Curve {keys [
         {interpMode=RCIM_Cubic, tangentMode=RCTM_User, time=0, value=0}
         {arriveTangent=-3.1257, interpMode=RCIM_Cubic, leaveTangent=-3.1257, tangentMode=RCTM_User, time=0.4, value=1}
         {tangentMode=RCTM_User, time=1, value=0}
        ]}}})
    ScaleA=true
    ScaleRGB=true
   )
   &CameraOffset(Camera Offset Amount=20)
  ]
  rendererClasses [&NiagaraMeshRendererProperties]
  renderers [{rendererClass=&NiagaraMeshRendererProperties, rendererIndex=0}]
 }
 Smoke_Poof {
  emitterUpdateScript [
   &EmitterState(Life Cycle Mode=::NewEnumerator1, Loop Behavior=::NewEnumerator1, Loop Delay=0.2, Loop Duration=1, UseLoopDelay=true)
   &SpawnBurst_Instantaneous(
    Spawn Count=&RandomRangeInt(Minimum=40, Maximum=30)
    Spawn Probability=&MakeCustomFloatFromBool(Bool=$User.bIsFrontEndPreview, True Float=0, False Float=1)
    Use Spawn Probability=true
    UseLoopCountLimit=false
   )
  ]
  particleSpawnScript [
   /SetVariables(Particles.SpriteRotationRate=&RandomRangeFloat(Minimum=-0.8, Maximum=0.8, Evaluation Type=::NewEnumerator1)) {isSetParameters=true}
   &InitializeParticle(
    Color=<rgba 0.35 0.805 1 0.2>
    Lifetime Max=1.2
    Lifetime Min=0.8
    Lifetime Mode=::NewEnumerator1
    Position Offset=<0 0 -45>
    Sprite Rotation Mode=::NewEnumerator2
    Sprite Size Mode=::NewEnumerator1
    Uniform Sprite Size Max=15
    Uniform Sprite Size Min=25
    UsePositionOffset=true
    Write Material Random=true
   ) {module=InitializeParticle001}
   &CameraOffset(Camera Offset Amount=5)
   &AddVelocity(
    Cone Angle=160
    Cone Axis=<0 0 1>
    Inner Cone Angle=120
    Module.UseConeFalloff=false
    Module.UseVelDistribution=false
    Override Seed=false
    Rotation Coordinate Space=::World
    Velocity Mode=::NewEnumerator2
    Velocity Speed=600
   )
  ]
  particleUpdateScript [
   &ParticleState()
   &SpriteRotationRate(Rotation Rate=&RandomRangeFloat(Minimum=-0.8, Maximum=0.8))
   &ScaleColor(
    Scale Alpha=&FloatFromCurve(FloatCurve=DataInterface() {propertyValues {Curve {keys [
         {arriveTangent=9.0725, interpMode=RCIM_Cubic, time=0, value=0}
         {arriveTangent=-2.5249, interpMode=RCIM_Cubic, leaveTangent=-2.5249, tangentMode=RCTM_User, time=0.1408, value=0.9847}
         {arriveTangent=-0.0405, interpMode=RCIM_Cubic, leaveTangent=-0.0405, tangentMode=RCTM_User, time=1, value=0}
        ]}}})
    ScaleA=true
    ScaleRGB=true
   )
   &ScaleSpriteSize(Scale Factor=&Vector2DFromFloat(Value=&FloatFromCurve(
      FloatCurve=DataInterface() {propertyValues {Curve {keys [
          {arriveTangent=1.9916, interpMode=RCIM_Cubic, time=0, value=0.5}
          {arriveTangent=0.8423, interpMode=RCIM_Cubic, leaveTangent=0.8423, tangentMode=RCTM_User, time=0.2, value=0.75}
          <at 1 1 RCIM_Cubic>
         ]}}}
      Scale Curve=4
     )))
   &AccelerationForce(Acceleration=<0 0 40>)
   &Drag(
    Drag=&FloatFromCurve(
     FloatCurve=DataInterface() {propertyValues {Curve {keys [{interpMode=RCIM_Cubic, tangentMode=RCTM_User, time=0, value=5}, {interpMode=RCIM_Cubic, tangentMode=RCTM_User, time=1, value=6}]}}}
     Scale Curve=1.75
    )
    Use Linear Drag=true
    Use Rotational Drag=true
   )
   &CurlNoiseForce(Noise Quality / Cost=::NewEnumerator0, Noise Strength=1000, Pan Noise=true, Pan Noise Field=<0 0 1>) {enabled=false}
   &SolveForcesAndVelocity(Clamp Velocity=false, Limit Acceleration=false)
   &DynamicMaterialParameters(
    Index 0 Param 1=&FloatFromCurve(FloatCurve=DataInterface() {propertyValues {Curve {keys [<at 0 1>,<at 1 0>]}}}, Scale Curve=0.2)
    Module.Param0WriteEnabled=true
    Module.Param1WriteEnabled=false
    Module.Param2WriteEnabled=false
    Module.Param3WriteEnabled=false
   )
  ]
  rendererClasses [&NiagaraSpriteRendererProperties]
  renderers [{rendererClass=&NiagaraSpriteRendererProperties}]
 }
 Sparks {
  simTarget=GPUComputeSim
  emitterUpdateScript [
   &EmitterState(
    Delay First Loop Only=true
    Life Cycle Mode=::NewEnumerator1
    Loop Behavior=::NewEnumerator2
    Loop Count=5
    Loop Delay=0.6
    Loop Duration=0.1
    UseLoopDelay=true
   )
   &SpawnBurst_Instantaneous(Spawn Count=300, Use Spawn Probability=false)
  ]
  particleSpawnScript [
   /SetVariables(Particles.SpawnGroup=$Emitter.LoopCount) {isSetParameters=true, module=SetVariables_#1}
   /SetVariables(
    Particles.U:NiagaraFloat=&RandomRangeFloat(Minimum=0, Maximum=1)
    Particles.V:NiagaraFloat=&RandomRangeFloat(Minimum=0, Maximum=1)
   ) {isSetParameters=true}
   /Niagara/Modules/Update/Textures/SampleTexture(
    Texture=DataInterface() {propertyValues {
      Texture=&T_ShieldCaustics
      TextureUserParameter {parameter {name=None, typeDefHandle {registeredTypeIndex=2}, varData []}}
     }}
    UV:Vector2f=&MakeVector2D(X=&Multiply_Float(A=$Particles.U, B=0.1), Y=$Particles.V)
   )
   &ShapeLocation(
    Fixed Random Seed=true
    MaskSpawnGroup=false
    Override Seed=true
    Random Seed:NiagaraInt32=$Emitter.LoopCount
    Sphere Radius=25
    UseOwnerScale=true
   )
   /SetVariables(Particles.SampleTexture:NiagaraFloat=/Game/Effects/Niagara/DynamicInputs/Math/FN_Fit(
     Clamp:NiagaraBool=true
     Value:NiagaraFloat=&VectorLength(Vector=&MakeVectorFromLinearColorRGB(Color=$Output.SampleTexture.SampledColor))
     SourceMin:NiagaraFloat=0.1
     SourceMax:NiagaraFloat=1
     DestMin:NiagaraFloat=0.1
     DestMax:NiagaraFloat=1
    )) {isSetParameters=true}
   &InitializeParticle(
    Color=$Output.SampleTexture.SampledColor
    Lifetime=&FN_Fit(
     Clamp:NiagaraBool=false
     Value:NiagaraFloat=&VectorLength(Vector=&MakeVectorFromLinearColorRGB(Color=$Output.SampleTexture.SampledColor))
     SourceMin:NiagaraFloat=0.2
     SourceMax:NiagaraFloat=1
     DestMin:NiagaraFloat=0.1
     DestMax:NiagaraFloat=&RandomRangeFloat(Minimum=1, Maximum=3)
    )
    Mass Mode=::NewEnumerator3
    Position Mode=::NewEnumerator0
    Sprite Size=<3 5>
    Sprite Size Mode=::NewEnumerator4
    Write Material Random=true
   )
   &CurlNoiseForce(Noise Strength=100, Pan Noise=false)
   &AddVelocity(Velocity=&Multiply_VectorByFloat(
     Vector=$Output.SkeletalMeshLocation.SampledNormal
     Float=&Multiply_Float(A=$Particles.SampleTexture, B=&RandomRangeFloat(Minimum=200, Maximum=400))
    )) {enabled=false}
   &ApplyInitialForces(Apply Rotational Force to Rotational Velocity=false)
  ]
  particleUpdateScript [
   &ParticleState()
   &CurlNoiseForce(
    Noise Frequency=20
    Noise Strength=&FloatFromCurve(
     FloatCurve=DataInterface() {propertyValues {Curve {keys [{arriveTangent=-3.2987, interpMode=RCIM_Cubic, leaveTangent=-3.2987, tangentMode=RCTM_User, time=0, value=1}, <at 1 0>]}}}
     Scale Curve=&RandomRangeFloat(Minimum=200, Maximum=300)
    )
    Pan Noise=false
   ) {module=CurlNoiseForce001}
   &Drag(Drag=2, Use Linear Drag=true, Use Rotational Drag=false)
   &GravityForce(Gravity=<100 0 0>)
   &SolveForcesAndVelocity(Clamp Velocity=false, Limit Acceleration=false)
   &Color(Color=&HueShiftLinearColor(
     Hue=&FloatFromCurve(
      FloatCurve=DataInterface() {propertyValues {Curve {keys [
          <at 0.5003 -0.0104 RCIM_Cubic>
          {arriveTangent=-0.5346, interpMode=RCIM_Cubic, leaveTangent=-0.5346, tangentMode=RCTM_User, time=0.9942, value=1}
         ]}}}
      Scale Curve=0.1
     )
     Saturation=1
     Value=&FloatFromCurve(
      FloatCurve=DataInterface() {propertyValues {Curve {keys [
          {arriveTangent=0.0107, interpMode=RCIM_Cubic, leaveTangent=0.0107, tangentMode=RCTM_User, time=0, value=1}
          <at 0.2146 0.8502>
          <at 0.2468 1.9811>
          <at 0.2765 0.816>
          <at 0.5049 0.1471>
          <at 0.5699 1.4137>
          <at 0.6632 0.3582>
          <at 1 0>
         ]}}}
      Scale Curve=&Multiply_Float(
       A=0.8
       B=&OneMinusFloat(Float=&FN_Fit(
         Clamp:NiagaraBool=false
         Value:NiagaraFloat=$Particles.SampleTexture
         SourceMin:NiagaraFloat=0.4
         SourceMax:NiagaraFloat=0.6
         DestMin:NiagaraFloat=0.5
         DestMax:NiagaraFloat=1
        ))
      )
     )
     Linear Color=&ColorFromCurve(
      ColorCurve=DataInterface() {propertyValues {
        AlphaCurve {keys [<at 0 1 RCIM_Cubic>,<at 1 1 RCIM_Cubic>]}
        BlueCurve {keys [<at 0 1 RCIM_Cubic>,<at 1 1 RCIM_Cubic>]}
        GreenCurve {keys [<at 0 0.4088 RCIM_Cubic>,<at 1 0.0729 RCIM_Cubic>]}
        RedCurve {keys [<at 0 0 RCIM_Cubic>,<at 1 0.1126 RCIM_Cubic>]}
       }}
      CurveIndex=$Particles.SampleTexture
     )
    ))
   /NiagaraFN/Sprite/FN_SpriteAlignmentAndFacing(SpriteAlignment:Vector3f=/Niagara/DynamicInputs/Math/SlerpVector(
     Vector A:Vector3f=&NormalizeVector(Vector To Normalize=$Particles.Velocity)
     Vector B:Vector3f=&RotateVectorByQuaternion(
      Vector To Rotate=&RandomVector(Vector Scale=1)
      Quaternion=&MakeQuatFromAxisAngle(
       Axis=&RandomVector(Vector Scale=1)
       AngleInDegrees=&Multiply_Float(A=$Particles.NormalizedAge, B=&RandomRangeFloat(Minimum=1000, Maximum=2000))
      )
     )
     Alpha:NiagaraFloat=&FloatFromCurve(FloatCurve=DataInterface() {propertyValues {Curve {keys [{arriveTangent=2.9194, interpMode=RCIM_Cubic, time=0.1366, value=0.0107}, <at 0.6462 1.0007>]}}})
     Scale:NiagaraFloat=1
    ))
   &ScaleSpriteSize(Scale Factor=&Vector2DFromCurve(
     Vector2Curve=DataInterface() {propertyValues {
       XCurve {keys [<at -0.0121 0.7831>,<at 0.9963 0.0572>]}
       YCurve {keys [
         {arriveTangent=-8.879, interpMode=RCIM_Cubic, leaveTangent=-8.8789, tangentMode=RCTM_User, time=-0.0078, value=4.6741}
         {arriveTangent=-0.1215, interpMode=RCIM_Cubic, leaveTangent=-0.1215, tangentMode=RCTM_User, time=0.4744, value=0.4347}
         <at 0.9963 0.0572>
        ]}
      }}
     ScaleCurve=<1 1>
    ))
  ]
  rendererClasses [&NiagaraSpriteRendererProperties]
  renderers [{rendererClass=&NiagaraSpriteRendererProperties}]
 }
 ZeroPointFlare {
  emitterUpdateScript [
   &EmitterState(Life Cycle Mode=::NewEnumerator1, Loop Behavior=::NewEnumerator1, Loop Duration=0.95, UseLoopDelay=false)
   &SpawnRate(SpawnRate=7, Use Spawn Probability=false)
  ]
  particleSpawnScript [
   /Niagara/Modules/Spawn/Initialization/InitializeParticle(
    Lifetime=&Multiply_Float(
     A=&RandomRangeFloat(Minimum=0.5, Maximum=0.8)
     B=&FloatFromCurve(
      FloatCurve=DataInterface() {propertyValues {Curve {keys [{arriveTangent=-1.5125, interpMode=RCIM_Cubic, leaveTangent=-1.5125, tangentMode=RCTM_User, time=0, value=1}, <at 0.9949 0.2248>]}}}
      CurveIndex=$Emitter.NormalizedLoopAge
     )
    )
    Module.Write Color=true
    Module.Write Lifetime=true
    Module.Write Mass=true
    Module.Write Material Random=true
    Module.Write Position=false
    Module.Write Scale=false
    Module.Write SpriteRotation=true
    Module.Write SpriteSize=true
    Sprite Rotation=&RandomRangeFloat(Minimum=0, Maximum=360)
    Sprite Size=<3000 3000>
   )
   &Color(Color=&MakeLinearColorFromVectorAndFloat("Vector (RGB)"=<0.0175 0.0243 0.03>, "Float (Alpha)"=&ClampFloat(Float=0.6)))
   &CameraOffset()
  ]
  particleUpdateScript [
   &ParticleState()
   &ScaleColor(
    Scale Alpha=&FloatFromCurve(FloatCurve=DataInterface() {propertyValues {Curve {keys [<at 0 1 RCIM_Cubic>,<at 0.25 1 RCIM_Cubic>,<at 0.75 1 RCIM_Cubic>,<at 1 1>]}}})
    ScaleA=true
    ScaleRGB=true
   )
   /SetVariables(Particles.Size:NiagaraFloat=&FloatFromCurve(
     FloatCurve=DataInterface() {propertyValues {Curve {keys [
         <at -0.0039 0.0032 RCIM_Cubic>
         {arriveTangent=0.0358, interpMode=RCIM_Cubic, leaveTangent=0.0358, tangentMode=RCTM_User, time=0.4967, value=0.9107}
         <at 1 0>
        ]}}}
     Scale Curve=&FloatFromCurve(
      FloatCurve=DataInterface() {propertyValues {Curve {keys [
          <at -0.0005 0.984 RCIM_Cubic>
          {arriveTangent=0.3225, interpMode=RCIM_Cubic, leaveTangent=0.3225, tangentMode=RCTM_User, time=0.4833, value=3.7073}
          <at 1.0108 1.1471>
         ]}}}
      CurveIndex=$Emitter.NormalizedLoopAge
      Scale Curve=0.5
     )
    )) {isSetParameters=true}
   &FN_Sprite_Attributes(
    Camera Offset=-20
    Create Camera Offset Attribute=true
    Create Color Attribute=true
    Create SpriteSize Attribute=true
    SpriteSize=&Vector2DFromFloat(Value=&Multiply_Float(A=500, B=$Particles.Size))
    Write to Module Namespace=true
   )
   &FN_Sprite_Attributes(
    Camera Offset=30
    Color=&MakeLinearColorFromVectorAndFloat("Vector (RGB)"=<0.5625 1.8964 3>, "Float (Alpha)"=0.1)
    Create Camera Offset Attribute=true
    Create Color Attribute=true
    Create SpriteSize Attribute=true
    SpriteSize=&Vector2DFromFloat(Value=&Multiply_Float(A=500, B=$Particles.Size))
    Write to Module Namespace=true
   ) {module=FN_Sprite_Attributes001}
   &FN_Sprite_Attributes(
    Camera Offset=72
    Color=&MakeLinearColorFromVectorAndFloat(
     "Vector (RGB)"=<0.8602 1.4152 2>
     "Float (Alpha)"=&FloatFromCurve(
      FloatCurve=DataInterface() {propertyValues {Curve {keys [
          <at 0.0048 -0.0024 RCIM_Cubic>
          {arriveTangent=2.9233, interpMode=RCIM_Cubic, leaveTangent=2.9233, tangentMode=RCTM_User, time=0.7236, value=0.9786}
          <at 0.9988 -0.0024>
         ]}}}
      Scale Curve=&FloatFromCurve(
       FloatCurve=DataInterface() {propertyValues {Curve {keys [<at 0.0654 -0.0428>,<at 0.5237 0.7921 RCIM_Cubic>,<at 0.9993 -0.0428>]}}}
       CurveIndex=$Emitter.NormalizedLoopAge
       Scale Curve=0.3
      )
     )
    )
    Create Camera Offset Attribute=true
    Create Color Attribute=true
    Create SpriteSize Attribute=true
    SpriteSize=&Vector2DFromFloat(Value=&Multiply_Float(
      A=200
      B=&FloatFromCurve(
       FloatCurve=DataInterface() {propertyValues {Curve {keys [
           <at 0.004 1.0518 RCIM_Cubic>
           {arriveTangent=-0.8729, interpMode=RCIM_Cubic, leaveTangent=-0.8729, tangentMode=RCTM_User, time=0.9881, value=0.7637}
          ]}}}
       Scale Curve=2
      )
     ))
    Write to Module Namespace=true
   ) {module=FN_Sprite_Attributes002}
   &FN_Sprite_Attributes(
    Camera Offset=10
    Color=&MakeLinearColorFromVectorAndFloat(
     "Vector (RGB)"=&Multiply_VectorByFloat(Vector=<1.5665 2.3369 6>, Float=0.5)
     "Float (Alpha)"=&FloatFromCurve(
      FloatCurve=DataInterface() {propertyValues {Curve {keys [
          <at 0.0048 -0.0024 RCIM_Cubic>
          {arriveTangent=2.9233, interpMode=RCIM_Cubic, leaveTangent=2.9233, tangentMode=RCTM_User, time=0.7236, value=0.9786}
          <at 0.9988 -0.0024>
         ]}}}
      Scale Curve=0.5
     )
    )
    Create Camera Offset Attribute=true
    Create Color Attribute=true
    Create Sprite Rotation Attribute=true
    Create SpriteSize Attribute=true
    SpriteSize=&Multiply_Vector2(
     A=<1000 20>
     B=&Vector2DFromFloat(Value=&RandomRangeFloat(Minimum=0.2, Maximum=1))
    )
    Write to Module Namespace=true
   ) {module=FN_Sprite_Attributes003}
   &FN_Sprite_Attributes(
    Camera Offset=60
    Color=&MakeLinearColorFromVectorAndFloat("Vector (RGB)"=<0.2398 1.0892 1.5>, "Float (Alpha)"=&ClampFloat(Float=0.5))
    Create Camera Offset Attribute=true
    Create Color Attribute=true
    Create Sprite Rotation Attribute=true
    Create SpriteSize Attribute=true
    SpriteSize=&Multiply_Vector2(
     A=&RandomRangeVector2D(Minimum=<120 80>, Maximum=<200 140>, Evaluation Type=::NewEnumerator1)
     B=&Vector2DFromFloat(Value=&FloatFromCurve(
       FloatCurve=DataInterface() {propertyValues {Curve {keys [
           <at 0 0 RCIM_Cubic>
           {arriveTangent=1.4286, interpMode=RCIM_Cubic, leaveTangent=1.4286, time=0.5754, value=0.3493}
           {arriveTangent=-0.8228, interpMode=RCIM_Cubic, leaveTangent=-0.8228, time=0.7, value=1}
           <at 1 0>
          ]}}}
       Scale Curve=5
      ))
    )
    Write to Module Namespace=true
   ) {module=FN_Sprite_Attributes004}
   &FN_Light_Attributes(
    Color=<rgba 1 2 10 1>
    Create Color Attribute=true
    Create Position Attribute=true
    Enable Visibility Tag=true
    Position=&Add_Vector(
     A=&Add_Vector(
      A=Linked() {linkedVariable {name=Particles.Position, type=NiagaraPosition}}
      B=&RandomVector(Vector Scale=20, Evaluation Type=::NewEnumerator1)
     )
     B=&TransformVector(Vector=<0 0 50>, Source Space=::World, Destination Space=::Local)
    )
    Scale Color=&FloatFromCurve(
     FloatCurve=DataInterface() {propertyValues {Curve {keys [<at -0.0008 -0.006 RCIM_Cubic>, {arriveTangent=0.006, interpMode=RCIM_Cubic, leaveTangent=0.006, time=0.4887, value=0.9748}, <at 1 0>]}}}
     Scale Curve=10
    )
    VisibilityTag=HlslExpression() {hlslExpression="Particles.UniqueID && 1"}
   )
   &FN_Gen5_Light(Color:LinearColor=<rgba 1 2 10 1>, Intensity=1000, Position Offset:Vector3f=<0 0 400>)
  ]
  rendererClasses [&NiagaraSpriteRendererProperties]
  renderers [
   {rendererClass=&NiagaraSpriteRendererProperties}
   {rendererClass=&NiagaraSpriteRendererProperties, rendererIndex=1}
   {rendererClass=&NiagaraSpriteRendererProperties, rendererIndex=2}
   {rendererClass=&NiagaraSpriteRendererProperties, rendererIndex=3}
   {rendererClass=&NiagaraSpriteRendererProperties, rendererIndex=4}
  ]
 }
}
summary {
 systemName=NS_AugContainer_Portal_Rift_Teleport
 emitters [
  {emitterName=ShockWave, rendererClasses [&NiagaraMeshRendererProperties]}
  {emitterName=Glass, simTarget=GPUComputeSim, rendererClasses [&NiagaraMeshRendererProperties]}
  {emitterName=Sparks, simTarget=GPUComputeSim, rendererClasses [&NiagaraSpriteRendererProperties]}
  {emitterName=ZeroPointFlare, rendererClasses [&NiagaraSpriteRendererProperties]}
  {emitterName=Light, rendererClasses [&NiagaraLightRendererProperties, &NiagaraComponentRendererProperties]}
  {emitterName=CrackSpline, rendererClasses [&NiagaraMeshRendererProperties]}
  {
   bEnabled=false
   emitterName=Ref
   rendererClasses [
    &NiagaraSpriteRendererProperties
    &NiagaraComponentRendererProperties
    &NiagaraMeshRendererProperties
   ]
  }
  {emitterName=Particle_Decale, rendererClasses [&NiagaraSpriteRendererProperties]}
  {emitterName=Smoke_Poof, rendererClasses [&NiagaraSpriteRendererProperties]}
 ]
 userVariables [
  User.bIsFrontEndPreview {type=NiagaraBool}
  User.LocalPostProcessEffect {type=NiagaraBool}
 ]
}
