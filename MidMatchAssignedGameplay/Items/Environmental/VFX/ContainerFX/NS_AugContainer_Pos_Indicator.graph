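# Niagara beam that marks an augment container's position, tinted by its current rarity.
# PickupBeam spawns one looping mesh particle whose X scale and material parameter grow with camera distance, keeping it readable from afar.
# Alpha fades near the camera, varies with time of day and follows a system timeline, while RarityChangeFloat_01 drives brightness flicker during rarity changes.
# A disabled second emitter holds a friendly/hostile captured-color variant.
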NiagaraSystem /MidMatchAssignedGameplay/Items/Environmental/VFX/ContainerFX/NS_AugContainer_Pos_Indicator
dependencies {
 usedDataInterfaces [&NiagaraDataInterfaceCamera, &NiagaraDataInterfaceCurve, &NiagaraDataInterfaceArrayColor]
 usedDynamicInputs []
 usedModules [
  &SystemState
  &self:SystemSpawnScript.SystemScriptSource.SystemScriptGraph.NiagaraNodeAssignment_0.SetVariables_#1
  &self:SystemSpawnScript.SystemScriptSource.SystemScriptGraph.NiagaraNodeAssignment_1.SetVariables_#2
  &self:SystemSpawnScript.SystemScriptSource.SystemScriptGraph.NiagaraNodeAssignment_3.SetVariables_#3
  &self:SystemSpawnScript.SystemScriptSource.SystemScriptGraph.NiagaraNodeAssignment_2.SetVariables_#4
  &Timeline
 ]
}
emitters {
 PickupBeam {
  emitterUpdateScript [
   &EmitterState()
   &SpawnBurst_Instantaneous(Age:NiagaraFloat=$Emitter.Age, Spawn Count=1, Use Spawn Probability=false)
   /SetVariables(Emitter.Cam Pos:NiagaraPosition=&ReturnCameraPosition(Coordinate Space:ENiagaraCoordinateSpace=::Simulation, Camera Query:NiagaraDataInterfaceCamera=DataInterface())) {isSetParameters=true, module=SetVariables_#5}
  ]
  particleSpawnScript [
   &InitializeParticle(
    Color=<rgba 1.5 0.5995 0.0781 0.35>
    Lifetime=4
    Mesh Scale=<2 1 1>
    Mesh Scale Mode=::NewEnumerator4
    UsePositionOffset=false
    Write Material Random=true
   ) {module=InitializeParticle001}
   /SetVariables(Particles.FriendlyColor:LinearColor=<rgba 0 0.3 2 1>, Particles.HostileColor:LinearColor=<rgba 1 0.0667 0 1>) {isSetParameters=true}
   /SetVariables(Particles.MeshOrientation=&MakeQuatFromAxisAngle(Axis=<0 -1 0>, AngleInDegrees=90)) {isSetParameters=true, module=SetVariables_#6}
   &CameraOffset(Camera Offset Amount=-20)
  ]
  particleUpdateScript [
   &ParticleState(
    Kill Particles When Lifetime Has Elapsed=false
    Let Infinitely Lived Particles Die When Emitter Deactivates=true
    Loop Particles Lifetime=true
   )
   &ScaleMeshSize(Scale Factor=&MakeVector(
     X=&FloatFromCurve(
      FloatCurve=DataInterface() {propertyValues {Curve {keys [<at 2000 1>,<at 30000 60>]}}}
      CurveIndex=&VectorLength(Vector=&Subtract_Position(A=$Particles.Position, B=$Emitter.Cam Pos))
     )
     Y=1
     Z=1
    ))
   &DynamicMaterialParameters(
    Index 0 Param 1=&Lerp_Float(
     A=150
     B=500
     Alpha=&FloatFromCurve(
      FloatCurve=DataInterface() {propertyValues {Curve {keys [
          <at 0.22 0 RCIM_Cubic>
          {arriveTangentWeight=0.3012, interpMode=RCIM_Cubic, leaveTangentWeight=0.1309, tangentMode=RCTM_User, time=0.75, value=1}
          {interpMode=RCIM_Cubic, leaveTangent=3.4262, time=1, value=1}
         ]}}}
      CurveIndex=&NormalizeDistanceRange(
       Start Position:NiagaraPosition=/Niagara/DynamicInputs/Helpers/SimulationPosition()
       End Position:NiagaraPosition=&VectorFromFloat(Value=&Multiply_Float(
         A=&VectorLength(Vector=&ReturnCameraVector(Camera Vectors=::NewEnumerator4, Coordinate Space=::Simulation, Origin Position=$Particles.Position))
         B=15000
        ))
       Distance:NiagaraFloat=15000
       Invert Normalized Range:NiagaraBool=false
      )
     )
    )
    Index 0 Param 2=1
    Index 0 Param 3=1
    Index 0 Param 4=1
    Module.Param0WriteEnabled=true
    Module.Param1WriteEnabled=false
    Module.Param2WriteEnabled=false
    Module.Param3WriteEnabled=false
   ) {enabled=false, module=DynamicMaterialParameters001}
   &DynamicMaterialParameters(
    Index 0 Param 1=&FloatFromCurve(
     FloatCurve=DataInterface() {propertyValues {Curve {keys [<at 2000 60>,<at 30000 900>]}}}
     CurveIndex=&VectorLength(Vector=&Subtract_Position(A=$Particles.Position, B=$Emitter.Cam Pos))
    )
    Index 0 Param 2=1
    Index 0 Param 3=1
    Index 0 Param 4=1
    Module.Param0WriteEnabled=true
    Module.Param1WriteEnabled=false
    Module.Param2WriteEnabled=false
    Module.Param3WriteEnabled=false
   )
   &Color(
    Color=&LerpLinearColors(
     StartColor=&SelectLinearColorFromArray(
      Array Sampling Mode=::NewEnumerator1
      Direct Array Index=$System.SysCurrentRarity
      Color Selection Array=DataInterface() {propertyValues {ColorData [
         <rgba 0.4271 0.3821 0.3417 1>
         <rgba 0.2063 0.7656 0.0997 1>
         <rgba 0.0656 0.3836 1.8 1>
         <rgba 0.6968 0.0229 2.2 1>
         <rgba 1 0.526 0.0521 1>
         <rgba 0.775 0.3185 0.004 1>
         <rgba 0.0661 0.6476 0.8 1>
         <rgba 0.0908 0.5841 0.006 1>
        ]}}
      Random Seed=?
     )
     EndColor=&SelectLinearColorFromArray(
      Array Sampling Mode=::NewEnumerator1
      Direct Array Index=$System.SysNextRarity
      Color Selection Array=DataInterface() {propertyValues {ColorData [
         <rgba 0.4271 0.3821 0.3417 1>
         <rgba 0.2063 0.7656 0.0997 1>
         <rgba 0.0656 0.3836 1.8 1>
         <rgba 0.6968 0.0229 2.2 1>
         <rgba 1 0.3997 0.0521 1>
         <rgba 0.775 0.3185 0.004 1>
         <rgba 0.0661 0.6476 0.8 1>
         <rgba 0.0908 0.5841 0.006 1>
        ]}}
      Random Seed=?
     )
     LerpFactor=0
    )
    Scale Alpha=&Multiply_Float(
     A=&FNTimeOfDayScalar(
      Night Scale=&FloatFromCurve(
       FloatCurve=DataInterface() {propertyValues {Curve {keys [<at 500 0.05>,<at 2000 0.35>]}}}
       CurveIndex=&VectorLength(Vector=&Subtract_Position(A=$Particles.Position, B=$Emitter.Cam Pos))
      )
      Morning Scale=&FloatFromCurve(
       FloatCurve=DataInterface() {propertyValues {Curve {keys [<at 500 0.05>,<at 2000 1>]}}}
       CurveIndex=&VectorLength(Vector=&Subtract_Position(A=$Particles.Position, B=$Emitter.Cam Pos))
      )
      Day Scale=&FloatFromCurve(
       FloatCurve=DataInterface() {propertyValues {Curve {keys [<at 500 0.05>,<at 2000 1>]}}}
       CurveIndex=&VectorLength(Vector=&Subtract_Position(A=$Particles.Position, B=$Emitter.Cam Pos))
      )
      Evening Scale=&FloatFromCurve(
       FloatCurve=DataInterface() {propertyValues {Curve {keys [<at 500 0.05>,<at 2000 1>]}}}
       CurveIndex=&VectorLength(Vector=&Subtract_Position(A=$Particles.Position, B=$Emitter.Cam Pos))
      )
     )
     B=&FloatFromCurve(FloatCurve=DataInterface() {propertyValues {Curve {keys [<at 0 1>,<at 1 0>]}}}, CurveIndex=$System.Timeline.NormalizedTime)
    )
    Scale Color=&VectorFromFloat(Value=&MakeCustomFloatFromBool(
      Bool=&Boolean_AndOperation(
       Bool1=&SetBoolByFloatComparison(A=$User.RarityChangeFloat_01, B=0.01)
       Bool2=&SetBoolByFloatComparison(A=$User.RarityChangeFloat_01, B=0.99, Comparison Type=::NewEnumerator6)
      )
      True Float=&Lerp_Float(
       A=0.5
       B=1
       Alpha=&Lerp_Float(
        A=0
        B=&OneMinusFloat(Float=/Niagara/DynamicInputs/Integer/V2/Round_Float(Float:NiagaraFloat=&ClampFloat(Float=&Sine(
            Normalized Angle=&Multiply_Float(A=$User.RarityChangeFloat_01, B=10)
            Period=&FloatFromCurve(
             FloatCurve=DataInterface() {propertyValues {Curve {keys [{tangentMode=RCTM_User, time=0, value=2}, <at 0.925 0.75>, {tangentMode=RCTM_User, time=0.925, value=0.05}, <at 0.995 0.05>, <at 1 0>]}}}
             CurveIndex=$User.RarityChangeFloat_01
            )
            Scale=0.5
            Bias=&FloatFromCurve(
             FloatCurve=DataInterface() {propertyValues {Curve {keys [
                 <at 0 0.9 RCIM_Cubic>
                 {arriveTangent=-0.432, interpMode=RCIM_Cubic, leaveTangent=-0.432, time=0.925, value=0.65}
                 <at 0.926 0.5>
                 <at 1 0.5>
                ]}}}
             CurveIndex=$User.RarityChangeFloat_01
            )
           ))))
        Alpha=&FloatFromCurve(
         FloatCurve=DataInterface() {propertyValues {Curve {keys [<at 0 0>,<at 0.001 1>,<at 0.999 1>,<at 1 0>]}}}
         CurveIndex=$User.RarityChangeFloat_01
        )
       )
      )
      False Float=1
     ))
   )
  ]
  rendererClasses [&NiagaraMeshRendererProperties]
  renderers [{rendererClass=&NiagaraMeshRendererProperties, rendererIndex=0}]
 }
 PickupBeam001 {
  bEnabled=false
  emitterUpdateScript [
   &EmitterState()
   &SpawnBurst_Instantaneous(Age:NiagaraFloat=$Emitter.Age, Spawn Count=1, Use Spawn Probability=false)
   /SetVariables(Emitter.Cam Pos:NiagaraPosition=&ReturnCameraPosition(Coordinate Space:ENiagaraCoordinateSpace=::Simulation, Camera Query:NiagaraDataInterfaceCamera=DataInterface())) {isSetParameters=true, module=SetVariables_#5}
  ]
  particleSpawnScript [
   &InitializeParticle(
    Color=<rgba 1.5 0.5995 0.0781 0.35>
    Lifetime=4
    Mesh Scale=<2 1 1>
    Mesh Scale Mode=::NewEnumerator4
    UsePositionOffset=false
    Write Material Random=true
   ) {module=InitializeParticle001}
   /SetVariables(Particles.MeshOrientation=&MakeQuatFromAxisAngle(Axis=<0 -1 0>, AngleInDegrees=90)) {isSetParameters=true, module=SetVariables_#6}
   &CameraOffset(Camera Offset Amount=-20)
   /SetVariables(Particles.FriendlyColor:LinearColor=<rgba 0 0.3 2 1>, Particles.HostileColor:LinearColor=<rgba 1 0.0667 0 1>) {isSetParameters=true}
  ]
  particleUpdateScript [
   &ParticleState(
    Kill Particles When Lifetime Has Elapsed=false
    Let Infinitely Lived Particles Die When Emitter Deactivates=true
    Loop Particles Lifetime=true
   )
   &ScaleMeshSize(Scale Factor=&MakeVector(
     X=&FloatFromCurve(
      FloatCurve=DataInterface() {propertyValues {Curve {keys [<at 2000 1>,<at 30000 60>]}}}
      CurveIndex=&VectorLength(Vector=&Subtract_Position(A=$Particles.Position, B=$Emitter.Cam Pos))
     )
     Y=1
     Z=1
    ))
   &DynamicMaterialParameters(
    Index 0 Param 1=&Lerp_Float(
     A=150
     B=500
     Alpha=&FloatFromCurve(
      FloatCurve=DataInterface() {propertyValues {Curve {keys [
          <at 0.22 0 RCIM_Cubic>
          {arriveTangentWeight=0.3012, interpMode=RCIM_Cubic, leaveTangentWeight=0.1309, tangentMode=RCTM_User, time=0.75, value=1}
          {interpMode=RCIM_Cubic, leaveTangent=3.4262, time=1, value=1}
         ]}}}
      CurveIndex=&NormalizeDistanceRange(
       Start Position:NiagaraPosition=&SimulationPosition()
       End Position:NiagaraPosition=&VectorFromFloat(Value=&Multiply_Float(
         A=&VectorLength(Vector=&ReturnCameraVector(Camera Vectors=::NewEnumerator4, Coordinate Space=::Simulation, Origin Position=$Particles.Position))
         B=15000
        ))
       Distance:NiagaraFloat=15000
       Invert Normalized Range:NiagaraBool=false
      )
     )
    )
    Index 0 Param 2=1
    Index 0 Param 3=1
    Index 0 Param 4=1
    Module.Param0WriteEnabled=true
    Module.Param1WriteEnabled=false
    Module.Param2WriteEnabled=false
    Module.Param3WriteEnabled=false
   ) {enabled=false, module=DynamicMaterialParameters001}
   &DynamicMaterialParameters(
    Index 0 Param 1=&FloatFromCurve(
     FloatCurve=DataInterface() {propertyValues {Curve {keys [<at 2000 60>,<at 30000 900>]}}}
     CurveIndex=&VectorLength(Vector=&Subtract_Position(A=$Particles.Position, B=$Emitter.Cam Pos))
    )
    Index 0 Param 2=1
    Index 0 Param 3=1
    Index 0 Param 4=1
    Module.Param0WriteEnabled=true
    Module.Param1WriteEnabled=false
    Module.Param2WriteEnabled=false
    Module.Param3WriteEnabled=false
   )
   &Color(
    Color=&LerpLinearColors(
     StartColor=$Particles.Initial.Color
     EndColor=&LerpLinearColors(StartColor=$Particles.FriendlyColor, EndColor=$Particles.HostileColor, LerpFactor=$User.FriendlyHostileFloat)
     LerpFactor=$User.CapturedFloat
    )
    Scale Alpha=&FloatFromCurve(
     FloatCurve=DataInterface() {propertyValues {Curve {keys [<at 500 0.05>,<at 2000 1>]}}}
     CurveIndex=&VectorLength(Vector=&Subtract_Position(A=$Particles.Position, B=$Emitter.Cam Pos))
    )
   )
  ]
  rendererClasses [&NiagaraMeshRendererProperties]
  renderers [{rendererClass=&NiagaraMeshRendererProperties, rendererIndex=0}]
 }
}
summary {
 systemName=NS_AugContainer_Pos_Indicator
 emitters [
  {emitterName=PickupBeam, rendererClasses [&NiagaraMeshRendererProperties]}
  {bEnabled=false, emitterName=PickupBeam001, rendererClasses [&NiagaraMeshRendererProperties]}
 ]
 userVariables [
  User.CapturedFloat {type=NiagaraFloat, defaultValue=0, description=""}
  User.CurrentRarityLevel {type=NiagaraInt32, defaultValue=4}
  User.FriendlyHostileFloat {type=NiagaraFloat, defaultValue=0, description=""}
  User.HideBeam {type=NiagaraBool}
  User.NextRarityLevel {type=NiagaraInt32, defaultValue=3}
  User.RarityChangeFloat_01 {type=NiagaraFloat, defaultValue=0, description=""}
 ]
}
