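# Niagara system for an Aug container's light cone and capture icon, tinted by item rarity.
# Sprite icon pops in on a timeline curve, passes User.ProgressAmount through MaterialRandom, and blends current-to-next rarity colors from indexed arrays.
# Mesh cone uses current rarity color, modulated by time of day, flash, flicker and fades.
# User.RarityChangeFloat_01 drives an accelerating rounded-sine cone blink; several variant emitters are disabled.
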NiagaraSystem /MidMatchAssignedGameplay/Items/Environmental/VFX/ContainerFX/NS_Aug_Container_Lightcone_01
dependencies {
 usedDataInterfaces [&NiagaraDataInterfaceArrayColor, &NiagaraDataInterfaceCurve, &NiagaraDataInterfaceCamera]
 usedDynamicInputs [&Boolean_NotOperation, &Divide_Float, &Subtract_Float]
 usedModules [
  &self:SystemSpawnScript.SystemScriptSource.SystemScriptGraph.NiagaraNodeAssignment_3.SetVariables_#1
  &self:SystemSpawnScript.SystemScriptSource.SystemScriptGraph.NiagaraNodeAssignment_5.SetVariables_#2
  &self:SystemSpawnScript.SystemScriptSource.SystemScriptGraph.NiagaraNodeAssignment_6.SetVariables_#3
  &self:SystemSpawnScript.SystemScriptSource.SystemScriptGraph.NiagaraNodeAssignment_7.SetVariables_#4
  &SystemState
  &self:SystemSpawnScript.SystemScriptSource.SystemScriptGraph.NiagaraNodeAssignment_0.SetVariables_#5
  &Timeline
  &self:SystemSpawnScript.SystemScriptSource.SystemScriptGraph.NiagaraNodeAssignment_2.SetVariables_#6
  &self:SystemSpawnScript.SystemScriptSource.SystemScriptGraph.NiagaraNodeAssignment_1.SetVariables_#7
  &self:SystemSpawnScript.SystemScriptSource.SystemScriptGraph.NiagaraNodeAssignment_4.SetVariables_#8
  &self:SystemSpawnScript.SystemScriptSource.SystemScriptGraph.NiagaraNodeAssignment_8.SetVariables_#9
  &self:SystemSpawnScript.SystemScriptSource.SystemScriptGraph.NiagaraNodeAssignment_9.SetVariables_#10
  &self:SystemSpawnScript.SystemScriptSource.SystemScriptGraph.NiagaraNodeAssignment_12.SetVariables_#11
  &self:SystemSpawnScript.SystemScriptSource.SystemScriptGraph.NiagaraNodeAssignment_13.SetVariables_#12
  &self:SystemSpawnScript.SystemScriptSource.SystemScriptGraph.NiagaraNodeAssignment_10.SetVariables_#13
 ]
}
emitters {
 Capture_Icon {
  emitterUpdateScript [
   &EmitterState(
    Enable Distance Culling=true
    MaxDistance=1000
    Module.bUseMaxDistance=false
    Module.bUseMinDistance=false
    Reset Age On Awaken=true
    Scalability Mode=::NewEnumerator1
   )
   &SpawnBurst_Instantaneous(Use Spawn Probability=false, UseLoopCountLimit=true)
   &SpawnBurst_Instantaneous(Spawn Time=1.9063, Use Spawn Probability=false, UseLoopCountLimit=false) {module=SpawnBurst_Instantaneous001}
  ]
  particleSpawnScript [
   &InitializeParticle(
    Color Mode=::NewEnumerator0
    Lifetime=5
    Position Offset=&Add_Vector(A=$System.Master Offset, B=<0 0 60>)
    Sprite Rotation Angle=-90
    Sprite Rotation Mode=::NewEnumerator0
    Sprite Size Mode=::NewEnumerator3
    Uniform Sprite Size=80
    UsePositionOffset=true
    Write Material Random=true
   )
   /SetVariables(Particles.FriendlyColor:LinearColor=<rgba 0 1.2 8 1>, Particles.HostileColor:LinearColor=<rgba 8 0.5333 0 1>) {isSetParameters=true, module=SetVariables_#14}
  ]
  particleUpdateScript [
   &ParticleState(Kill Particles When Lifetime Has Elapsed=false, Let Infinitely Lived Particles Die When Emitter Deactivates=true)
   &ScaleSpriteSize(
    Scale Sprite Size Mode=::NewEnumerator0
    Uniform Scale Factor=&FloatFromCurve(
     FloatCurve=DataInterface() {propertyValues {Curve {keys [
         {
          arriveTangentWeight=0.6259
          interpMode=RCIM_Cubic
          leaveTangentWeight=0.6259
          tangentMode=RCTM_User
          tangentWeightMode=RCTWM_WeightedBoth
          time=0
          value=0
         }
         {
          arriveTangent=1
          arriveTangentWeight=0.2828
          interpMode=RCIM_Cubic
          leaveTangent=1
          leaveTangentWeight=0.1886
          tangentMode=RCTM_Break
          tangentWeightMode=RCTWM_WeightedBoth
          time=0.7
          value=1.225
         }
         {
          arriveTangentWeight=0.1946
          interpMode=RCIM_Cubic
          leaveTangentWeight=0.1946
          tangentMode=RCTM_User
          tangentWeightMode=RCTWM_WeightedBoth
          time=1
          value=1
         }
        ]}}}
     CurveIndex=$System.Timeline004.NormalizedTime
    )
   )
   &Color(Color=&LerpLinearColors(
     StartColor=&SelectLinearColorFromArray(
      Array Sampling Mode=::NewEnumerator1
      Direct Array Index=$System.SysCurrentRarity
      Color Selection Array=DataInterface() {propertyValues {ColorData [
         <rgba 0.356 0.3185 0.2848 0.5>
         <rgba 0.0959 0.356 0.0464 0.5>
         <rgba 0.013 0.0759 0.356 0.5>
         <rgba 0.3167 0.0104 1 0.5>
         <rgba 0.356 0.1873 0.0185 0.5>
         <rgba 0.356 0.1317 0.0019 0.5>
         <rgba 0.0294 0.2882 0.356 0.5>
         <rgba 0.0554 0.356 0.0037 0.5>
        ]}}
      Random Seed=?
     )
     EndColor=&SelectLinearColorFromArray(
      Array Sampling Mode=::NewEnumerator1
      Direct Array Index=$System.SysNextRarity
      Color Selection Array=DataInterface() {propertyValues {ColorData [
         <rgba 0.356 0.3185 0.2848 0.5>
         <rgba 0.0959 0.356 0.0464 0.5>
         <rgba 0.013 0.0759 0.356 0.5>
         <rgba 0.3167 0.0104 1 0.5>
         <rgba 0.356 0.1873 0.0185 0.5>
         <rgba 0.356 0.1317 0.0019 0.5>
         <rgba 0.0294 0.2882 0.356 0.5>
         <rgba 0.0554 0.356 0.0037 0.5>
        ]}}
      Random Seed=?
     )
     LerpFactor=$System.LerpRarityColor
    ))
   /SetVariables(Particles.New Vector 4:Vector4f=&MakeVector4_FromLinearColor(LinearColor=&LerpLinearColors(StartColor=$Particles.FriendlyColor, EndColor=$Particles.HostileColor, LerpFactor=0))) {isSetParameters=true, module=SetVariables_#15}
   /SetVariables(Particles.New Vector 4:Vector4f=&MakeVector4_FromLinearColor(LinearColor=&LerpLinearColors(
      StartColor=&SelectLinearColorFromArray(
       Array Sampling Mode=::NewEnumerator1
       Direct Array Index=$System.SysCurrentRarity
       Color Selection Array=DataInterface() {propertyValues {ColorData [
          <rgba 3.25 2.9076 2.6 1>
          <rgba 0.8757 3.25 0.4232 1>
          <rgba 0.2917 1.7049 8 1>
          <rgba 3.1674 0.1042 10 1>
          <rgba 5 2.6302 0.2604 1>
          <rgba 5 1.8498 0.026 1>
          <rgba 0.4128 4.0475 5 1>
          <rgba 0.5055 3.25 0.0337 1>
         ]}}
       Random Seed=?
      )
      EndColor=&SelectLinearColorFromArray(
       Array Sampling Mode=::NewEnumerator1
       Direct Array Index=$System.SysNextRarity
       Color Selection Array=DataInterface() {propertyValues {ColorData [
          <rgba 3.25 2.9076 2.6 1>
          <rgba 0.8757 3.25 0.4232 1>
          <rgba 0.2917 1.7049 8 1>
          <rgba 3.1674 0.1042 10 1>
          <rgba 5 2.6302 0.2604 1>
          <rgba 5 1.8498 0.026 1>
          <rgba 0.4128 4.0475 5 1>
          <rgba 0.5055 3.25 0.0337 1>
         ]}}
       Random Seed=?
      )
      LerpFactor=$System.LerpRarityColor
     ))) {isSetParameters=true, module=SetVariables_#16}
   /SetVariables(Particles.MaterialRandom:NiagaraFloat=$User.ProgressAmount) {isSetParameters=true, module=SetVariables_#17}
   /SetVariables(
    Particles.Back Color:LinearColor=<rgba 0 0 0 1>
    Particles.Back Size 1:Vector2f=&Multiply_Vector2D_ByFloat(Float=0.95, Vector2D=$Particles.SpriteSize)
   ) {isSetParameters=true, module=SetVariables_#18}
   /SetVariables(
    Particles.Back COlor Dynam 2:Vector4f=<0 0 0 0.35>
    Particles.Back COlor Dynam 2001:Vector2f=&Multiply_Vector2D_ByFloat(Float=1.04, Vector2D=$Particles.SpriteSize)
    Particles.Back Color 2:LinearColor=<rgba 0 0 0 0.35>
   ) {isSetParameters=true, module=SetVariables_#19}
   &ScaleColor(
    Scale Alpha=&FloatFromCurve(
     FloatCurve=DataInterface() {propertyValues {Curve {keys [{tangentMode=RCTM_User, time=0, value=1}, {tangentMode=RCTM_User, time=1, value=0}]}}}
     CurveIndex=$System.Timeline005.NormalizedTime
    )
    ScaleA=true
    ScaleRGB=true
   ) {enabled=false, module=ScaleColor003}
  ]
  rendererClasses [&NiagaraSpriteRendererProperties]
  renderers [
   {rendererClass=&NiagaraSpriteRendererProperties}
   {rendererClass=&NiagaraSpriteRendererProperties, rendererIndex=1}
   {rendererClass=&NiagaraSpriteRendererProperties, rendererIndex=2}
  ]
 }
 Capture_Icon001 {
  bEnabled=false
  emitterUpdateScript [
   &EmitterState(
    Enable Distance Culling=true
    MaxDistance=1000
    Module.bUseMaxDistance=false
    Module.bUseMinDistance=false
    Reset Age On Awaken=true
    Scalability Mode=::NewEnumerator1
   )
   &SpawnBurst_Instantaneous(Use Spawn Probability=false, UseLoopCountLimit=true)
   &SpawnBurst_Instantaneous(Spawn Time=1.9063, Use Spawn Probability=false, UseLoopCountLimit=false) {module=SpawnBurst_Instantaneous001}
  ]
  particleSpawnScript [
   &InitializeParticle(
    Color Mode=::NewEnumerator0
    Lifetime=5
    Position Offset=&Add_Vector(A=$System.Master Offset, B=<0 0 45>)
    Sprite Rotation Angle=-90
    Sprite Rotation Mode=::NewEnumerator0
    Sprite Size Mode=::NewEnumerator3
    Uniform Sprite Size=65
    UsePositionOffset=true
    Write Material Random=true
   )
   /SetVariables(Particles.FriendlyColor:LinearColor=<rgba 0 1.2 8 1>, Particles.HostileColor:LinearColor=<rgba 8 0.5333 0 1>) {isSetParameters=true, module=SetVariables_#14}
  ]
  particleUpdateScript [
   &ParticleState(Kill Particles When Lifetime Has Elapsed=false, Let Infinitely Lived Particles Die When Emitter Deactivates=true)
   &ScaleSpriteSize(
    Scale Sprite Size Mode=::NewEnumerator0
    Uniform Scale Factor=&FloatFromCurve(
     FloatCurve=DataInterface() {propertyValues {Curve {keys [
         {
          arriveTangentWeight=0.6259
          interpMode=RCIM_Cubic
          leaveTangentWeight=0.6259
          tangentMode=RCTM_User
          tangentWeightMode=RCTWM_WeightedBoth
          time=0
          value=0
         }
         {
          arriveTangent=1
          arriveTangentWeight=0.2828
          interpMode=RCIM_Cubic
          leaveTangent=1
          leaveTangentWeight=0.1886
          tangentMode=RCTM_Break
          tangentWeightMode=RCTWM_WeightedBoth
          time=0.7
          value=1.225
         }
         {
          arriveTangentWeight=0.1946
          interpMode=RCIM_Cubic
          leaveTangentWeight=0.1946
          tangentMode=RCTM_User
          tangentWeightMode=RCTWM_WeightedBoth
          time=1
          value=1
         }
        ]}}}
     CurveIndex=$System.Timeline004.NormalizedTime
    )
   )
   &Color(Color=&LerpLinearColors(
     StartColor=<rgba 0.356 0.1869 0.089 0.5>
     EndColor=&LerpLinearColors(StartColor=<rgba 0.25 0.3625 1 1>, EndColor=<rgba 1 0.3 0.25 1>, LerpFactor=$User.FriendlyHostileFloat)
     LerpFactor=0
    ))
   /SetVariables(Particles.New Vector 4:Vector4f=&MakeVector4_FromLinearColor(LinearColor=&LerpLinearColors(StartColor=$Particles.FriendlyColor, EndColor=$Particles.HostileColor, LerpFactor=0))) {isSetParameters=true, module=SetVariables_#15}
   /SetVariables(Particles.New Vector 4:Vector4f=&MakeVector4_FromLinearColor(LinearColor=<rgba 3.25 1.3975 0.325 1>)) {isSetParameters=true, module=SetVariables_#16}
   /SetVariables(Particles.MaterialRandom:NiagaraFloat=$User.ProgressAmount) {isSetParameters=true, module=SetVariables_#17}
   /SetVariables(Particles.Back Color:LinearColor=<rgba 0 0 0 1>) {isSetParameters=true, module=SetVariables_#18}
   /SetVariables(
    Particles.Back COlor Dynam 2:Vector4f=<0 0 0 0.35>
    Particles.Back COlor Dynam 2001:Vector2f=&Multiply_Vector2D_ByFloat(Float=2, Vector2D=$Particles.SpriteSize)
    Particles.Back Color 2:LinearColor=<rgba 0 0 0 0.35>
   ) {isSetParameters=true, module=SetVariables_#19}
  ]
  rendererClasses [&NiagaraSpriteRendererProperties]
  renderers [
   {rendererClass=&NiagaraSpriteRendererProperties}
   {rendererClass=&NiagaraSpriteRendererProperties, rendererIndex=1}
   {rendererClass=&NiagaraSpriteRendererProperties, rendererIndex=2}
  ]
 }
 Capture_Icon002 {
  bEnabled=false
  emitterUpdateScript [
   &EmitterState(
    Enable Distance Culling=true
    MaxDistance=1000
    Module.bUseMaxDistance=false
    Module.bUseMinDistance=false
    Reset Age On Awaken=true
    Scalability Mode=::NewEnumerator1
   )
   &SpawnBurst_Instantaneous(Use Spawn Probability=false, UseLoopCountLimit=true)
   &SpawnBurst_Instantaneous(Spawn Time=1.9063, Use Spawn Probability=false, UseLoopCountLimit=false) {module=SpawnBurst_Instantaneous001}
  ]
  particleSpawnScript [
   &InitializeParticle(
    Color Mode=::NewEnumerator0
    Lifetime=5
    Position Offset=&Add_Vector(A=$System.Master Offset, B=<0 0 45>)
    Sprite Rotation Angle=-90
    Sprite Rotation Mode=::NewEnumerator0
    Sprite Size Mode=::NewEnumerator3
    Uniform Sprite Size=65
    UsePositionOffset=true
    Write Material Random=true
   )
   /SetVariables(Particles.FriendlyColor:LinearColor=<rgba 0 1.2 8 1>, Particles.HostileColor:LinearColor=<rgba 8 0.5333 0 1>) {isSetParameters=true, module=SetVariables_#14}
  ]
  particleUpdateScript [
   &ParticleState(Kill Particles When Lifetime Has Elapsed=false, Let Infinitely Lived Particles Die When Emitter Deactivates=true)
   &ScaleSpriteSize(
    Scale Sprite Size Mode=::NewEnumerator0
    Uniform Scale Factor=&FloatFromCurve(
     FloatCurve=DataInterface() {propertyValues {Curve {keys [
         {
          arriveTangentWeight=0.6259
          interpMode=RCIM_Cubic
          leaveTangentWeight=0.6259
          tangentMode=RCTM_User
          tangentWeightMode=RCTWM_WeightedBoth
          time=0
          value=0
         }
         {
          arriveTangent=1
          arriveTangentWeight=0.2828
          interpMode=RCIM_Cubic
          leaveTangent=1
          leaveTangentWeight=0.1886
          tangentMode=RCTM_Break
          tangentWeightMode=RCTWM_WeightedBoth
          time=0.7
          value=1.225
         }
         {
          arriveTangentWeight=0.1946
          interpMode=RCIM_Cubic
          leaveTangentWeight=0.1946
          tangentMode=RCTM_User
          tangentWeightMode=RCTWM_WeightedBoth
          time=1
          value=1
         }
        ]}}}
     CurveIndex=$System.Timeline004.NormalizedTime
    )
   )
   &Color(Color=&LerpLinearColors(
     StartColor=&SelectLinearColorFromArray(
      Array Sampling Mode=::NewEnumerator1
      Direct Array Index=$System.SysCurrentRarity
      Color Selection Array=DataInterface() {propertyValues {ColorData [
         <rgba 0.356 0.3185 0.2848 0.5>
         <rgba 0.0959 0.356 0.0464 0.5>
         <rgba 0.013 0.0759 0.356 0.5>
         <rgba 0.3167 0.0104 1 0.5>
         <rgba 0.356 0.1873 0.0185 0.5>
         <rgba 0.356 0.1317 0.0019 0.5>
         <rgba 0.0294 0.2882 0.356 0.5>
         <rgba 0.0554 0.356 0.0037 0.5>
        ]}}
      Random Seed=?
     )
     EndColor=&SelectLinearColorFromArray(
      Array Sampling Mode=::NewEnumerator1
      Direct Array Index=$System.SysNextRarity
      Color Selection Array=DataInterface() {propertyValues {ColorData [
         <rgba 0.356 0.3185 0.2848 0.5>
         <rgba 0.0959 0.356 0.0464 0.5>
         <rgba 0.013 0.0759 0.356 0.5>
         <rgba 0.3167 0.0104 1 0.5>
         <rgba 0.356 0.1873 0.0185 0.5>
         <rgba 0.356 0.1317 0.0019 0.5>
         <rgba 0.0294 0.2882 0.356 0.5>
         <rgba 0.0554 0.356 0.0037 0.5>
        ]}}
      Random Seed=?
     )
     LerpFactor=$System.LerpRarityColor
    ))
   /SetVariables(Particles.New Vector 4:Vector4f=&MakeVector4_FromLinearColor(LinearColor=&LerpLinearColors(StartColor=$Particles.FriendlyColor, EndColor=$Particles.HostileColor, LerpFactor=0))) {isSetParameters=true, module=SetVariables_#15}
   /SetVariables(Particles.New Vector 4:Vector4f=&MakeVector4_FromLinearColor(LinearColor=&LerpLinearColors(
      StartColor=&SelectLinearColorFromArray(
       Array Sampling Mode=::NewEnumerator1
       Direct Array Index=$System.SysCurrentRarity
       Color Selection Array=DataInterface() {propertyValues {ColorData [
          <rgba 3.25 2.9076 2.6 1>
          <rgba 0.8757 3.25 0.4232 1>
          <rgba 0.2917 1.7049 8 1>
          <rgba 3.1674 0.1042 10 1>
          <rgba 5 2.6302 0.2604 1>
          <rgba 5 1.8498 0.026 1>
          <rgba 0.4128 4.0475 5 1>
          <rgba 0.5055 3.25 0.0337 1>
         ]}}
       Random Seed=?
      )
      EndColor=&SelectLinearColorFromArray(
       Array Sampling Mode=::NewEnumerator1
       Direct Array Index=$System.SysNextRarity
       Color Selection Array=DataInterface() {propertyValues {ColorData [
          <rgba 3.25 2.9076 2.6 1>
          <rgba 0.8757 3.25 0.4232 1>
          <rgba 0.2917 1.7049 8 1>
          <rgba 3.1674 0.1042 10 1>
          <rgba 5 2.6302 0.2604 1>
          <rgba 5 1.8498 0.026 1>
          <rgba 0.4128 4.0475 5 1>
          <rgba 0.5055 3.25 0.0337 1>
         ]}}
       Random Seed=?
      )
      LerpFactor=$System.LerpRarityColor
     ))) {isSetParameters=true, module=SetVariables_#16}
   /SetVariables(Particles.MaterialRandom:NiagaraFloat=$User.ProgressAmount) {isSetParameters=true, module=SetVariables_#17}
   /SetVariables(Particles.Back Color:LinearColor=<rgba 0 0 0 1>) {isSetParameters=true, module=SetVariables_#18}
   /SetVariables(
    Particles.Back COlor Dynam 2:Vector4f=<0 0 0 0.35>
    Particles.Back COlor Dynam 2001:Vector2f=&Multiply_Vector2D_ByFloat(Float=2, Vector2D=$Particles.SpriteSize)
    Particles.Back Color 2:LinearColor=<rgba 0 0 0 0.35>
   ) {isSetParameters=true, module=SetVariables_#19}
   &ScaleColor(
    Scale Alpha=&FloatFromCurve(
     FloatCurve=DataInterface() {propertyValues {Curve {keys [{tangentMode=RCTM_User, time=0, value=1}, {tangentMode=RCTM_User, time=1, value=0}]}}}
     CurveIndex=$System.Timeline005.NormalizedTime
    )
    ScaleA=true
    ScaleRGB=true
   ) {enabled=false, module=ScaleColor003}
  ]
  rendererClasses [&NiagaraSpriteRendererProperties]
  renderers [
   {rendererClass=&NiagaraSpriteRendererProperties}
   {rendererClass=&NiagaraSpriteRendererProperties, rendererIndex=1}
   {rendererClass=&NiagaraSpriteRendererProperties, rendererIndex=2}
  ]
 }
 Cone_Ray {
  emitterUpdateScript [
   &EmitterState(
    Inactive Response=::NewEnumerator1
    Life Cycle Mode=::NewEnumerator1
    Loop Behavior=::NewEnumerator1
    Loop Duration Mode=::NewEnumerator1
    UseLoopDelay=false
   )
   &SpawnBurst_Instantaneous(Spawn Count=1, Use Spawn Probability=false)
  ]
  particleSpawnScript [
   &InitializeParticle(
    Color=<rgba 1 1 1 0.21>
    Mesh Scale=&Multiply_VectorByFloat(Vector=<0.22 0.32 0.32>, Float=0.9)
    Mesh Scale Mode=::NewEnumerator4
    Position Offset=&Add_Vector(A=$System.Master Offset, B=<0 0 -8>)
    UsePositionOffset=true
    Write Material Random=true
   )
   &InitialMeshOrientation(Orientation Vector=<0 0 1>, Rotation=<0 0.01 0>, Use Orientation Vector=true, Use Rotation Vector=false)
   /SetVariables(Particles.FriendlyColor:LinearColor=<rgba 0 1.2 8 0.21>, Particles.HostileColor:LinearColor=<rgba 8 0.5333 0 0.21>) {isSetParameters=true, module=SetVariables_#14}
  ]
  particleUpdateScript [
   &ParticleState(Kill Particles When Lifetime Has Elapsed=false, Let Infinitely Lived Particles Die When Emitter Deactivates=true)
   /SetVariables(Particles.Cone_Ray_Dark:LinearColor=&MakeLinearColorFromVectorAndFloat(
     "Vector (RGB)"=&MakeVectorFromLinearColorRGB(Color=&MultiplyLinearColor(
       Linear Color=&LerpLinearColors(
        StartColor=&SelectLinearColorFromArray(
         Array Sampling Mode=::NewEnumerator1
         Direct Array Index=$System.SysCurrentRarity
         Color Selection Array=DataInterface() {propertyValues {ColorData [
            <rgba 0.1 0.08 0.062 1>
            <rgba 0.0539 0.2 0.026 1>
            <rgba 0.0182 0.1066 0.5 1>
            <rgba 0.2376 0.0078 0.75 1>
            <rgba 0.7 0.2798 0.0365 1>
            <rgba 0.2 0.0822 0.001 1>
            <rgba 0.0289 0.2833 0.35 1>
            <rgba 0.0156 0.1 0.001 1>
           ]}}
         Random Seed=?
        )
        EndColor=&SelectLinearColorFromArray(
         Array Sampling Mode=::NewEnumerator1
         Direct Array Index=$System.SysNextRarity
         Color Selection Array=DataInterface() {propertyValues {ColorData [
            <rgba 0.1 0.08 0.062 1>
            <rgba 0.0539 0.2 0.026 1>
            <rgba 0.0182 0.1066 0.5 1>
            <rgba 0.2376 0.0078 0.75 1>
            <rgba 0.7 0.2798 0.0365 1>
            <rgba 0.2 0.0822 0.001 1>
            <rgba 0.0289 0.2833 0.35 1>
            <rgba 0.0156 0.1 0.001 1>
           ]}}
         Random Seed=?
        )
        LerpFactor=0
       )
       Scale Factor=&MakeVector4_FromFloat(Float=&FNTimeOfDayScalar(Night Scale=0.5, Morning Scale=0.3, Evening Scale=0.4))
      ))
     "Float (Alpha)"=&Multiply_Float(
      A=&Multiply_Float(
       A=&Multiply_Float(
        A=&Multiply_Float(A=&Multiply_Float(A=0.21, B=$System.LightConeFade), B=$System.TOD)
        B=$System.Flicker
       )
       B=&FloatFromCurve(FloatCurve=DataInterface() {propertyValues {Curve {keys [<at 0 1>,<at 1 0>]}}}, CurveIndex=$System.Timeline005.NormalizedTime)
      )
      B=&Lerp_Float(
       A=1
       B=&OneMinusFloat(Float=/Niagara/DynamicInputs/Integer/V2/Round_Float(Float:NiagaraFloat=&ClampFloat(Float=&Sine(
           Normalized Angle=&Multiply_Float(A=$User.RarityChangeFloat_01, B=10)
           Period=&FloatFromCurve(
            FloatCurve=DataInterface() {propertyValues {Curve {keys [{tangentMode=RCTM_User, time=0, value=2}, <at 0.925 0.75>, {tangentMode=RCTM_User, time=0.925, value=0.05}, <at 0.995 0.05>, <at 1 0>]}}}
            CurveIndex=$User.RarityChangeFloat_01
           )
           Scale=0.5
           Bias=&FloatFromCurve(
            FloatCurve=DataInterface() {propertyValues {Curve {keys [
                <at 0 0.9 RCIM_Cubic>
                {arriveTangent=-0.432, interpMode=RCIM_Cubic, leaveTangent=-0.432, time=0.925, value=0.65}
                <at 0.926 0.5>
                <at 1 0.5>
               ]}}}
            CurveIndex=$User.RarityChangeFloat_01
           )
          ))))
       Alpha=&FloatFromCurve(
        FloatCurve=DataInterface() {propertyValues {Curve {keys [<at 0 0>,<at 0.001 1>,<at 0.999 1>,<at 1 0>]}}}
        CurveIndex=$User.RarityChangeFloat_01
       )
      )
     )
    )) {isSetParameters=true, module=SetVariables_#20}
   /SetVariables(Particles.Cone_Ray_Brightness_Color:LinearColor=&MakeLinearColorFromVectorAndFloat(
     "Vector (RGB)"=&MakeVectorFromLinearColorRGB(Color=&MultiplyLinearColor(
       Linear Color=&LerpLinearColors(
        StartColor=&SelectLinearColorFromArray(
         Array Sampling Mode=::NewEnumerator1
         Direct Array Index=$System.SysCurrentRarity
         Color Selection Array=DataInterface() {propertyValues {ColorData [
            <rgba 1 0.7998 0.62 1>
            <rgba 0.2063 0.7656 0.0997 1>
            <rgba 0.0656 0.3836 1.8 1>
            <rgba 0.6968 0.0229 2.2 1>
            <rgba 1.5 0.5995 0.0781 1>
            <rgba 0.775 0.3185 0.004 1>
            <rgba 0.1238 1.2143 1.5 1>
            <rgba 0.0908 0.5841 0.006 1>
           ]}}
         Random Seed=?
        )
        EndColor=&SelectLinearColorFromArray(
         Array Sampling Mode=::NewEnumerator1
         Direct Array Index=$System.SysNextRarity
         Color Selection Array=DataInterface() {propertyValues {ColorData [
            <rgba 0.1 0.08 0.062 1>
            <rgba 0.0539 0.2 0.026 1>
            <rgba 0.0182 0.1066 0.5 1>
            <rgba 0.2376 0.0078 0.75 1>
            <rgba 0.7 0.2798 0.0365 1>
            <rgba 0.2 0.0822 0.001 1>
            <rgba 0.0289 0.2833 0.35 1>
            <rgba 0.0156 0.1 0.001 1>
           ]}}
         Random Seed=?
        )
        LerpFactor=0
       )
       Scale Factor=&MakeVector4_FromLinearColor(LinearColor=/Niagara/DynamicInputs/TypeConversions/MakeLinearColorFromFloat(Value:NiagaraFloat=&Lerp_Float(A=1, B=$System.Flash, Alpha=0.2)))
      ))
     "Float (Alpha)"=&Multiply_Float(
      A=&Multiply_Float(
       A=&Multiply_Float(
        A=&Multiply_Float(A=&Multiply_Float(A=0.168, B=$System.LightConeFade), B=$System.TOD)
        B=$System.Flicker
       )
       B=&FloatFromCurve(FloatCurve=DataInterface() {propertyValues {Curve {keys [<at 0 1>,<at 1 0>]}}}, CurveIndex=$System.Timeline005.NormalizedTime)
      )
      B=&Lerp_Float(
       A=1
       B=&OneMinusFloat(Float=&Round_Float(Float:NiagaraFloat=&ClampFloat(Float=&Sine(
           Normalized Angle=&Multiply_Float(A=$User.RarityChangeFloat_01, B=10)
           Period=&FloatFromCurve(
            FloatCurve=DataInterface() {propertyValues {Curve {keys [{tangentMode=RCTM_User, time=0, value=2}, <at 0.925 0.75>, {tangentMode=RCTM_User, time=0.925, value=0.05}, <at 0.995 0.05>, <at 1 0>]}}}
            CurveIndex=$User.RarityChangeFloat_01
           )
           Scale=0.5
           Bias=&FloatFromCurve(
            FloatCurve=DataInterface() {propertyValues {Curve {keys [
                <at 0 0.9 RCIM_Cubic>
                {arriveTangent=-0.432, interpMode=RCIM_Cubic, leaveTangent=-0.432, time=0.925, value=0.65}
                <at 0.926 0.5>
                <at 1 0.5>
               ]}}}
            CurveIndex=$User.RarityChangeFloat_01
           )
          ))))
       Alpha=&FloatFromCurve(
        FloatCurve=DataInterface() {propertyValues {Curve {keys [<at 0 0>,<at 0.001 1>,<at 0.999 1>,<at 1 0>]}}}
        CurveIndex=$User.RarityChangeFloat_01
       )
      )
     )
    )) {isSetParameters=true, module=SetVariables_#21}
   &ApplyOwnerScaleToAttributes(Scale Initial Mesh Scale=true)
   &SolveForcesAndVelocity(Clamp Velocity=false, Limit Acceleration=false)
   &ScaleMeshSize(Scale Factor=&VectorFromFloat(Value=&Multiply_Float(A=3, B=$System.Master Scale)))
   &ScaleColor(
    Scale Alpha=&Multiply_Float(A=&Multiply_Float(A=$System.LightConeFade, B=$System.TOD), B=$System.Flicker)
    ScaleA=true
    ScaleRGB=true
   ) {module=ScaleColor001}
   &ScaleColor(
    Color Value To Scale=$Particles.Color
    Scale RGB=&VectorFromFloat(Value=&Lerp_Float(A=1, B=$System.Flash, Alpha=1))
    ScaleA=false
    ScaleRGB=true
   ) {module=ScaleColor002}
   &ScaleColor(
    Scale Alpha=&FloatFromCurve(FloatCurve=DataInterface() {propertyValues {Curve {keys [<at 0 1>,<at 1 0>]}}}, CurveIndex=$System.Timeline005.NormalizedTime)
    Scale RGB=&MakeVectorFromLinearColorRGB(Color=&LerpLinearColors(
      StartColor=&SelectLinearColorFromArray(
       Array Sampling Mode=::NewEnumerator1
       Direct Array Index=$System.SysCurrentRarity
       Color Selection Array=DataInterface() {propertyValues {ColorData [
          <rgba 0.4271 0.3821 0.3417 1>
          <rgba 0.2063 0.7656 0.0997 1>
          <rgba 0.0656 0.3836 1.8 1>
          <rgba 0.6968 0.0229 2.2 1>
          <rgba 1.5 0.7891 0.0781 1>
          <rgba 0.775 0.3185 0.004 1>
          <rgba 0.1238 1.2143 1.5 1>
          <rgba 0.0908 0.5841 0.006 1>
         ]}}
       Random Seed=?
      )
      EndColor=&SelectLinearColorFromArray(
       Array Sampling Mode=::NewEnumerator1
       Direct Array Index=$System.SysNextRarity
       Color Selection Array=DataInterface() {propertyValues {ColorData [
          <rgba 0.4271 0.3821 0.3417 1>
          <rgba 0.2063 0.7656 0.0997 1>
          <rgba 0.0656 0.3836 1.8 1>
          <rgba 0.6968 0.0229 2.2 1>
          <rgba 1.5 0.5995 0.0781 1>
          <rgba 0.775 0.3185 0.004 1>
          <rgba 0.1238 1.2143 1.5 1>
          <rgba 0.0908 0.5841 0.006 1>
         ]}}
       Random Seed=?
      )
      LerpFactor=0
     ))
    ScaleA=true
    ScaleRGB=true
   )
   &Color(Color=&LerpLinearColors(
     StartColor=$Particles.Initial.Color
     EndColor=&LerpLinearColors(StartColor=<rgba 1 1 1 1>, EndColor=<rgba 0 0 0 1>, LerpFactor=0)
     LerpFactor=0
    )) {enabled=false}
   &ScaleColor(
    Scale RGB=&Lerp_Vector(
     A=&VectorFromFloat(Value=1)
     B=&VectorFromFloat(Value=&Lerp_Float(
       A=0.5
       B=1
       Alpha=&Lerp_Float(
        A=0
        B=&OneMinusFloat(Float=&Round_Float(Float:NiagaraFloat=&ClampFloat(Float=&Sine(
            Normalized Angle=&Multiply_Float(A=$User.RarityChangeFloat_01, B=10)
            Period=&FloatFromCurve(
             FloatCurve=DataInterface() {propertyValues {Curve {keys [{tangentMode=RCTM_User, time=0, value=2}, <at 0.925 0.75>, {tangentMode=RCTM_User, time=0.925, value=0.05}, <at 0.995 0.05>, <at 1 0>]}}}
             CurveIndex=$User.RarityChangeFloat_01
            )
            Scale=0.5
            Bias=&FloatFromCurve(
             FloatCurve=DataInterface() {propertyValues {Curve {keys [
                 <at 0 0.9 RCIM_Cubic>
                 {arriveTangent=-0.432, interpMode=RCIM_Cubic, leaveTangent=-0.432, time=0.925, value=0.65}
                 <at 0.926 0.5>
                 <at 1 0.5>
                ]}}}
             CurveIndex=$User.RarityChangeFloat_01
            )
           ))))
        Alpha=&FloatFromCurve(
         FloatCurve=DataInterface() {propertyValues {Curve {keys [<at 0 0>,<at 0.001 1>,<at 0.999 1>,<at 1 0>]}}}
         CurveIndex=$User.RarityChangeFloat_01
        )
       )
      ))
     Alpha=&MakeCustomFloatFromBool(
      Bool=&Boolean_AndOperation(
       Bool1=&SetBoolByFloatComparison(A=$User.RarityChangeFloat_01, B=0.01)
       Bool2=&SetBoolByFloatComparison(A=$User.RarityChangeFloat_01, B=0.99, Comparison Type=::NewEnumerator6)
      )
      True Float=1
      False Float=0
     )
    )
    ScaleA=true
    ScaleRGB=true
   ) {module=ScaleColor003}
  ]
  rendererClasses [&NiagaraMeshRendererProperties]
  renderers [
   {rendererClass=&NiagaraMeshRendererProperties, rendererIndex=0}
   {rendererClass=&NiagaraMeshRendererProperties, rendererIndex=1}
   {rendererClass=&NiagaraMeshRendererProperties, rendererIndex=2}
  ]
 }
 Cone_Ray001 {
  bEnabled=false
  emitterUpdateScript [
   &EmitterState(
    Inactive Response=::NewEnumerator1
    Life Cycle Mode=::NewEnumerator1
    Loop Behavior=::NewEnumerator1
    Loop Duration Mode=::NewEnumerator1
    UseLoopDelay=false
   )
   &SpawnBurst_Instantaneous(Spawn Count=1, Use Spawn Probability=false)
  ]
  particleSpawnScript [
   &InitializeParticle(
    Color=<rgba 1 1 1 0.21>
    Mesh Scale=&Multiply_VectorByFloat(Vector=<0.22 0.32 0.32>, Float=0.9)
    Mesh Scale Mode=::NewEnumerator4
    Position Offset=&Add_Vector(A=$System.Master Offset, B=<0 0 -8>)
    UsePositionOffset=true
    Write Material Random=true
   )
   &InitialMeshOrientation(Orientation Vector=<0 0 1>, Rotation=<0 0.01 0>, Use Orientation Vector=true, Use Rotation Vector=false)
   /SetVariables(Particles.FriendlyColor:LinearColor=<rgba 0 1.2 8 0.21>, Particles.HostileColor:LinearColor=<rgba 8 0.5333 0 0.21>) {isSetParameters=true, module=SetVariables_#14}
  ]
  particleUpdateScript [
   &ParticleState(Kill Particles When Lifetime Has Elapsed=false, Let Infinitely Lived Particles Die When Emitter Deactivates=true)
   /SetVariables(Particles.Cone_Ray_Dark:LinearColor=&MakeLinearColorFromVectorAndFloat(
     "Vector (RGB)"=&MakeVectorFromLinearColorRGB(Color=&MultiplyLinearColor(
       Linear Color=&LerpLinearColors(
        StartColor=<rgba 0.2 0.0799 0.0104 1>
        EndColor=&LerpLinearColors(StartColor=<rgba 1 1 1 1>, EndColor=<rgba 0 0 0 1>, LerpFactor=0)
        LerpFactor=0
       )
       Scale Factor=&MakeVector4_FromFloat(Float=&FNTimeOfDayScalar(Night Scale=0.5, Morning Scale=0.3, Evening Scale=0.4))
      ))
     "Float (Alpha)"=&Multiply_Float(
      A=&Multiply_Float(A=&Multiply_Float(A=0.21, B=$System.LightConeFade), B=$System.TOD)
      B=$System.Flicker
     )
    )) {isSetParameters=true, module=SetVariables_#20}
   /SetVariables(Particles.Cone_Ray_Brightness_Color:LinearColor=&MakeLinearColorFromVectorAndFloat(
     "Vector (RGB)"=&MakeVectorFromLinearColorRGB(Color=&MultiplyLinearColor(
       Linear Color=&LerpLinearColors(
        StartColor=<rgba 1 0.3997 0.0521 1>
        EndColor=&LerpLinearColors(StartColor=<rgba 1 1 1 1>, EndColor=<rgba 0 0 0 1>, LerpFactor=0)
        LerpFactor=0
       )
       Scale Factor=&MakeVector4_FromLinearColor(LinearColor=&MakeLinearColorFromFloat(Value:NiagaraFloat=&Lerp_Float(A=1, B=$System.Flash, Alpha=0.2)))
      ))
     "Float (Alpha)"=&Multiply_Float(
      A=&Multiply_Float(A=&Multiply_Float(A=0.168, B=$System.LightConeFade), B=$System.TOD)
      B=$System.Flicker
     )
    )) {isSetParameters=true, module=SetVariables_#21}
   &ApplyOwnerScaleToAttributes(Scale Initial Mesh Scale=true)
   &SolveForcesAndVelocity(Clamp Velocity=false, Limit Acceleration=false)
   &ScaleMeshSize(Scale Factor=&VectorFromFloat(Value=&Multiply_Float(A=3, B=$System.Master Scale)))
   &ScaleColor(
    Scale Alpha=&Multiply_Float(A=&Multiply_Float(A=$System.LightConeFade, B=$System.TOD), B=$System.Flicker)
    ScaleA=true
    ScaleRGB=true
   ) {module=ScaleColor001}
   &ScaleColor(
    Color Value To Scale=$Particles.Color
    Scale RGB=&VectorFromFloat(Value=&Lerp_Float(A=1, B=$System.Flash, Alpha=1))
    ScaleA=false
    ScaleRGB=true
   ) {module=ScaleColor002}
   &ScaleColor(
    Scale RGB=&MakeVectorFromLinearColorRGB(Color=&LerpLinearColors(
      StartColor=<rgba 1.5 0.5995 0.0781 0.21>
      EndColor=&LerpLinearColors(StartColor=$Particles.FriendlyColor, EndColor=$Particles.HostileColor, LerpFactor=$User.FriendlyHostileFloat)
      LerpFactor=0
     ))
    ScaleA=true
    ScaleRGB=true
   )
   &Color(Color=&LerpLinearColors(
     StartColor=$Particles.Initial.Color
     EndColor=&LerpLinearColors(StartColor=<rgba 1 1 1 1>, EndColor=<rgba 0 0 0 1>, LerpFactor=0)
     LerpFactor=0
    )) {enabled=false}
  ]
  rendererClasses [&NiagaraMeshRendererProperties]
  renderers [
   {rendererClass=&NiagaraMeshRendererProperties, rendererIndex=0}
   {rendererClass=&NiagaraMeshRendererProperties, rendererIndex=1}
   {rendererClass=&NiagaraMeshRendererProperties, rendererIndex=2}
  ]
 }
 Cone_Ray002 {
  bEnabled=false
  emitterUpdateScript [
   &EmitterState(
    Inactive Response=::NewEnumerator1
    Life Cycle Mode=::NewEnumerator1
    Loop Behavior=::NewEnumerator1
    Loop Duration Mode=::NewEnumerator1
    UseLoopDelay=false
   )
   &SpawnBurst_Instantaneous(Spawn Count=1, Use Spawn Probability=false)
  ]
  particleSpawnScript [
   &InitializeParticle(
    Color=<rgba 1 1 1 0.21>
    Mesh Scale=&Multiply_VectorByFloat(Vector=<0.22 0.32 0.32>, Float=0.9)
    Mesh Scale Mode=::NewEnumerator4
    Position Offset=&Add_Vector(A=$System.Master Offset, B=<0 0 -8>)
    UsePositionOffset=true
    Write Material Random=true
   )
   &InitialMeshOrientation(Orientation Vector=<0 0 1>, Rotation=<0 0.01 0>, Use Orientation Vector=true, Use Rotation Vector=false)
   /SetVariables(Particles.FriendlyColor:LinearColor=<rgba 0 1.2 8 0.21>, Particles.HostileColor:LinearColor=<rgba 8 0.5333 0 0.21>) {isSetParameters=true, module=SetVariables_#14}
  ]
  particleUpdateScript [
   &ParticleState(Kill Particles When Lifetime Has Elapsed=false, Let Infinitely Lived Particles Die When Emitter Deactivates=true)
   /SetVariables(Particles.Cone_Ray_Dark:LinearColor=&MakeLinearColorFromVectorAndFloat(
     "Vector (RGB)"=&MakeVectorFromLinearColorRGB(Color=&MultiplyLinearColor(
       Linear Color=&LerpLinearColors(
        StartColor=&SelectLinearColorFromArray(
         Array Sampling Mode=::NewEnumerator1
         Direct Array Index=$System.SysCurrentRarity
         Color Selection Array=DataInterface() {propertyValues {ColorData [
            <rgba 0.1 0.08 0.062 1>
            <rgba 0.0539 0.2 0.026 1>
            <rgba 0.0182 0.1066 0.5 1>
            <rgba 0.2376 0.0078 0.75 1>
            <rgba 0.7 0.2798 0.0365 1>
            <rgba 0.2 0.0822 0.001 1>
            <rgba 0.0289 0.2833 0.35 1>
            <rgba 0.0156 0.1 0.001 1>
           ]}}
         Random Seed=?
        )
        EndColor=&SelectLinearColorFromArray(
         Array Sampling Mode=::NewEnumerator1
         Direct Array Index=$System.SysNextRarity
         Color Selection Array=DataInterface() {propertyValues {ColorData [
            <rgba 0.1 0.08 0.062 1>
            <rgba 0.0539 0.2 0.026 1>
            <rgba 0.0182 0.1066 0.5 1>
            <rgba 0.2376 0.0078 0.75 1>
            <rgba 0.7 0.2798 0.0365 1>
            <rgba 0.2 0.0822 0.001 1>
            <rgba 0.0289 0.2833 0.35 1>
            <rgba 0.0156 0.1 0.001 1>
           ]}}
         Random Seed=?
        )
        LerpFactor=$System.LerpRarityColor
       )
       Scale Factor=&MakeVector4_FromFloat(Float=&FNTimeOfDayScalar(Night Scale=0.5, Morning Scale=0.3, Evening Scale=0.4))
      ))
     "Float (Alpha)"=&Multiply_Float(
      A=&Multiply_Float(
       A=&Multiply_Float(A=&Multiply_Float(A=0.21, B=$System.LightConeFade), B=$System.TOD)
       B=$System.Flicker
      )
      B=&FloatFromCurve(FloatCurve=DataInterface() {propertyValues {Curve {keys [<at 0 1>,<at 1 0>]}}}, CurveIndex=$System.Timeline005.NormalizedTime)
     )
    )) {isSetParameters=true, module=SetVariables_#20}
   /SetVariables(Particles.Cone_Ray_Brightness_Color:LinearColor=&MakeLinearColorFromVectorAndFloat(
     "Vector (RGB)"=&MakeVectorFromLinearColorRGB(Color=&MultiplyLinearColor(
       Linear Color=&LerpLinearColors(
        StartColor=&SelectLinearColorFromArray(
         Array Sampling Mode=::NewEnumerator1
         Direct Array Index=$System.SysCurrentRarity
         Color Selection Array=DataInterface() {propertyValues {ColorData [
            <rgba 1 0.7998 0.62 1>
            <rgba 0.2063 0.7656 0.0997 1>
            <rgba 0.0656 0.3836 1.8 1>
            <rgba 0.6968 0.0229 2.2 1>
            <rgba 1.5 0.5995 0.0781 1>
            <rgba 0.775 0.3185 0.004 1>
            <rgba 0.1238 1.2143 1.5 1>
            <rgba 0.0908 0.5841 0.006 1>
           ]}}
         Random Seed=?
        )
        EndColor=&SelectLinearColorFromArray(
         Array Sampling Mode=::NewEnumerator1
         Direct Array Index=$System.SysNextRarity
         Color Selection Array=DataInterface() {propertyValues {ColorData [
            <rgba 0.1 0.08 0.062 1>
            <rgba 0.0539 0.2 0.026 1>
            <rgba 0.0182 0.1066 0.5 1>
            <rgba 0.2376 0.0078 0.75 1>
            <rgba 0.7 0.2798 0.0365 1>
            <rgba 0.2 0.0822 0.001 1>
            <rgba 0.0289 0.2833 0.35 1>
            <rgba 0.0156 0.1 0.001 1>
           ]}}
         Random Seed=?
        )
        LerpFactor=$System.LerpRarityColor
       )
       Scale Factor=&MakeVector4_FromLinearColor(LinearColor=&MakeLinearColorFromFloat(Value:NiagaraFloat=&Lerp_Float(A=1, B=$System.Flash, Alpha=0.2)))
      ))
     "Float (Alpha)"=&Multiply_Float(
      A=&Multiply_Float(
       A=&Multiply_Float(A=&Multiply_Float(A=0.168, B=$System.LightConeFade), B=$System.TOD)
       B=$System.Flicker
      )
      B=&FloatFromCurve(FloatCurve=DataInterface() {propertyValues {Curve {keys [<at 0 1>,<at 1 0>]}}}, CurveIndex=$System.Timeline005.NormalizedTime)
     )
    )) {isSetParameters=true, module=SetVariables_#21}
   &ApplyOwnerScaleToAttributes(Scale Initial Mesh Scale=true)
   &SolveForcesAndVelocity(Clamp Velocity=false, Limit Acceleration=false)
   &ScaleMeshSize(Scale Factor=&VectorFromFloat(Value=&Multiply_Float(A=3, B=$System.Master Scale)))
   &ScaleColor(
    Scale Alpha=&Multiply_Float(A=&Multiply_Float(A=$System.LightConeFade, B=$System.TOD), B=$System.Flicker)
    ScaleA=true
    ScaleRGB=true
   ) {module=ScaleColor001}
   &ScaleColor(
    Color Value To Scale=$Particles.Color
    Scale RGB=&VectorFromFloat(Value=&Lerp_Float(A=1, B=$System.Flash, Alpha=1))
    ScaleA=false
    ScaleRGB=true
   ) {module=ScaleColor002}
   &ScaleColor(
    Scale Alpha=&FloatFromCurve(FloatCurve=DataInterface() {propertyValues {Curve {keys [<at 0 1>,<at 1 0>]}}}, CurveIndex=$System.Timeline005.NormalizedTime)
    Scale RGB=&MakeVectorFromLinearColorRGB(Color=&LerpLinearColors(
      StartColor=&SelectLinearColorFromArray(
       Array Sampling Mode=::NewEnumerator1
       Direct Array Index=$System.SysCurrentRarity
       Color Selection Array=DataInterface() {propertyValues {ColorData [
          <rgba 0.4271 0.3821 0.3417 1>
          <rgba 0.2063 0.7656 0.0997 1>
          <rgba 0.0656 0.3836 1.8 1>
          <rgba 0.6968 0.0229 2.2 1>
          <rgba 1.5 0.7891 0.0781 1>
          <rgba 0.775 0.3185 0.004 1>
          <rgba 0.1238 1.2143 1.5 1>
          <rgba 0.0908 0.5841 0.006 1>
         ]}}
       Random Seed=?
      )
      EndColor=&SelectLinearColorFromArray(
       Array Sampling Mode=::NewEnumerator1
       Direct Array Index=$System.SysNextRarity
       Color Selection Array=DataInterface() {propertyValues {ColorData [
          <rgba 0.4271 0.3821 0.3417 1>
          <rgba 0.2063 0.7656 0.0997 1>
          <rgba 0.0656 0.3836 1.8 1>
          <rgba 0.6968 0.0229 2.2 1>
          <rgba 1.5 0.5995 0.0781 1>
          <rgba 0.775 0.3185 0.004 1>
          <rgba 0.1238 1.2143 1.5 1>
          <rgba 0.0908 0.5841 0.006 1>
         ]}}
       Random Seed=?
      )
      LerpFactor=$System.LerpRarityColor
     ))
    ScaleA=true
    ScaleRGB=true
   )
   &Color(Color=&LerpLinearColors(
     StartColor=$Particles.Initial.Color
     EndColor=&LerpLinearColors(StartColor=<rgba 1 1 1 1>, EndColor=<rgba 0 0 0 1>, LerpFactor=0)
     LerpFactor=0
    )) {enabled=false}
   &ScaleColor(
    Scale RGB=&Lerp_Vector(
     A=&VectorFromFloat(Value=1)
     B=&VectorFromFloat(Value=&Lerp_Float(
       A=6
       B=1
       Alpha=&Sine(Normalized Angle=&Multiply_Float(A=$User.RarityChangeFloat_01, B=10), Period=0.75, Scale=0.5, Bias=0.5)
      ))
     Alpha=&MakeCustomFloatFromBool(
      Bool=&Boolean_AndOperation(
       Bool1=&SetBoolByFloatComparison(A=$User.RarityChangeFloat_01, B=0.01)
       Bool2=&SetBoolByFloatComparison(A=$User.RarityChangeFloat_01, B=0.99, Comparison Type=::NewEnumerator6)
      )
      True Float=1
      False Float=0
     )
    )
    ScaleA=true
    ScaleRGB=true
   ) {module=ScaleColor003}
  ]
  rendererClasses [&NiagaraMeshRendererProperties]
  renderers [
   {rendererClass=&NiagaraMeshRendererProperties, rendererIndex=0}
   {rendererClass=&NiagaraMeshRendererProperties, rendererIndex=1}
   {rendererClass=&NiagaraMeshRendererProperties, rendererIndex=2}
  ]
 }
 Cone_pulse {
  emitterUpdateScript [
   &EmitterState(
    Inactive Response=::NewEnumerator1
    Life Cycle Mode=::NewEnumerator1
    Loop Behavior=::NewEnumerator1
    Loop Duration Mode=::NewEnumerator1
    UseLoopDelay=false
   )
   &SpawnBurst_Instantaneous(Spawn Count=1, Use Spawn Probability=false)
  ]
  particleSpawnScript [
   &InitializeParticle(
    Mesh Scale=&Multiply_VectorByFloat(Vector=<0.22 0.32 0.32>, Float=0.8)
    Mesh Scale Mode=::NewEnumerator4
    Position Offset=&Add_Vector(A=$System.Master Offset, B=<0 0 -5>)
    UsePositionOffset=true
    Write Material Random=true
   )
   &InitialMeshOrientation(Orientation Vector=<0 0 1>, Rotation=<0 0.01 0>, Use Orientation Vector=true, Use Rotation Vector=false)
   /SetVariables(Particles.FriendlyColor:LinearColor=<rgba 0 7.5 50 1>, Particles.HostileColor:LinearColor=<rgba 50 3.3334 0 1>) {isSetParameters=true, module=SetVariables_#14}
  ]
  particleUpdateScript [
   &ParticleState(Kill Particles When Lifetime Has Elapsed=false, Let Infinitely Lived Particles Die When Emitter Deactivates=true)
   &ApplyOwnerScaleToAttributes(Scale Initial Mesh Scale=true)
   &SolveForcesAndVelocity(Clamp Velocity=false, Limit Acceleration=false)
   &ScaleColor(
    Scale Alpha=&Multiply_Float(A=0.015, B=$System.TOD)
    Scale RGB=&Multiply_VectorByFloat(Vector=&Multiply_VectorByFloat(Vector=<1 1 1>, Float=1), Float=$System.TOD)
    ScaleA=true
    ScaleRGB=false
   )
   &ScaleColor(
    Scale RGB=&MakeVectorFromLinearColorRGB(Color=&LerpLinearColors(
      StartColor=<rgba 50 19.9826 2.6042 1>
      EndColor=&LerpLinearColors(StartColor=$Particles.FriendlyColor, EndColor=$Particles.HostileColor, LerpFactor=$User.FriendlyHostileFloat)
      LerpFactor=0
     ))
    ScaleA=true
    ScaleRGB=true
   ) {enabled=false, module=ScaleColor002}
   &ScaleColor(
    Scale RGB=&Multiply_VectorByFloat(
     Vector=&MakeVectorFromLinearColorRGB(Color=&LerpLinearColors(
       StartColor=&SelectLinearColorFromArray(
        Array Sampling Mode=::NewEnumerator1
        Direct Array Index=$System.SysCurrentRarity
        Color Selection Array=DataInterface() {propertyValues {ColorData [
           <rgba 0.4271 0.3821 0.3417 1>
           <rgba 0.2063 0.7656 0.0997 1>
           <rgba 0.0656 0.3836 1.8 1>
           <rgba 0.6968 0.0229 2.2 1>
           <rgba 1.5 0.7891 0.0781 1>
           <rgba 0.775 0.3185 0.004 1>
           <rgba 0.1238 1.2143 1.5 1>
           <rgba 0.0908 0.5841 0.006 1>
          ]}}
        Random Seed=?
       )
       EndColor=&SelectLinearColorFromArray(
        Array Sampling Mode=::NewEnumerator1
        Direct Array Index=$System.SysNextRarity
        Color Selection Array=DataInterface() {propertyValues {ColorData [
           <rgba 0.4271 0.3821 0.3417 1>
           <rgba 0.2063 0.7656 0.0997 1>
           <rgba 0.0656 0.3836 1.8 1>
           <rgba 0.6968 0.0229 2.2 1>
           <rgba 1.5 0.5995 0.0781 1>
           <rgba 0.775 0.3185 0.004 1>
           <rgba 0.1238 1.2143 1.5 1>
           <rgba 0.0908 0.5841 0.006 1>
          ]}}
        Random Seed=?
       )
       LerpFactor=0
      ))
     Float=50
    )
    ScaleA=true
    ScaleRGB=true
   ) {module=ScaleColor003}
   &ScaleMeshSize(Scale Factor=&VectorFromFloat(Value=&Multiply_Float(A=3, B=$System.Master Scale)))
   &DynamicMaterialParameters(
    Index 0 Param 1=$System.Timeline.NormalizedTime
    Module.Param0WriteEnabled=true
    Module.Param1WriteEnabled=true
    Module.Param2WriteEnabled=true
    Module.Param3WriteEnabled=true
   )
   &ScaleColor(Scale Alpha=&Multiply_Float(A=$System.LightConeFade, B=$System.Flicker), ScaleA=true, ScaleRGB=false) {module=ScaleColor001}
   &ScaleColor(
    Scale Alpha=&FloatFromCurve(
     FloatCurve=DataInterface() {propertyValues {Curve {keys [{tangentMode=RCTM_User, time=0, value=1}, {tangentMode=RCTM_User, time=1, value=0}]}}}
     CurveIndex=$System.Timeline005.NormalizedTime
    )
    ScaleA=true
    ScaleRGB=true
   ) {module=ScaleColor004}
   &ScaleColor(
    Scale RGB=&Lerp_Vector(
     A=&VectorFromFloat(Value=1)
     B=&VectorFromFloat(Value=&Lerp_Float(
       A=0.5
       B=1
       Alpha=&Sine(Normalized Angle=&Multiply_Float(A=$User.RarityChangeFloat_01, B=10), Period=0.75, Scale=0.5, Bias=0.5)
      ))
     Alpha=&MakeCustomFloatFromBool(
      Bool=&Boolean_AndOperation(
       Bool1=&SetBoolByFloatComparison(A=$User.RarityChangeFloat_01, B=0.01)
       Bool2=&SetBoolByFloatComparison(A=$User.RarityChangeFloat_01, B=0.99, Comparison Type=::NewEnumerator6)
      )
      True Float=1
      False Float=0
     )
    )
    ScaleA=true
    ScaleRGB=true
   ) {module=ScaleColor005}
  ]
  rendererClasses [&NiagaraMeshRendererProperties]
  renderers [{rendererClass=&NiagaraMeshRendererProperties, rendererIndex=0}]
 }
 Cone_pulse001 {
  bEnabled=false
  emitterUpdateScript [
   &EmitterState(
    Inactive Response=::NewEnumerator1
    Life Cycle Mode=::NewEnumerator1
    Loop Behavior=::NewEnumerator1
    Loop Duration Mode=::NewEnumerator1
    UseLoopDelay=false
   )
   &SpawnBurst_Instantaneous(Spawn Count=1, Use Spawn Probability=false)
  ]
  particleSpawnScript [
   &InitializeParticle(
    Mesh Scale=&Multiply_VectorByFloat(Vector=<0.22 0.32 0.32>, Float=0.8)
    Mesh Scale Mode=::NewEnumerator4
    Position Offset=&Add_Vector(A=$System.Master Offset, B=<0 0 -5>)
    UsePositionOffset=true
    Write Material Random=true
   )
   &InitialMeshOrientation(Orientation Vector=<0 0 1>, Rotation=<0 0.01 0>, Use Orientation Vector=true, Use Rotation Vector=false)
   /SetVariables(Particles.FriendlyColor:LinearColor=<rgba 0 7.5 50 1>, Particles.HostileColor:LinearColor=<rgba 50 3.3334 0 1>) {isSetParameters=true, module=SetVariables_#14}
  ]
  particleUpdateScript [
   &ParticleState(Kill Particles When Lifetime Has Elapsed=false, Let Infinitely Lived Particles Die When Emitter Deactivates=true)
   &ApplyOwnerScaleToAttributes(Scale Initial Mesh Scale=true)
   &SolveForcesAndVelocity(Clamp Velocity=false, Limit Acceleration=false)
   &ScaleColor(
    Scale Alpha=&Multiply_Float(A=0.015, B=$System.TOD)
    Scale RGB=&Multiply_VectorByFloat(Vector=&Multiply_VectorByFloat(Vector=<1 1 1>, Float=1), Float=$System.TOD)
    ScaleA=true
    ScaleRGB=false
   )
   &ScaleColor(
    Scale RGB=&MakeVectorFromLinearColorRGB(Color=&LerpLinearColors(
      StartColor=<rgba 50 19.9826 2.6042 1>
      EndColor=&LerpLinearColors(StartColor=$Particles.FriendlyColor, EndColor=$Particles.HostileColor, LerpFactor=$User.FriendlyHostileFloat)
      LerpFactor=0
     ))
    ScaleA=true
    ScaleRGB=true
   ) {module=ScaleColor002}
   &ScaleMeshSize(Scale Factor=&VectorFromFloat(Value=&Multiply_Float(A=3, B=$System.Master Scale)))
   &DynamicMaterialParameters(
    Index 0 Param 1=$System.Timeline.NormalizedTime
    Module.Param0WriteEnabled=true
    Module.Param1WriteEnabled=true
    Module.Param2WriteEnabled=true
    Module.Param3WriteEnabled=true
   )
   &ScaleColor(Scale Alpha=&Multiply_Float(A=$System.LightConeFade, B=$System.Flicker), ScaleA=true, ScaleRGB=false) {module=ScaleColor001}
  ]
  rendererClasses [&NiagaraMeshRendererProperties]
  renderers [{rendererClass=&NiagaraMeshRendererProperties, rendererIndex=0}]
 }
 Cone_pulse002 {
  bEnabled=false
  emitterUpdateScript [
   &EmitterState(
    Inactive Response=::NewEnumerator1
    Life Cycle Mode=::NewEnumerator1
    Loop Behavior=::NewEnumerator1
    Loop Duration Mode=::NewEnumerator1
    UseLoopDelay=false
   )
   &SpawnBurst_Instantaneous(Spawn Count=1, Use Spawn Probability=false)
  ]
  particleSpawnScript [
   &InitializeParticle(
    Mesh Scale=&Multiply_VectorByFloat(Vector=<0.22 0.32 0.32>, Float=0.8)
    Mesh Scale Mode=::NewEnumerator4
    Position Offset=&Add_Vector(A=$System.Master Offset, B=<0 0 -5>)
    UsePositionOffset=true
    Write Material Random=true
   )
   &InitialMeshOrientation(Orientation Vector=<0 0 1>, Rotation=<0 0.01 0>, Use Orientation Vector=true, Use Rotation Vector=false)
   /SetVariables(Particles.FriendlyColor:LinearColor=<rgba 0 7.5 50 1>, Particles.HostileColor:LinearColor=<rgba 50 3.3334 0 1>) {isSetParameters=true, module=SetVariables_#14}
  ]
  particleUpdateScript [
   &ParticleState(Kill Particles When Lifetime Has Elapsed=false, Let Infinitely Lived Particles Die When Emitter Deactivates=true)
   &ApplyOwnerScaleToAttributes(Scale Initial Mesh Scale=true)
   &SolveForcesAndVelocity(Clamp Velocity=false, Limit Acceleration=false)
   &ScaleColor(
    Scale Alpha=&Multiply_Float(A=0.015, B=$System.TOD)
    Scale RGB=&Multiply_VectorByFloat(Vector=&Multiply_VectorByFloat(Vector=<1 1 1>, Float=1), Float=$System.TOD)
    ScaleA=true
    ScaleRGB=false
   )
   &ScaleColor(
    Scale RGB=&MakeVectorFromLinearColorRGB(Color=&LerpLinearColors(
      StartColor=<rgba 50 19.9826 2.6042 1>
      EndColor=&LerpLinearColors(StartColor=$Particles.FriendlyColor, EndColor=$Particles.HostileColor, LerpFactor=$User.FriendlyHostileFloat)
      LerpFactor=0
     ))
    ScaleA=true
    ScaleRGB=true
   ) {enabled=false, module=ScaleColor002}
   &ScaleColor(
    Scale RGB=&Multiply_VectorByFloat(
     Vector=&MakeVectorFromLinearColorRGB(Color=&LerpLinearColors(
       StartColor=&SelectLinearColorFromArray(
        Array Sampling Mode=::NewEnumerator1
        Direct Array Index=$System.SysCurrentRarity
        Color Selection Array=DataInterface() {propertyValues {ColorData [
           <rgba 0.4271 0.3821 0.3417 1>
           <rgba 0.2063 0.7656 0.0997 1>
           <rgba 0.0656 0.3836 1.8 1>
           <rgba 0.6968 0.0229 2.2 1>
           <rgba 1.5 0.7891 0.0781 1>
           <rgba 0.775 0.3185 0.004 1>
           <rgba 0.1238 1.2143 1.5 1>
           <rgba 0.0908 0.5841 0.006 1>
          ]}}
        Random Seed=?
       )
       EndColor=&SelectLinearColorFromArray(
        Array Sampling Mode=::NewEnumerator1
        Direct Array Index=$System.SysNextRarity
        Color Selection Array=DataInterface() {propertyValues {ColorData [
           <rgba 0.4271 0.3821 0.3417 1>
           <rgba 0.2063 0.7656 0.0997 1>
           <rgba 0.0656 0.3836 1.8 1>
           <rgba 0.6968 0.0229 2.2 1>
           <rgba 1.5 0.5995 0.0781 1>
           <rgba 0.775 0.3185 0.004 1>
           <rgba 0.1238 1.2143 1.5 1>
           <rgba 0.0908 0.5841 0.006 1>
          ]}}
        Random Seed=?
       )
       LerpFactor=$System.LerpRarityColor
      ))
     Float=50
    )
    ScaleA=true
    ScaleRGB=true
   ) {module=ScaleColor003}
   &ScaleMeshSize(Scale Factor=&VectorFromFloat(Value=&Multiply_Float(A=3, B=$System.Master Scale)))
   &DynamicMaterialParameters(
    Index 0 Param 1=$System.Timeline.NormalizedTime
    Module.Param0WriteEnabled=true
    Module.Param1WriteEnabled=true
    Module.Param2WriteEnabled=true
    Module.Param3WriteEnabled=true
   )
   &ScaleColor(Scale Alpha=&Multiply_Float(A=$System.LightConeFade, B=$System.Flicker), ScaleA=true, ScaleRGB=false) {module=ScaleColor001}
   &ScaleColor(
    Scale Alpha=&FloatFromCurve(
     FloatCurve=DataInterface() {propertyValues {Curve {keys [{tangentMode=RCTM_User, time=0, value=1}, {tangentMode=RCTM_User, time=1, value=0}]}}}
     CurveIndex=$System.Timeline005.NormalizedTime
    )
    ScaleA=true
    ScaleRGB=true
   ) {module=ScaleColor004}
   &ScaleColor(
    Scale RGB=&Lerp_Vector(
     A=&VectorFromFloat(Value=1)
     B=&VectorFromFloat(Value=&Lerp_Float(
       A=6
       B=1
       Alpha=&Sine(Normalized Angle=&Multiply_Float(A=$User.RarityChangeFloat_01, B=10), Period=0.75, Scale=0.5, Bias=0.5)
      ))
     Alpha=&MakeCustomFloatFromBool(
      Bool=&Boolean_AndOperation(
       Bool1=&SetBoolByFloatComparison(A=$User.RarityChangeFloat_01, B=0.01)
       Bool2=&SetBoolByFloatComparison(A=$User.RarityChangeFloat_01, B=0.99, Comparison Type=::NewEnumerator6)
      )
      True Float=1
      False Float=0
     )
    )
    ScaleA=true
    ScaleRGB=true
   ) {module=ScaleColor005}
  ]
  rendererClasses [&NiagaraMeshRendererProperties]
  renderers [{rendererClass=&NiagaraMeshRendererProperties, rendererIndex=0}]
 }
 Degrade_Glow {
  emitterUpdateScript [
   &EmitterState()
   &SpawnBurst_Instantaneous(Use Spawn Probability=false, UseLoopCountLimit=false) {enabled=false}
   &SpawnRate(
    Spawn Probability=&MakeCustomFloatFromBool(
     Bool=&Boolean_AndOperation(
      Bool1=&SetBoolByFloatComparison(A=$System.LerpRarityColor, B=0.01)
      Bool2=&SetBoolByFloatComparison(A=$System.LerpRarityColor, B=0.99, Comparison Type=::NewEnumerator6)
     )
     True Float=1
     False Float=0
    )
    SpawnRate=1.3333
    Use Spawn Probability=true
   )
  ]
  particleSpawnScript [
   &InitializeParticle(
    Color Mode=::NewEnumerator0
    Lifetime=1.65
    Position Offset=<0 0 25>
    Sprite Size Mode=::NewEnumerator3
    Uniform Sprite Size=600
    UsePositionOffset=true
    Write Material Random=true
   )
   &CameraOffset(Camera Offset Amount=50)
  ]
  particleUpdateScript [
   &ParticleState()
   &ScaleSpriteSize(Uniform Curve Sprite Scale=DataInterface() {propertyValues {Curve {keys [{arriveTangent=1.5807, interpMode=RCIM_Cubic, leaveTangent=1.5807, tangentMode=RCTM_User, time=0, value=0}, <at 1 1 RCIM_Cubic>]}}})
   &SolveForcesAndVelocity(Clamp Velocity=false, Limit Acceleration=false)
   &DynamicMaterialParameters(
    Index 0 Param 1=1
    Index 0 Param 2=1
    Index 0 Param 4=1
    Module.Param0WriteEnabled=true
    Module.Param1WriteEnabled=true
    Module.Param2WriteEnabled=true
    Module.Param3WriteEnabled=true
   )
   &ScaleColor(
    Scale Alpha=0.1
    Scale RGB=&MakeVectorFromLinearColorRGB(Color=&LerpLinearColors(
      StartColor=&SelectLinearColorFromArray(
       Array Sampling Mode=::NewEnumerator1
       Direct Array Index=$System.SysCurrentRarity
       Color Selection Array=DataInterface() {propertyValues {ColorData [
          <rgba 0.4271 0.3821 0.3417 1>
          <rgba 0.2063 0.7656 0.0997 1>
          <rgba 0.0656 0.3836 1.8 1>
          <rgba 0.6968 0.0229 2.2 1>
          <rgba 1.5 0.7891 0.0781 1>
          <rgba 0.775 0.3185 0.004 1>
          <rgba 0.1238 1.2143 1.5 1>
          <rgba 0.0908 0.5841 0.006 1>
         ]}}
       Random Seed=?
      )
      EndColor=&SelectLinearColorFromArray(
       Array Sampling Mode=::NewEnumerator1
       Direct Array Index=$System.SysNextRarity
       Color Selection Array=DataInterface() {propertyValues {ColorData [
          <rgba 0.4271 0.3821 0.3417 1>
          <rgba 0.2063 0.7656 0.0997 1>
          <rgba 0.0656 0.3836 1.8 1>
          <rgba 0.6968 0.0229 2.2 1>
          <rgba 1.5 0.5995 0.0781 1>
          <rgba 0.775 0.3185 0.004 1>
          <rgba 0.1238 1.2143 1.5 1>
          <rgba 0.0908 0.5841 0.006 1>
         ]}}
       Random Seed=?
      )
      LerpFactor=0
     ))
    ScaleA=true
    ScaleRGB=true
   ) {module=ScaleColor001}
   &ScaleColor(
    Scale Alpha=&FloatFromCurve(FloatCurve=DataInterface() {propertyValues {Curve {keys [
         {arriveTangent=9.0725, interpMode=RCIM_Cubic, leaveTangent=9.0725, tangentMode=RCTM_User, time=0, value=0}
         <at 0.1408 0.9847 RCIM_Cubic>
         {arriveTangentWeight=0.8458, interpMode=RCIM_Cubic, leaveTangentWeight=0.8458, time=1, value=0}
        ]}}})
    ScaleA=true
    ScaleRGB=true
   )
  ]
  rendererClasses [&NiagaraSpriteRendererProperties]
  renderers [{rendererClass=&NiagaraSpriteRendererProperties}]
 }
 Degrade_Glow001 {
  bEnabled=false
  emitterUpdateScript [
   &EmitterState()
   &SpawnBurst_Instantaneous(Use Spawn Probability=false, UseLoopCountLimit=false) {enabled=false}
   &SpawnRate(
    Spawn Probability=&MakeCustomFloatFromBool(
     Bool=&Boolean_AndOperation(
      Bool1=&SetBoolByFloatComparison(A=$User.RarityChangeFloat_01, B=0.01)
      Bool2=&SetBoolByFloatComparison(A=$User.RarityChangeFloat_01, B=0.99, Comparison Type=::NewEnumerator6)
     )
     True Float=1
     False Float=0
    )
    SpawnRate=1.3333
    Use Spawn Probability=true
   )
  ]
  particleSpawnScript [
   &InitializeParticle(
    Color Mode=::NewEnumerator0
    Lifetime=1.65
    Position Offset=<0 0 25>
    Sprite Size Mode=::NewEnumerator3
    Uniform Sprite Size=600
    UsePositionOffset=true
    Write Material Random=true
   )
   &CameraOffset(Camera Offset Amount=50)
  ]
  particleUpdateScript [
   &ParticleState()
   &ScaleSpriteSize(Uniform Curve Sprite Scale=DataInterface() {propertyValues {Curve {keys [{arriveTangent=1.5807, interpMode=RCIM_Cubic, leaveTangent=1.5807, tangentMode=RCTM_User, time=0, value=0}, <at 1 1 RCIM_Cubic>]}}})
   &SolveForcesAndVelocity(Clamp Velocity=false, Limit Acceleration=false)
   &DynamicMaterialParameters(
    Index 0 Param 1=1
    Index 0 Param 2=1
    Index 0 Param 4=1
    Module.Param0WriteEnabled=true
    Module.Param1WriteEnabled=true
    Module.Param2WriteEnabled=true
    Module.Param3WriteEnabled=true
   )
   &ScaleColor(
    Scale Alpha=0.1
    Scale RGB=&MakeVectorFromLinearColorRGB(Color=&LerpLinearColors(
      StartColor=&SelectLinearColorFromArray(
       Array Sampling Mode=::NewEnumerator1
       Direct Array Index=$System.SysCurrentRarity
       Color Selection Array=DataInterface() {propertyValues {ColorData [
          <rgba 0.4271 0.3821 0.3417 1>
          <rgba 0.2063 0.7656 0.0997 1>
          <rgba 0.0656 0.3836 1.8 1>
          <rgba 0.6968 0.0229 2.2 1>
          <rgba 1.5 0.7891 0.0781 1>
          <rgba 0.775 0.3185 0.004 1>
          <rgba 0.1238 1.2143 1.5 1>
          <rgba 0.0908 0.5841 0.006 1>
         ]}}
       Random Seed=?
      )
      EndColor=&SelectLinearColorFromArray(
       Array Sampling Mode=::NewEnumerator1
       Direct Array Index=$System.SysNextRarity
       Color Selection Array=DataInterface() {propertyValues {ColorData [
          <rgba 0.4271 0.3821 0.3417 1>
          <rgba 0.2063 0.7656 0.0997 1>
          <rgba 0.0656 0.3836 1.8 1>
          <rgba 0.6968 0.0229 2.2 1>
          <rgba 1.5 0.5995 0.0781 1>
          <rgba 0.775 0.3185 0.004 1>
          <rgba 0.1238 1.2143 1.5 1>
          <rgba 0.0908 0.5841 0.006 1>
         ]}}
       Random Seed=?
      )
      LerpFactor=$System.LerpRarityColor
     ))
    ScaleA=true
    ScaleRGB=true
   ) {module=ScaleColor001}
   &ScaleColor(
    Scale Alpha=&FloatFromCurve(FloatCurve=DataInterface() {propertyValues {Curve {keys [
         {arriveTangent=9.0725, interpMode=RCIM_Cubic, leaveTangent=9.0725, tangentMode=RCTM_User, time=0, value=0}
         <at 0.1408 0.9847 RCIM_Cubic>
         {arriveTangentWeight=0.8458, interpMode=RCIM_Cubic, leaveTangentWeight=0.8458, time=1, value=0}
        ]}}})
    ScaleA=true
    ScaleRGB=true
   )
  ]
  rendererClasses [&NiagaraSpriteRendererProperties]
  renderers [{rendererClass=&NiagaraSpriteRendererProperties}]
 }
 Glow {
  emitterUpdateScript [
   &EmitterState(
    Inactive Response=::NewEnumerator1
    Life Cycle Mode=::NewEnumerator1
    Loop Behavior=::NewEnumerator1
    Loop Duration Mode=::NewEnumerator1
    UseLoopDelay=false
   )
   &SpawnBurst_Instantaneous(Spawn Count=1, Use Spawn Probability=false)
  ]
  particleSpawnScript [
   &FN_SetSpawnGroup() {enabled=false}
   &InitializeParticle(
    Position Offset=$System.Master Offset
    Sprite Rotation Mode=::NewEnumerator2
    Sprite Size Mode=::NewEnumerator3
    Uniform Sprite Size=64.8455
    UsePositionOffset=true
    Write Material Random=true
   ) {module=InitializeParticle001}
   &DynamicMaterialParameters(
    Index 0 Param 1=50
    Module.Param0WriteEnabled=true
    Module.Param1WriteEnabled=true
    Module.Param2WriteEnabled=true
    Module.Param3WriteEnabled=true
   )
   &SystemLocation(MaskSpawnGroup=false) {enabled=false}
   &CameraOffset(Camera Offset Amount=4)
   /SetVariables(Particles.FriendlyColor:LinearColor=<rgba 0 0.6 4 1>, Particles.HostileColor:LinearColor=<rgba 4 0.2667 0 1>) {isSetParameters=true, module=SetVariables_#14}
  ]
  particleUpdateScript [
   &ParticleState(Kill Particles When Lifetime Has Elapsed=false, Let Infinitely Lived Particles Die When Emitter Deactivates=true)
   &SolveForcesAndVelocity(Clamp Velocity=false, Limit Acceleration=false)
   &ScaleColor(
    Scale Alpha=&FNTimeOfDayScalar(Night Scale=1, Morning Scale=0.5, Evening Scale=0.8)
    Scale RGB=&VectorFromFloat(Value=0.8)
    ScaleA=true
    ScaleRGB=true
   )
   &ScaleSpriteSize(Scale Factor=&Multiply_Vector2D_ByFloat(
     Float=$System.Master Scale
     Vector2D=&Multiply_Vector2(
      A=&Vector2DFromFloat(Value=&RandomRangeFloat(Minimum=0.85, Maximum=1, Evaluation Type=::NewEnumerator1))
      B=&Vector2DFromFloat(Value=&MakeFloatFromVector(Vector=$Engine.Owner.Scale, Channel=::NewEnumerator0))
     )
    )) {module=ScaleSpriteSize001}
   &ScaleColor(
    Scale RGB=&Multiply_VectorByFloat(
     Vector=&MakeVectorFromLinearColorRGB(Color=&LerpLinearColors(
       StartColor=&SelectLinearColorFromArray(
        Array Sampling Mode=::NewEnumerator1
        Direct Array Index=$System.SysCurrentRarity
        Color Selection Array=DataInterface() {propertyValues {ColorData [
           <rgba 1 0.8946 0.8 1>
           <rgba 0.3368 1.25 0.1628 1>
           <rgba 0.1458 0.8524 4 1>
           <rgba 1.267 0.0417 4 1>
           <rgba 1.5 0.7891 0.0781 1>
           <rgba 2 1.0052 0.0104 1>
           <rgba 0.1238 1.2143 1.5 1>
           <rgba 0.1866 1.2 0.0124 1>
          ]}}
        Random Seed=?
       )
       EndColor=&SelectLinearColorFromArray(
        Array Sampling Mode=::NewEnumerator1
        Direct Array Index=$System.SysNextRarity
        Color Selection Array=DataInterface() {propertyValues {ColorData [
           <rgba 1 0.8946 0.8 1>
           <rgba 0.3368 1.25 0.1628 1>
           <rgba 0.1458 0.8524 4 1>
           <rgba 1.267 0.0417 4 1>
           <rgba 1.5 0.7891 0.0781 1>
           <rgba 2 1.0052 0.0104 1>
           <rgba 0.1238 1.2143 1.5 1>
           <rgba 0.1866 1.2 0.0124 1>
          ]}}
        Random Seed=?
       )
       LerpFactor=0
      ))
     Float=0.8
    )
    ScaleA=true
    ScaleRGB=true
   ) {module=ScaleColor003}
   &ScaleColor(
    Color Value To Scale=$Particles.Color
    Scale Alpha=&Multiply_Float(A=$System.LightConeFade, B=$System.Flicker)
    Scale RGB=&VectorFromFloat(Value=&Lerp_Float(A=1, B=$System.Flash, Alpha=0.5))
    ScaleA=true
    ScaleRGB=true
   ) {module=ScaleColor001}
   &ScaleColor(
    Scale Alpha=&FloatFromCurve(
     FloatCurve=DataInterface() {propertyValues {Curve {keys [{tangentMode=RCTM_User, time=0, value=1}, {tangentMode=RCTM_User, time=1, value=0}]}}}
     CurveIndex=$System.Timeline005.NormalizedTime
    )
    ScaleA=true
    ScaleRGB=true
   ) {module=ScaleColor002}
   &ScaleColor(
    Scale RGB=&Lerp_Vector(
     A=&VectorFromFloat(Value=1)
     B=&VectorFromFloat(Value=&Lerp_Float(
       A=0.5
       B=1
       Alpha=&Lerp_Float(
        A=0
        B=&OneMinusFloat(Float=&Round_Float(Float:NiagaraFloat=&ClampFloat(Float=&Sine(
            Normalized Angle=&Multiply_Float(A=$User.RarityChangeFloat_01, B=10)
            Period=&FloatFromCurve(
             FloatCurve=DataInterface() {propertyValues {Curve {keys [{tangentMode=RCTM_User, time=0, value=2}, <at 0.925 0.75>, {tangentMode=RCTM_User, time=0.925, value=0.05}, <at 0.995 0.05>, <at 1 0>]}}}
             CurveIndex=$User.RarityChangeFloat_01
            )
            Scale=0.5
            Bias=&FloatFromCurve(
             FloatCurve=DataInterface() {propertyValues {Curve {keys [
                 <at 0 0.9 RCIM_Cubic>
                 {arriveTangent=-0.432, interpMode=RCIM_Cubic, leaveTangent=-0.432, time=0.925, value=0.65}
                 <at 0.926 0.5>
                 <at 1 0.5>
                ]}}}
             CurveIndex=$User.RarityChangeFloat_01
            )
           ))))
        Alpha=&FloatFromCurve(
         FloatCurve=DataInterface() {propertyValues {Curve {keys [<at 0 0>,<at 0.001 1>,<at 0.999 1>,<at 1 0>]}}}
         CurveIndex=$User.RarityChangeFloat_01
        )
       )
      ))
     Alpha=&MakeCustomFloatFromBool(
      Bool=&Boolean_AndOperation(
       Bool1=&SetBoolByFloatComparison(A=$User.RarityChangeFloat_01, B=0.01)
       Bool2=&SetBoolByFloatComparison(A=$User.RarityChangeFloat_01, B=0.99, Comparison Type=::NewEnumerator6)
      )
      True Float=1
      False Float=0
     )
    )
    ScaleA=true
    ScaleRGB=true
   ) {module=ScaleColor004}
  ]
  rendererClasses [&NiagaraSpriteRendererProperties]
  renderers [{rendererClass=&NiagaraSpriteRendererProperties}]
 }
 Glow001 {
  bEnabled=false
  emitterUpdateScript [
   &EmitterState(
    Inactive Response=::NewEnumerator1
    Life Cycle Mode=::NewEnumerator1
    Loop Behavior=::NewEnumerator1
    Loop Duration Mode=::NewEnumerator1
    UseLoopDelay=false
   )
   &SpawnBurst_Instantaneous(Spawn Count=1, Use Spawn Probability=false)
  ]
  particleSpawnScript [
   &FN_SetSpawnGroup() {enabled=false}
   &InitializeParticle(
    Position Offset=$System.Master Offset
    Sprite Rotation Mode=::NewEnumerator2
    Sprite Size Mode=::NewEnumerator3
    Uniform Sprite Size=64.8455
    UsePositionOffset=true
    Write Material Random=true
   ) {module=InitializeParticle001}
   &DynamicMaterialParameters(
    Index 0 Param 1=50
    Module.Param0WriteEnabled=true
    Module.Param1WriteEnabled=true
    Module.Param2WriteEnabled=true
    Module.Param3WriteEnabled=true
   )
   &SystemLocation(MaskSpawnGroup=false) {enabled=false}
   &CameraOffset(Camera Offset Amount=4)
   /SetVariables(Particles.FriendlyColor:LinearColor=<rgba 0 0.6 4 1>, Particles.HostileColor:LinearColor=<rgba 4 0.2667 0 1>) {isSetParameters=true, module=SetVariables_#14}
  ]
  particleUpdateScript [
   &ParticleState(Kill Particles When Lifetime Has Elapsed=false, Let Infinitely Lived Particles Die When Emitter Deactivates=true)
   &SolveForcesAndVelocity(Clamp Velocity=false, Limit Acceleration=false)
   &ScaleColor(
    Scale Alpha=&FNTimeOfDayScalar(Night Scale=1, Morning Scale=0.5, Evening Scale=0.8)
    Scale RGB=&VectorFromFloat(Value=0.8)
    ScaleA=true
    ScaleRGB=true
   )
   &ScaleSpriteSize(Scale Factor=&Multiply_Vector2D_ByFloat(
     Float=$System.Master Scale
     Vector2D=&Multiply_Vector2(
      A=&Vector2DFromFloat(Value=&RandomRangeFloat(Minimum=0.85, Maximum=1, Evaluation Type=::NewEnumerator1))
      B=&Vector2DFromFloat(Value=&MakeFloatFromVector(Vector=$Engine.Owner.Scale, Channel=::NewEnumerator0))
     )
    )) {module=ScaleSpriteSize001}
   &ScaleColor(
    Scale RGB=&MakeVectorFromLinearColorRGB(Color=&LerpLinearColors(
      StartColor=<rgba 1.5 0.5995 0.0781 1>
      EndColor=&LerpLinearColors(StartColor=$Particles.FriendlyColor, EndColor=$Particles.HostileColor, LerpFactor=$User.FriendlyHostileFloat)
      LerpFactor=0
     ))
    ScaleA=true
    ScaleRGB=true
   ) {module=ScaleColor002}
   &ScaleColor(
    Color Value To Scale=$Particles.Color
    Scale Alpha=&Multiply_Float(A=$System.LightConeFade, B=$System.Flicker)
    Scale RGB=&VectorFromFloat(Value=&Lerp_Float(A=1, B=$System.Flash, Alpha=0.5))
    ScaleA=true
    ScaleRGB=true
   ) {module=ScaleColor001}
  ]
  rendererClasses [&NiagaraSpriteRendererProperties]
  renderers [{rendererClass=&NiagaraSpriteRendererProperties}]
 }
 Glow002 {
  bEnabled=false
  emitterUpdateScript [
   &EmitterState(
    Inactive Response=::NewEnumerator1
    Life Cycle Mode=::NewEnumerator1
    Loop Behavior=::NewEnumerator1
    Loop Duration Mode=::NewEnumerator1
    UseLoopDelay=false
   )
   &SpawnBurst_Instantaneous(Spawn Count=1, Use Spawn Probability=false)
  ]
  particleSpawnScript [
   &FN_SetSpawnGroup() {enabled=false}
   &InitializeParticle(
    Position Offset=$System.Master Offset
    Sprite Rotation Mode=::NewEnumerator2
    Sprite Size Mode=::NewEnumerator3
    Uniform Sprite Size=64.8455
    UsePositionOffset=true
    Write Material Random=true
   ) {module=InitializeParticle001}
   &DynamicMaterialParameters(
    Index 0 Param 1=50
    Module.Param0WriteEnabled=true
    Module.Param1WriteEnabled=true
    Module.Param2WriteEnabled=true
    Module.Param3WriteEnabled=true
   )
   &SystemLocation(MaskSpawnGroup=false) {enabled=false}
   &CameraOffset(Camera Offset Amount=4)
   /SetVariables(Particles.FriendlyColor:LinearColor=<rgba 0 0.6 4 1>, Particles.HostileColor:LinearColor=<rgba 4 0.2667 0 1>) {isSetParameters=true, module=SetVariables_#14}
  ]
  particleUpdateScript [
   &ParticleState(Kill Particles When Lifetime Has Elapsed=false, Let Infinitely Lived Particles Die When Emitter Deactivates=true)
   &SolveForcesAndVelocity(Clamp Velocity=false, Limit Acceleration=false)
   &ScaleColor(
    Scale Alpha=&FNTimeOfDayScalar(Night Scale=1, Morning Scale=0.5, Evening Scale=0.8)
    Scale RGB=&VectorFromFloat(Value=0.8)
    ScaleA=true
    ScaleRGB=true
   )
   &ScaleSpriteSize(Scale Factor=&Multiply_Vector2D_ByFloat(
     Float=$System.Master Scale
     Vector2D=&Multiply_Vector2(
      A=&Vector2DFromFloat(Value=&RandomRangeFloat(Minimum=0.85, Maximum=1, Evaluation Type=::NewEnumerator1))
      B=&Vector2DFromFloat(Value=&MakeFloatFromVector(Vector=$Engine.Owner.Scale, Channel=::NewEnumerator0))
     )
    )) {module=ScaleSpriteSize001}
   &ScaleColor(
    Scale RGB=&Multiply_VectorByFloat(
     Vector=&MakeVectorFromLinearColorRGB(Color=&LerpLinearColors(
       StartColor=&SelectLinearColorFromArray(
        Array Sampling Mode=::NewEnumerator1
        Direct Array Index=$System.SysCurrentRarity
        Color Selection Array=DataInterface() {propertyValues {ColorData [
           <rgba 1 0.8946 0.8 1>
           <rgba 0.3368 1.25 0.1628 1>
           <rgba 0.1458 0.8524 4 1>
           <rgba 1.267 0.0417 4 1>
           <rgba 1.5 0.7891 0.0781 1>
           <rgba 2 1.0052 0.0104 1>
           <rgba 0.1238 1.2143 1.5 1>
           <rgba 0.1866 1.2 0.0124 1>
          ]}}
        Random Seed=?
       )
       EndColor=&SelectLinearColorFromArray(
        Array Sampling Mode=::NewEnumerator1
        Direct Array Index=$System.SysNextRarity
        Color Selection Array=DataInterface() {propertyValues {ColorData [
           <rgba 1 0.8946 0.8 1>
           <rgba 0.3368 1.25 0.1628 1>
           <rgba 0.1458 0.8524 4 1>
           <rgba 1.267 0.0417 4 1>
           <rgba 1.5 0.7891 0.0781 1>
           <rgba 2 1.0052 0.0104 1>
           <rgba 0.1238 1.2143 1.5 1>
           <rgba 0.1866 1.2 0.0124 1>
          ]}}
        Random Seed=?
       )
       LerpFactor=$System.LerpRarityColor
      ))
     Float=0.8
    )
    ScaleA=true
    ScaleRGB=true
   ) {module=ScaleColor003}
   &ScaleColor(
    Color Value To Scale=$Particles.Color
    Scale Alpha=&Multiply_Float(A=$System.LightConeFade, B=$System.Flicker)
    Scale RGB=&VectorFromFloat(Value=&Lerp_Float(A=1, B=$System.Flash, Alpha=0.5))
    ScaleA=true
    ScaleRGB=true
   ) {module=ScaleColor001}
   &ScaleColor(
    Scale Alpha=&FloatFromCurve(
     FloatCurve=DataInterface() {propertyValues {Curve {keys [{tangentMode=RCTM_User, time=0, value=1}, {tangentMode=RCTM_User, time=1, value=0}]}}}
     CurveIndex=$System.Timeline005.NormalizedTime
    )
    ScaleA=true
    ScaleRGB=true
   ) {module=ScaleColor002}
   &ScaleColor(
    Scale RGB=&Lerp_Vector(
     A=&VectorFromFloat(Value=1)
     B=&VectorFromFloat(Value=&Lerp_Float(
       A=6
       B=1
       Alpha=&Sine(Normalized Angle=&Multiply_Float(A=$User.RarityChangeFloat_01, B=10), Period=0.75, Scale=0.5, Bias=0.5)
      ))
     Alpha=&MakeCustomFloatFromBool(
      Bool=&Boolean_AndOperation(
       Bool1=&SetBoolByFloatComparison(A=$User.RarityChangeFloat_01, B=0.01)
       Bool2=&SetBoolByFloatComparison(A=$User.RarityChangeFloat_01, B=0.99, Comparison Type=::NewEnumerator6)
      )
      True Float=1
      False Float=0
     )
    )
    ScaleA=true
    ScaleRGB=true
   ) {module=ScaleColor004}
  ]
  rendererClasses [&NiagaraSpriteRendererProperties]
  renderers [{rendererClass=&NiagaraSpriteRendererProperties}]
 }
 Light_loop {
  emitterUpdateScript [
   &EmitterState(
    Inactive Response=::NewEnumerator1
    Life Cycle Mode=::NewEnumerator1
    Loop Behavior=::NewEnumerator1
    Loop Delay=0.9
    Loop Duration Mode=::NewEnumerator1
    UseLoopDelay=false
   )
   &SpawnBurst_Instantaneous(Spawn Count=1, Use Spawn Probability=false)
  ]
  particleSpawnScript [
   /SetVariables(Particles.Size=0.3909) {isSetParameters=true, module=SetVariables_#22}
   &InitializeParticle(
    Color=<rgba 50 50 50 1>
    Mesh Scale Mode=::NewEnumerator3
    Mesh Uniform Scale=$Particles.Size
    Position Offset=&Add_Vector(A=$System.Master Offset, B=<0 0 35>)
    Sprite Size Mode=::NewEnumerator3
    Uniform Sprite Size=30
    UsePositionOffset=true
    Write Material Random=true
   ) {module=InitializeParticle001}
   &SystemLocation(MaskSpawnGroup=false, Offset=<10 0 0>) {enabled=false}
   /SetVariables(Particles.FriendlyColor:LinearColor=<rgba 0 7.5 50 1>, Particles.HostileColor:LinearColor=<rgba 25 1.6667 0 1>) {enabled=false, isSetParameters=true, module=SetVariables_#14}
  ]
  particleUpdateScript [
   /SetVariables(Particles.LightRadius=&Multiply_Float(A=100, B=&MakeFloatFromVector(Vector=$Engine.Owner.Scale, Channel=::NewEnumerator0))) {enabled=false, isSetParameters=true, module=SetVariables_#23}
   &ParticleState(
    Kill Particles When Lifetime Has Elapsed=false
    Let Infinitely Lived Particles Die When Emitter Deactivates=true
    Loop Particles Lifetime=true
   )
   &SolveForcesAndVelocity(Clamp Velocity=false, Limit Acceleration=false)
   &ScaleColor(
    Scale RGB=&VectorFromFloat(Value=&Multiply_Float(A=&Multiply_Float(A=$System.LightConeFade, B=$System.TOD), B=$System.Flicker))
    ScaleA=false
    ScaleRGB=true
   ) {enabled=false, module=ScaleColor001}
   &ScaleColor(
    Scale RGB=&Multiply_VectorByFloat(
     Vector=&MakeVectorFromLinearColorRGB(Color=&LerpLinearColors(
       StartColor=&SelectLinearColorFromArray(
        Array Sampling Mode=::NewEnumerator1
        Direct Array Index=$System.SysCurrentRarity
        Color Selection Array=DataInterface() {propertyValues {ColorData [
           <rgba 1 0.8946 0.8 1>
           <rgba 0.3368 1.25 0.1628 1>
           <rgba 0.1458 0.8524 4 1>
           <rgba 1.267 0.0417 4 1>
           <rgba 3 1.5781 0.1562 1>
           <rgba 3 1.1099 0.0156 1>
           <rgba 0.2477 2.4285 3 1>
           <rgba 0.3111 2 0.0207 1>
          ]}}
        Random Seed=?
       )
       EndColor=&SelectLinearColorFromArray(
        Array Sampling Mode=::NewEnumerator1
        Direct Array Index=$System.SysNextRarity
        Color Selection Array=DataInterface() {propertyValues {ColorData [
           <rgba 1 0.8946 0.8 1>
           <rgba 0.3233 1.2 0.1562 1>
           <rgba 0.1458 0.8524 4 1>
           <rgba 1.267 0.0417 4 1>
           <rgba 1.5 0.5995 0.0781 1>
           <rgba 0.775 0.3185 0.004 1>
           <rgba 0.2477 2.4285 3 1>
           <rgba 0.3111 2 0.0207 1>
          ]}}
        Random Seed=?
       )
       LerpFactor=0
      ))
     Float=0.8
    )
    ScaleA=true
    ScaleRGB=true
   ) {module=ScaleColor003}
   &ScaleColor(
    Scale Alpha=&FloatFromCurve(
     FloatCurve=DataInterface() {propertyValues {Curve {keys [{tangentMode=RCTM_User, time=0, value=1}, {tangentMode=RCTM_User, time=1, value=0}]}}}
     CurveIndex=$System.Timeline005.NormalizedTime
    )
    ScaleA=true
    ScaleRGB=true
   )
   &ScaleColor(
    Scale RGB=&Lerp_Vector(
     A=&VectorFromFloat(Value=0.85)
     B=&VectorFromFloat(Value=&Lerp_Float(
       A=0.5
       B=1
       Alpha=&Lerp_Float(
        A=0
        B=&OneMinusFloat(Float=&Round_Float(Float:NiagaraFloat=&ClampFloat(Float=&Sine(
            Normalized Angle=&Multiply_Float(A=$User.RarityChangeFloat_01, B=10)
            Period=&FloatFromCurve(
             FloatCurve=DataInterface() {propertyValues {Curve {keys [{tangentMode=RCTM_User, time=0, value=2}, <at 0.925 0.75>, {tangentMode=RCTM_User, time=0.925, value=0.05}, <at 0.995 0.05>, <at 1 0>]}}}
             CurveIndex=$User.RarityChangeFloat_01
            )
            Scale=0.5
            Bias=&FloatFromCurve(
             FloatCurve=DataInterface() {propertyValues {Curve {keys [
                 <at 0 0.9 RCIM_Cubic>
                 {arriveTangent=-0.432, interpMode=RCIM_Cubic, leaveTangent=-0.432, time=0.925, value=0.65}
                 <at 0.926 0.5>
                 <at 1 0.5>
                ]}}}
             CurveIndex=$User.RarityChangeFloat_01
            )
           ))))
        Alpha=&FloatFromCurve(
         FloatCurve=DataInterface() {propertyValues {Curve {keys [<at 0 0>,<at 0.001 1>,<at 0.999 1>,<at 1 0>]}}}
         CurveIndex=$User.RarityChangeFloat_01
        )
       )
      ))
     Alpha=&MakeCustomFloatFromBool(
      Bool=&Boolean_AndOperation(
       Bool1=&SetBoolByFloatComparison(A=$User.RarityChangeFloat_01, B=0.01)
       Bool2=&SetBoolByFloatComparison(A=$User.RarityChangeFloat_01, B=0.99, Comparison Type=::NewEnumerator6)
      )
      True Float=1
      False Float=0
     )
    )
    ScaleA=true
    ScaleRGB=true
   ) {module=ScaleColor002}
   &FN_Light_Attributes(
    Color=$Particles.Color
    Create Color Attribute=true
    Create Exponent Attribute=true
    Create Radius Attribute=true
    Exponent=2
    Radius=350
    Scale Color=0.5
   )
   &FN_Gen5_Light(
    Attenutation=350
    Color:LinearColor=&SelectLinearColorFromArray(
     Array Sampling Mode=::NewEnumerator1
     Direct Array Index=$System.SysCurrentRarity
     Color Selection Array=DataInterface() {propertyValues {ColorData [
        <rgba 1 0.8946 0.8 1>
        <rgba 0.2694 1 0.1302 1>
        <rgba 0.224 0.3663 1 1>
        <rgba 0.3167 0.0104 1 1>
        <rgba 1 0.8262 0.0521 1>
        <rgba 1 0.7679 0.0052 1>
        <rgba 0.0826 0.8095 1 1>
        <rgba 0.1555 1 0.0104 1>
       ]}}
     Random Seed=?
    )
    Exponent=2
    Intensity:NiagaraFloat=&Multiply_Float(
     A=0.4
     B=&Multiply_Float(
      A=&FloatFromCurve(
       FloatCurve=DataInterface() {propertyValues {Curve {keys [{tangentMode=RCTM_User, time=0, value=1}, {tangentMode=RCTM_User, time=1, value=0}]}}}
       CurveIndex=$System.Timeline005.NormalizedTime
       Scale Curve=50
      )
      B=&Lerp_Float(
       A=1
       B=&Lerp_Float(
        A=0.5
        B=1
        Alpha=&Lerp_Float(
         A=0
         B=&OneMinusFloat(Float=&Round_Float(Float:NiagaraFloat=&ClampFloat(Float=&Sine(
             Normalized Angle=&Multiply_Float(A=$User.RarityChangeFloat_01, B=10)
             Period=&FloatFromCurve(
              FloatCurve=DataInterface() {propertyValues {Curve {keys [{tangentMode=RCTM_User, time=0, value=2}, <at 0.925 0.75>, {tangentMode=RCTM_User, time=0.925, value=0.05}, <at 0.995 0.05>, <at 1 0>]}}}
              CurveIndex=$User.RarityChangeFloat_01
             )
             Scale=0.5
             Bias=&FloatFromCurve(
              FloatCurve=DataInterface() {propertyValues {Curve {keys [
                  <at 0 0.9 RCIM_Cubic>
                  {arriveTangent=-0.432, interpMode=RCIM_Cubic, leaveTangent=-0.432, time=0.925, value=0.65}
                  <at 0.926 0.5>
                  <at 1 0.5>
                 ]}}}
              CurveIndex=$User.RarityChangeFloat_01
             )
            ))))
         Alpha=&FloatFromCurve(
          FloatCurve=DataInterface() {propertyValues {Curve {keys [<at 0 0>,<at 0.001 1>,<at 0.999 1>,<at 1 0>]}}}
          CurveIndex=$User.RarityChangeFloat_01
         )
        )
       )
       Alpha=&MakeCustomFloatFromBool(
        Bool=&Boolean_AndOperation(
         Bool1=&SetBoolByFloatComparison(A=$User.RarityChangeFloat_01, B=0.01)
         Bool2=&SetBoolByFloatComparison(A=$User.RarityChangeFloat_01, B=0.99, Comparison Type=::NewEnumerator6)
        )
        True Float=1
        False Float=0
       )
      )
     )
    )
   )
   /SetVariables(Particles.g5lColor:LinearColor=&SelectLinearColorFromArray(
     Array Sampling Mode=::NewEnumerator1
     Direct Array Index=$System.SysCurrentRarity
     Color Selection Array=DataInterface() {propertyValues {ColorData [
        <rgba 1 0.8946 0.8 1>
        <rgba 0.2694 1 0.1302 1>
        <rgba 0.031 0.2086 1 1>
        <rgba 0.3167 0.0104 1 1>
        <rgba 1 0.5576 0.0521 1>
        <rgba 1 0.7679 0.0052 1>
        <rgba 0.0826 0.8095 1 1>
        <rgba 0.1555 1 0.0104 1>
       ]}}
     Random Seed=?
    )) {isSetParameters=true}
   /SetVariables(Particles.New Position:NiagaraPosition=&AddVectorToPosition(
     Position=$Particles.Position
     Vector=&Multiply_VectorByFloat(
      Vector=&ReturnCameraVector(Camera Vectors=::NewEnumerator4, Coordinate Space=::Simulation, Origin Position=$Particles.Position)
      Float=-25
     )
    )) {enabled=false, isSetParameters=true}
  ]
  rendererClasses [
   &NiagaraLightRendererProperties
   &NiagaraComponentRendererProperties
   &NiagaraSpriteRendererProperties
  ]
  renderers [
   {rendererClass=&NiagaraLightRendererProperties, rendererIndex=0}
   {rendererClass=&NiagaraComponentRendererProperties, rendererIndex=1}
   {rendererClass=&NiagaraSpriteRendererProperties, rendererIndex=2}
  ]
 }
 Light_loop001 {
  bEnabled=false
  emitterUpdateScript [
   &EmitterState(
    Inactive Response=::NewEnumerator1
    Life Cycle Mode=::NewEnumerator1
    Loop Behavior=::NewEnumerator1
    Loop Delay=0.9
    Loop Duration Mode=::NewEnumerator1
    UseLoopDelay=false
   )
   &SpawnBurst_Instantaneous(Spawn Count=1, Use Spawn Probability=false)
  ]
  particleSpawnScript [
   /SetVariables(Particles.Size=0.3909) {isSetParameters=true, module=SetVariables_#22}
   &InitializeParticle(
    Color=<rgba 50 19.9826 2.6042 1>
    Mesh Scale Mode=::NewEnumerator3
    Mesh Uniform Scale=$Particles.Size
    Position Offset=&Add_Vector(A=$System.Master Offset, B=<0 0 16>)
    Sprite Size Mode=::NewEnumerator3
    Uniform Sprite Size=30
    UsePositionOffset=true
    Write Material Random=true
   ) {module=InitializeParticle001}
   &SystemLocation(MaskSpawnGroup=false, Offset=<10 0 0>) {enabled=false}
   /SetVariables(Particles.FriendlyColor:LinearColor=<rgba 0 7.5 50 1>, Particles.HostileColor:LinearColor=<rgba 25 1.6667 0 1>) {isSetParameters=true, module=SetVariables_#14}
  ]
  particleUpdateScript [
   /SetVariables(Particles.LightRadius=&Multiply_Float(A=100, B=&MakeFloatFromVector(Vector=$Engine.Owner.Scale, Channel=::NewEnumerator0))) {isSetParameters=true, module=SetVariables_#23}
   &ParticleState(Kill Particles When Lifetime Has Elapsed=false, Let Infinitely Lived Particles Die When Emitter Deactivates=true)
   &SolveForcesAndVelocity(Clamp Velocity=false, Limit Acceleration=false)
   &ScaleColor(
    Scale RGB=&VectorFromFloat(Value=&Multiply_Float(A=&Multiply_Float(A=$System.LightConeFade, B=$System.TOD), B=$System.Flicker))
    ScaleA=false
    ScaleRGB=true
   ) {module=ScaleColor001}
   &FN_Light_Attributes(
    Create Exponent Attribute=true
    Create Radius Attribute=true
    Exponent=3
    Radius=&Multiply_Float(A=100, B=$Particles.Size)
   )
   &ScaleColor(
    Scale Alpha=0.1
    Scale RGB=&MakeVectorFromLinearColorRGB(Color=&LerpLinearColors(
      StartColor=<rgba 50 19.9826 2.6042 1>
      EndColor=&LerpLinearColors(StartColor=$Particles.FriendlyColor, EndColor=$Particles.HostileColor, LerpFactor=$User.FriendlyHostileFloat)
      LerpFactor=0
     ))
    ScaleA=true
    ScaleRGB=true
   ) {module=ScaleColor002}
  ]
  rendererClasses [&NiagaraLightRendererProperties, &NiagaraMeshRendererProperties]
  renderers [
   {rendererClass=&NiagaraLightRendererProperties, rendererIndex=0}
   {rendererClass=&NiagaraMeshRendererProperties, rendererIndex=1}
  ]
 }
 Light_loop002 {
  bEnabled=false
  emitterUpdateScript [
   &EmitterState(
    Inactive Response=::NewEnumerator1
    Life Cycle Mode=::NewEnumerator1
    Loop Behavior=::NewEnumerator1
    Loop Delay=0.9
    Loop Duration Mode=::NewEnumerator1
    UseLoopDelay=false
   )
   &SpawnBurst_Instantaneous(Spawn Count=1, Use Spawn Probability=false)
  ]
  particleSpawnScript [
   /SetVariables(Particles.Size=0.3909) {isSetParameters=true, module=SetVariables_#22}
   &InitializeParticle(
    Color=<rgba 50 50 50 1>
    Mesh Scale Mode=::NewEnumerator3
    Mesh Uniform Scale=$Particles.Size
    Position Offset=&Add_Vector(A=$System.Master Offset, B=<0 0 16>)
    Sprite Size Mode=::NewEnumerator3
    Uniform Sprite Size=30
    UsePositionOffset=true
    Write Material Random=true
   ) {module=InitializeParticle001}
   &SystemLocation(MaskSpawnGroup=false, Offset=<10 0 0>) {enabled=false}
   /SetVariables(Particles.FriendlyColor:LinearColor=<rgba 0 7.5 50 1>, Particles.HostileColor:LinearColor=<rgba 25 1.6667 0 1>) {enabled=false, isSetParameters=true, module=SetVariables_#14}
  ]
  particleUpdateScript [
   /SetVariables(Particles.LightRadius=&Multiply_Float(A=100, B=&MakeFloatFromVector(Vector=$Engine.Owner.Scale, Channel=::NewEnumerator0))) {enabled=false, isSetParameters=true, module=SetVariables_#23}
   &ParticleState(Kill Particles When Lifetime Has Elapsed=false, Let Infinitely Lived Particles Die When Emitter Deactivates=true)
   &SolveForcesAndVelocity(Clamp Velocity=false, Limit Acceleration=false)
   &ScaleColor(
    Scale RGB=&VectorFromFloat(Value=&Multiply_Float(A=&Multiply_Float(A=$System.LightConeFade, B=$System.TOD), B=$System.Flicker))
    ScaleA=false
    ScaleRGB=true
   ) {enabled=false, module=ScaleColor001}
   &FN_Light_Attributes(
    Create Exponent Attribute=true
    Create Radius Attribute=true
    Exponent=3
    Radius=&Multiply_Float(A=100, B=$Particles.Size)
   )
   &ScaleColor(
    Scale RGB=&Multiply_VectorByFloat(
     Vector=&MakeVectorFromLinearColorRGB(Color=&LerpLinearColors(
       StartColor=&SelectLinearColorFromArray(
        Array Sampling Mode=::NewEnumerator1
        Direct Array Index=$System.SysCurrentRarity
        Color Selection Array=DataInterface() {propertyValues {ColorData [
           <rgba 1 0.8946 0.8 1>
           <rgba 0.3368 1.25 0.1628 1>
           <rgba 0.1458 0.8524 4 1>
           <rgba 1.267 0.0417 4 1>
           <rgba 3 1.5781 0.1562 1>
           <rgba 3 1.1099 0.0156 1>
           <rgba 0.2477 2.4285 3 1>
           <rgba 0.3111 2 0.0207 1>
          ]}}
        Random Seed=?
       )
       EndColor=&SelectLinearColorFromArray(
        Array Sampling Mode=::NewEnumerator1
        Direct Array Index=$System.SysNextRarity
        Color Selection Array=DataInterface() {propertyValues {ColorData [
           <rgba 1 0.8946 0.8 1>
           <rgba 0.3233 1.2 0.1562 1>
           <rgba 0.1458 0.8524 4 1>
           <rgba 1.267 0.0417 4 1>
           <rgba 1.5 0.5995 0.0781 1>
           <rgba 0.775 0.3185 0.004 1>
           <rgba 0.2477 2.4285 3 1>
           <rgba 0.3111 2 0.0207 1>
          ]}}
        Random Seed=?
       )
       LerpFactor=$System.LerpRarityColor
      ))
     Float=0.8
    )
    ScaleA=true
    ScaleRGB=true
   ) {module=ScaleColor003}
   &ScaleColor(
    Scale Alpha=&FloatFromCurve(
     FloatCurve=DataInterface() {propertyValues {Curve {keys [{tangentMode=RCTM_User, time=0, value=1}, {tangentMode=RCTM_User, time=1, value=0}]}}}
     CurveIndex=$System.Timeline005.NormalizedTime
    )
    ScaleA=true
    ScaleRGB=true
   )
   &ScaleColor(
    Scale RGB=&Lerp_Vector(
     A=&VectorFromFloat(Value=1)
     B=&VectorFromFloat(Value=&Lerp_Float(
       A=6
       B=1
       Alpha=&Sine(Normalized Angle=&Multiply_Float(A=$User.RarityChangeFloat_01, B=10), Period=0.75, Scale=0.5, Bias=0.5)
      ))
     Alpha=&MakeCustomFloatFromBool(
      Bool=&Boolean_AndOperation(
       Bool1=&SetBoolByFloatComparison(A=$User.RarityChangeFloat_01, B=0.01)
       Bool2=&SetBoolByFloatComparison(A=$User.RarityChangeFloat_01, B=0.99, Comparison Type=::NewEnumerator6)
      )
      True Float=1
      False Float=0
     )
    )
    ScaleA=true
    ScaleRGB=true
   ) {module=ScaleColor002}
  ]
  rendererClasses [&NiagaraLightRendererProperties, &NiagaraMeshRendererProperties]
  renderers [
   {rendererClass=&NiagaraLightRendererProperties, rendererIndex=0}
   {rendererClass=&NiagaraMeshRendererProperties, rendererIndex=1}
  ]
 }
 Ray_Soft {
  emitterUpdateScript [
   &EmitterState(
    Inactive Response=::NewEnumerator1
    Life Cycle Mode=::NewEnumerator1
    Loop Behavior=::NewEnumerator1
    Loop Duration Mode=::NewEnumerator1
    UseLoopDelay=false
   )
   &SpawnBurst_Instantaneous(Spawn Count=1, Use Spawn Probability=false)
  ]
  particleSpawnScript [
   &InitializeParticle(
    Mesh Scale=<0.35 0.32 0.32>
    Mesh Scale Mode=::NewEnumerator4
    Position Offset=&Add_Vector(A=$System.Master Offset, B=<0 0 -10>)
    UsePositionOffset=true
    Write Material Random=true
   )
   &InitialMeshOrientation(Orientation Vector=<0 0 10>, Rotation=<0 0.01 0>, Use Orientation Vector=true, Use Rotation Vector=false) {module=InitialMeshOrientation001}
   /SetVariables(Particles.FriendlyColor:LinearColor=<rgba 0 0.1125 0.75 1>, Particles.HostileColor:LinearColor=<rgba 0.75 0.05 0 1>) {isSetParameters=true, module=SetVariables_#14}
  ]
  particleUpdateScript [
   &ParticleState(Kill Particles When Lifetime Has Elapsed=false, Let Infinitely Lived Particles Die When Emitter Deactivates=true)
   &ApplyOwnerScaleToAttributes(Scale Initial Mesh Scale=true)
   &SolveForcesAndVelocity(Clamp Velocity=false, Limit Acceleration=false)
   &ScaleColor(Scale Alpha=0.7, Scale RGB=&VectorFromFloat(Value=1), ScaleA=true, ScaleRGB=true)
   &ScaleColor(
    Scale RGB=&Multiply_VectorByFloat(
     Vector=&MakeVectorFromLinearColorRGB(Color=&LerpLinearColors(
       StartColor=&SelectLinearColorFromArray(
        Array Sampling Mode=::NewEnumerator1
        Direct Array Index=$System.SysCurrentRarity
        Color Selection Array=DataInterface() {propertyValues {ColorData [
           <rgba 0.4271 0.3821 0.3417 1>
           <rgba 0.2063 0.7656 0.0997 1>
           <rgba 0.0656 0.3836 1.8 1>
           <rgba 0.6968 0.0229 2.2 1>
           <rgba 1.5 0.7891 0.0781 1>
           <rgba 0.775 0.3185 0.004 1>
           <rgba 0.1238 1.2143 1.5 1>
           <rgba 0.0908 0.5841 0.006 1>
          ]}}
        Random Seed=?
       )
       EndColor=&SelectLinearColorFromArray(
        Array Sampling Mode=::NewEnumerator1
        Direct Array Index=$System.SysNextRarity
        Color Selection Array=DataInterface() {propertyValues {ColorData [
           <rgba 0.4271 0.3821 0.3417 1>
           <rgba 0.2063 0.7656 0.0997 1>
           <rgba 0.0656 0.3836 1.8 1>
           <rgba 0.6968 0.0229 2.2 1>
           <rgba 1.5 0.5995 0.0781 1>
           <rgba 0.775 0.3185 0.004 1>
           <rgba 0.1238 1.2143 1.5 1>
           <rgba 0.0908 0.5841 0.006 1>
          ]}}
        Random Seed=?
       )
       LerpFactor=0
      ))
     Float=0.8
    )
    ScaleA=true
    ScaleRGB=true
   ) {module=ScaleColor002}
   &ScaleMeshSize(Scale Factor=&VectorFromFloat(Value=&Multiply_Float(A=3.5, B=$System.Master Scale)))
   &ScaleColor(
    Scale Alpha=&Multiply_Float(A=&Multiply_Float(A=$System.LightConeFade, B=$System.TOD), B=$System.Flicker)
    Scale RGB=&VectorFromFloat(Value=&Lerp_Float(A=1, B=$System.Flash, Alpha=0.6))
    ScaleA=true
    ScaleRGB=true
   ) {module=ScaleColor001}
   &ScaleColor(
    Scale Alpha=&FloatFromCurve(
     FloatCurve=DataInterface() {propertyValues {Curve {keys [{tangentMode=RCTM_User, time=0, value=1}, {tangentMode=RCTM_User, time=1, value=0}]}}}
     CurveIndex=$System.Timeline005.NormalizedTime
    )
    ScaleA=true
    ScaleRGB=true
   ) {module=ScaleColor003}
   &ScaleColor(
    Scale RGB=&Lerp_Vector(
     A=&VectorFromFloat(Value=1)
     B=&VectorFromFloat(Value=&Lerp_Float(
       A=0.5
       B=1
       Alpha=&Lerp_Float(
        A=0
        B=&OneMinusFloat(Float=&Round_Float(Float:NiagaraFloat=&ClampFloat(Float=&Sine(
            Normalized Angle=&Multiply_Float(A=$User.RarityChangeFloat_01, B=10)
            Period=&FloatFromCurve(
             FloatCurve=DataInterface() {propertyValues {Curve {keys [{tangentMode=RCTM_User, time=0, value=2}, <at 0.925 0.75>, {tangentMode=RCTM_User, time=0.925, value=0.05}, <at 0.995 0.05>, <at 1 0>]}}}
             CurveIndex=$User.RarityChangeFloat_01
            )
            Scale=0.5
            Bias=&FloatFromCurve(
             FloatCurve=DataInterface() {propertyValues {Curve {keys [
                 <at 0 0.9 RCIM_Cubic>
                 {arriveTangent=-0.432, interpMode=RCIM_Cubic, leaveTangent=-0.432, time=0.925, value=0.65}
                 <at 0.926 0.5>
                 <at 1 0.5>
                ]}}}
             CurveIndex=$User.RarityChangeFloat_01
            )
           ))))
        Alpha=&FloatFromCurve(
         FloatCurve=DataInterface() {propertyValues {Curve {keys [<at 0 0>,<at 0.001 1>,<at 0.999 1>,<at 1 0>]}}}
         CurveIndex=$User.RarityChangeFloat_01
        )
       )
      ))
     Alpha=&MakeCustomFloatFromBool(
      Bool=&Boolean_AndOperation(
       Bool1=&SetBoolByFloatComparison(A=$User.RarityChangeFloat_01, B=0.01)
       Bool2=&SetBoolByFloatComparison(A=$User.RarityChangeFloat_01, B=0.99, Comparison Type=::NewEnumerator6)
      )
      True Float=1
      False Float=0
     )
    )
    ScaleA=true
    ScaleRGB=true
   ) {module=ScaleColor004}
  ]
  rendererClasses [&NiagaraMeshRendererProperties]
  renderers [{rendererClass=&NiagaraMeshRendererProperties, rendererIndex=0}]
 }
 Ray_Soft001 {
  bEnabled=false
  emitterUpdateScript [
   &EmitterState(
    Inactive Response=::NewEnumerator1
    Life Cycle Mode=::NewEnumerator1
    Loop Behavior=::NewEnumerator1
    Loop Duration Mode=::NewEnumerator1
    UseLoopDelay=false
   )
   &SpawnBurst_Instantaneous(Spawn Count=1, Use Spawn Probability=false)
  ]
  particleSpawnScript [
   &InitializeParticle(
    Mesh Scale=<0.35 0.32 0.32>
    Mesh Scale Mode=::NewEnumerator4
    Position Offset=&Add_Vector(A=$System.Master Offset, B=<0 0 -10>)
    UsePositionOffset=true
    Write Material Random=true
   )
   &InitialMeshOrientation(Orientation Vector=<0 0 10>, Rotation=<0 0.01 0>, Use Orientation Vector=true, Use Rotation Vector=false) {module=InitialMeshOrientation001}
   /SetVariables(Particles.FriendlyColor:LinearColor=<rgba 0 0.1125 0.75 1>, Particles.HostileColor:LinearColor=<rgba 0.75 0.05 0 1>) {isSetParameters=true, module=SetVariables_#14}
  ]
  particleUpdateScript [
   &ParticleState(Kill Particles When Lifetime Has Elapsed=false, Let Infinitely Lived Particles Die When Emitter Deactivates=true)
   &ApplyOwnerScaleToAttributes(Scale Initial Mesh Scale=true)
   &SolveForcesAndVelocity(Clamp Velocity=false, Limit Acceleration=false)
   &ScaleColor(Scale Alpha=0.7, Scale RGB=&VectorFromFloat(Value=1), ScaleA=true, ScaleRGB=true)
   &ScaleColor(
    Scale RGB=&MakeVectorFromLinearColorRGB(Color=&LerpLinearColors(
      StartColor=<rgba 0.75 0.2997 0.0391 1>
      EndColor=&LerpLinearColors(StartColor=$Particles.FriendlyColor, EndColor=$Particles.HostileColor, LerpFactor=$User.FriendlyHostileFloat)
      LerpFactor=0
     ))
    ScaleA=true
    ScaleRGB=true
   ) {module=ScaleColor002}
   &ScaleMeshSize(Scale Factor=&VectorFromFloat(Value=&Multiply_Float(A=3.5, B=$System.Master Scale)))
   &ScaleColor(
    Scale Alpha=&Multiply_Float(A=&Multiply_Float(A=$System.LightConeFade, B=$System.TOD), B=$System.Flicker)
    Scale RGB=&VectorFromFloat(Value=&Lerp_Float(A=1, B=$System.Flash, Alpha=0.6))
    ScaleA=true
    ScaleRGB=true
   ) {module=ScaleColor001}
  ]
  rendererClasses [&NiagaraMeshRendererProperties]
  renderers [{rendererClass=&NiagaraMeshRendererProperties, rendererIndex=0}]
 }
 Ray_Soft002 {
  bEnabled=false
  emitterUpdateScript [
   &EmitterState(
    Inactive Response=::NewEnumerator1
    Life Cycle Mode=::NewEnumerator1
    Loop Behavior=::NewEnumerator1
    Loop Duration Mode=::NewEnumerator1
    UseLoopDelay=false
   )
   &SpawnBurst_Instantaneous(Spawn Count=1, Use Spawn Probability=false)
  ]
  particleSpawnScript [
   &InitializeParticle(
    Mesh Scale=<0.35 0.32 0.32>
    Mesh Scale Mode=::NewEnumerator4
    Position Offset=&Add_Vector(A=$System.Master Offset, B=<0 0 -10>)
    UsePositionOffset=true
    Write Material Random=true
   )
   &InitialMeshOrientation(Orientation Vector=<0 0 10>, Rotation=<0 0.01 0>, Use Orientation Vector=true, Use Rotation Vector=false) {module=InitialMeshOrientation001}
   /SetVariables(Particles.FriendlyColor:LinearColor=<rgba 0 0.1125 0.75 1>, Particles.HostileColor:LinearColor=<rgba 0.75 0.05 0 1>) {isSetParameters=true, module=SetVariables_#14}
  ]
  particleUpdateScript [
   &ParticleState(Kill Particles When Lifetime Has Elapsed=false, Let Infinitely Lived Particles Die When Emitter Deactivates=true)
   &ApplyOwnerScaleToAttributes(Scale Initial Mesh Scale=true)
   &SolveForcesAndVelocity(Clamp Velocity=false, Limit Acceleration=false)
   &ScaleColor(Scale Alpha=0.7, Scale RGB=&VectorFromFloat(Value=1), ScaleA=true, ScaleRGB=true)
   &ScaleColor(
    Scale RGB=&Multiply_VectorByFloat(
     Vector=&MakeVectorFromLinearColorRGB(Color=&LerpLinearColors(
       StartColor=&SelectLinearColorFromArray(
        Array Sampling Mode=::NewEnumerator1
        Direct Array Index=$System.SysCurrentRarity
        Color Selection Array=DataInterface() {propertyValues {ColorData [
           <rgba 0.4271 0.3821 0.3417 1>
           <rgba 0.2063 0.7656 0.0997 1>
           <rgba 0.0656 0.3836 1.8 1>
           <rgba 0.6968 0.0229 2.2 1>
           <rgba 1.5 0.7891 0.0781 1>
           <rgba 0.775 0.3185 0.004 1>
           <rgba 0.1238 1.2143 1.5 1>
           <rgba 0.0908 0.5841 0.006 1>
          ]}}
        Random Seed=?
       )
       EndColor=&SelectLinearColorFromArray(
        Array Sampling Mode=::NewEnumerator1
        Direct Array Index=$System.SysNextRarity
        Color Selection Array=DataInterface() {propertyValues {ColorData [
           <rgba 0.4271 0.3821 0.3417 1>
           <rgba 0.2063 0.7656 0.0997 1>
           <rgba 0.0656 0.3836 1.8 1>
           <rgba 0.6968 0.0229 2.2 1>
           <rgba 1.5 0.5995 0.0781 1>
           <rgba 0.775 0.3185 0.004 1>
           <rgba 0.1238 1.2143 1.5 1>
           <rgba 0.0908 0.5841 0.006 1>
          ]}}
        Random Seed=?
       )
       LerpFactor=$System.LerpRarityColor
      ))
     Float=0.8
    )
    ScaleA=true
    ScaleRGB=true
   ) {module=ScaleColor002}
   &ScaleMeshSize(Scale Factor=&VectorFromFloat(Value=&Multiply_Float(A=3.5, B=$System.Master Scale)))
   &ScaleColor(
    Scale Alpha=&Multiply_Float(A=&Multiply_Float(A=$System.LightConeFade, B=$System.TOD), B=$System.Flicker)
    Scale RGB=&VectorFromFloat(Value=&Lerp_Float(A=1, B=$System.Flash, Alpha=0.6))
    ScaleA=true
    ScaleRGB=true
   ) {module=ScaleColor001}
   &ScaleColor(
    Scale Alpha=&FloatFromCurve(
     FloatCurve=DataInterface() {propertyValues {Curve {keys [{tangentMode=RCTM_User, time=0, value=1}, {tangentMode=RCTM_User, time=1, value=0}]}}}
     CurveIndex=$System.Timeline005.NormalizedTime
    )
    ScaleA=true
    ScaleRGB=true
   ) {module=ScaleColor003}
   &ScaleColor(
    Scale RGB=&Lerp_Vector(
     A=&VectorFromFloat(Value=1)
     B=&VectorFromFloat(Value=&Lerp_Float(
       A=6
       B=1
       Alpha=&Sine(Normalized Angle=&Multiply_Float(A=$User.RarityChangeFloat_01, B=10), Period=0.75, Scale=0.5, Bias=0.5)
      ))
     Alpha=&MakeCustomFloatFromBool(
      Bool=&Boolean_AndOperation(
       Bool1=&SetBoolByFloatComparison(A=$User.RarityChangeFloat_01, B=0.01)
       Bool2=&SetBoolByFloatComparison(A=$User.RarityChangeFloat_01, B=0.99, Comparison Type=::NewEnumerator6)
      )
      True Float=1
      False Float=0
     )
    )
    ScaleA=true
    ScaleRGB=true
   ) {module=ScaleColor004}
  ]
  rendererClasses [&NiagaraMeshRendererProperties]
  renderers [{rendererClass=&NiagaraMeshRendererProperties, rendererIndex=0}]
 }
 Rays_flash {
  emitterUpdateScript [
   &EmitterState(Inactive Response=::NewEnumerator1, Life Cycle Mode=::NewEnumerator1, Loop Duration=1.4, UseLoopDelay=false)
   &SpawnRate(SpawnRate=8, Use Spawn Probability=false)
  ]
  particleSpawnScript [
   &InitializeParticle(
    Color=<rgba 1.5 1.5 1.5 1>
    Lifetime Max=0.7
    Lifetime Min=0.5
    Lifetime Mode=::NewEnumerator1
    Mesh Scale Max=<0.45 0.2 0.06>
    Mesh Scale Min=<0.45 0.2 0.05>
    Mesh Scale Mode=::NewEnumerator2
    Position Offset=&Add_Vector(A=$System.Master Offset, B=<0 0 -5>)
    UsePositionOffset=true
    Write Material Random=true
   ) {module=InitializeParticle001}
   &DynamicMaterialParameters(
    Index 0 Param 1=&UniformRangedFloat(Minimum=-1, Maximum=1)
    Index 0 Param 2=1
    Module.Param0WriteEnabled=true
    Module.Param1WriteEnabled=true
    Module.Param2WriteEnabled=true
    Module.Param3WriteEnabled=true
   )
   /SetVariables(Particles.FriendlyColor:LinearColor=<rgba 0 0.3 2 3>, Particles.HostileColor:LinearColor=<rgba 2 0.1333 0 3>) {isSetParameters=true, module=SetVariables_#14}
   &InitialMeshOrientation(
    Mesh Orientation Mode=::NewEnumerator2
    Orientation Vector=&Lerp_Vector(
     A=&RandomVector(Vector Scale=1, Constrain to Hemisphere=true, Hemisphere Constraint Vector=<0 0 1>)
     B=<0 0 1>
     Alpha=0.6
    )
    Rotation=&RandomRangeVector(Minimum=<0 0 0>, Maximum=<0 0 0>)
    Use Rotation Vector=true
   ) {module=InitialMeshOrientation002}
  ]
  particleUpdateScript [
   &ParticleState()
   &ScaleMeshSize(Scale Factor=&MakeVector(
     X=&FloatFromCurve(FloatCurve=DataInterface() {propertyValues {Curve {keys [<at 0 0.8>,<at 1 1.2>]}}})
     Y=1
     Z=&FloatFromCurve(FloatCurve=DataInterface() {propertyValues {Curve {keys [<at 0 1.8>,<at 1 1>]}}})
    ))
   &ScaleMeshSize(Scale Factor=&VectorFromFloat(Value=&Multiply_Float(A=3, B=$System.Master Scale))) {module=ScaleMeshSize001}
   &ScaleColor(
    Scale Alpha=&Multiply_Float(
     A=&FloatFromCurve(
      FloatCurve=DataInterface() {propertyValues {Curve {keys [
          {interpMode=RCIM_Cubic, tangentMode=RCTM_User, time=0, value=0}
          {interpMode=RCIM_Cubic, tangentMode=RCTM_User, time=0.3, value=0.5}
          {tangentMode=RCTM_User, time=1, value=0}
         ]}}}
      Scale Curve=0.2
     )
     B=&RandomRangeFloat(Minimum=0.5, Maximum=1.7)
    )
    ScaleA=true
    ScaleRGB=true
   )
   &ScaleColor(
    Scale RGB=&Multiply_VectorByFloat(
     Vector=&MakeVectorFromLinearColorRGB(Color=&LerpLinearColors(
       StartColor=&SelectLinearColorFromArray(
        Array Sampling Mode=::NewEnumerator1
        Direct Array Index=$System.SysCurrentRarity
        Color Selection Array=DataInterface() {propertyValues {ColorData [
           <rgba 0.25 0.2237 0.2 1>
           <rgba 0.1347 0.5 0.0651 1>
           <rgba 0.0365 0.2131 1 1>
           <rgba 0.2376 0.0078 0.75 1>
           <rgba 1 0.526 0.0521 1>
           <rgba 0.775 0.3185 0.004 1>
           <rgba 0.0619 0.6071 0.75 1>
           <rgba 0.0622 0.4 0.0041 1>
          ]}}
        Random Seed=?
       )
       EndColor=&SelectLinearColorFromArray(
        Array Sampling Mode=::NewEnumerator1
        Direct Array Index=$System.SysNextRarity
        Color Selection Array=DataInterface() {propertyValues {ColorData [
           <rgba 0.25 0.2237 0.2 1>
           <rgba 0.1347 0.5 0.0651 1>
           <rgba 0.0365 0.2131 1 1>
           <rgba 0.2376 0.0078 0.75 1>
           <rgba 1 0.526 0.0521 1>
           <rgba 0.775 0.3185 0.004 1>
           <rgba 0.0619 0.6071 0.75 1>
           <rgba 0.0622 0.4 0.0041 1>
          ]}}
        Random Seed=?
       )
       LerpFactor=0
      ))
     Float=5
    )
    ScaleA=true
    ScaleRGB=true
   ) {module=ScaleColor002}
   &ScaleColor(
    Scale Alpha=&Multiply_Float(A=&Multiply_Float(A=$System.LightConeFade, B=$System.TOD), B=$System.Flicker)
    Scale RGB=&VectorFromFloat(Value=&Lerp_Float(A=1, B=$System.Flash, Alpha=2))
    ScaleA=true
    ScaleRGB=true
   ) {module=ScaleColor001}
   &ScaleColor(
    Scale Alpha=&FloatFromCurve(
     FloatCurve=DataInterface() {propertyValues {Curve {keys [{tangentMode=RCTM_User, time=0, value=1}, {tangentMode=RCTM_User, time=1, value=0}]}}}
     CurveIndex=$System.Timeline005.NormalizedTime
    )
    ScaleA=true
    ScaleRGB=true
   ) {module=ScaleColor003}
   &ScaleColor(
    Scale Alpha=&Lerp_Float(
     A=1
     B=0.5
     Alpha=&Lerp_Float(
      A=0
      B=&OneMinusFloat(Float=&Round_Float(Float:NiagaraFloat=&ClampFloat(Float=&Sine(
          Normalized Angle=&Multiply_Float(A=$User.RarityChangeFloat_01, B=10)
          Period=&FloatFromCurve(
           FloatCurve=DataInterface() {propertyValues {Curve {keys [{tangentMode=RCTM_User, time=0, value=2}, <at 0.925 0.75>, {tangentMode=RCTM_User, time=0.925, value=0.05}, <at 0.995 0.05>, <at 1 0>]}}}
           CurveIndex=$User.RarityChangeFloat_01
          )
          Scale=0.5
          Bias=&FloatFromCurve(
           FloatCurve=DataInterface() {propertyValues {Curve {keys [
               <at 0 0.9 RCIM_Cubic>
               {arriveTangent=-0.432, interpMode=RCIM_Cubic, leaveTangent=-0.432, time=0.925, value=0.65}
               <at 0.926 0.5>
               <at 1 0.5>
              ]}}}
           CurveIndex=$User.RarityChangeFloat_01
          )
         ))))
      Alpha=&FloatFromCurve(
       FloatCurve=DataInterface() {propertyValues {Curve {keys [<at 0 0>,<at 0.001 1>,<at 0.999 1>,<at 1 0>]}}}
       CurveIndex=$User.RarityChangeFloat_01
      )
     )
    )
    Scale RGB=&Lerp_Vector(
     A=&VectorFromFloat(Value=1)
     B=&VectorFromFloat(Value=&Lerp_Float(
       A=0.2
       B=1
       Alpha=&Lerp_Float(
        A=0
        B=&OneMinusFloat(Float=&Round_Float(Float:NiagaraFloat=&ClampFloat(Float=&Sine(
            Normalized Angle=&Multiply_Float(A=$User.RarityChangeFloat_01, B=10)
            Period=&FloatFromCurve(
             FloatCurve=DataInterface() {propertyValues {Curve {keys [{tangentMode=RCTM_User, time=0, value=2}, <at 0.925 0.75>, {tangentMode=RCTM_User, time=0.925, value=0.05}, <at 0.995 0.05>, <at 1 0>]}}}
             CurveIndex=$User.RarityChangeFloat_01
            )
            Scale=0.5
            Bias=&FloatFromCurve(
             FloatCurve=DataInterface() {propertyValues {Curve {keys [
                 <at 0 0.9 RCIM_Cubic>
                 {arriveTangent=-0.432, interpMode=RCIM_Cubic, leaveTangent=-0.432, time=0.925, value=0.65}
                 <at 0.926 0.5>
                 <at 1 0.5>
                ]}}}
             CurveIndex=$User.RarityChangeFloat_01
            )
           ))))
        Alpha=&FloatFromCurve(
         FloatCurve=DataInterface() {propertyValues {Curve {keys [<at 0 0>,<at 0.001 1>,<at 0.999 1>,<at 1 0>]}}}
         CurveIndex=$User.RarityChangeFloat_01
        )
       )
      ))
     Alpha=&MakeCustomFloatFromBool(
      Bool=&Boolean_AndOperation(
       Bool1=&SetBoolByFloatComparison(A=$User.RarityChangeFloat_01, B=0.01)
       Bool2=&SetBoolByFloatComparison(A=$User.RarityChangeFloat_01, B=0.99, Comparison Type=::NewEnumerator6)
      )
      True Float=1
      False Float=0
     )
    )
    ScaleA=true
    ScaleRGB=true
   ) {module=ScaleColor004}
  ]
  rendererClasses [&NiagaraMeshRendererProperties]
  renderers [{rendererClass=&NiagaraMeshRendererProperties, rendererIndex=0}]
 }
 Rays_flash001 {
  bEnabled=false
  emitterUpdateScript [
   &EmitterState(Inactive Response=::NewEnumerator1, Life Cycle Mode=::NewEnumerator1, Loop Duration=1.4, UseLoopDelay=false)
   &SpawnRate(SpawnRate=8, Use Spawn Probability=false)
  ]
  particleSpawnScript [
   &InitializeParticle(
    Color=<rgba 1.5 1.5 1.5 1>
    Lifetime Max=0.7
    Lifetime Min=0.5
    Lifetime Mode=::NewEnumerator1
    Mesh Scale Max=<0.45 0.2 0.06>
    Mesh Scale Min=<0.45 0.2 0.05>
    Mesh Scale Mode=::NewEnumerator2
    Position Offset=&Add_Vector(A=$System.Master Offset, B=<0 0 -5>)
    UsePositionOffset=true
    Write Material Random=true
   ) {module=InitializeParticle001}
   &DynamicMaterialParameters(
    Index 0 Param 1=&UniformRangedFloat(Minimum=-1, Maximum=1)
    Index 0 Param 2=1
    Module.Param0WriteEnabled=true
    Module.Param1WriteEnabled=true
    Module.Param2WriteEnabled=true
    Module.Param3WriteEnabled=true
   )
   /SetVariables(Particles.FriendlyColor:LinearColor=<rgba 0 0.3 2 3>, Particles.HostileColor:LinearColor=<rgba 2 0.1333 0 3>) {isSetParameters=true, module=SetVariables_#14}
   &InitialMeshOrientation(
    Mesh Orientation Mode=::NewEnumerator2
    Orientation Vector=&Lerp_Vector(
     A=&RandomVector(Vector Scale=1, Constrain to Hemisphere=true, Hemisphere Constraint Vector=<0 0 1>)
     B=<0 0 1>
     Alpha=0.6
    )
    Rotation=&RandomRangeVector(Minimum=<0 0 0>, Maximum=<0 0 0>)
    Use Rotation Vector=true
   ) {module=InitialMeshOrientation002}
  ]
  particleUpdateScript [
   &ParticleState()
   &ScaleMeshSize(Scale Factor=&MakeVector(
     X=&FloatFromCurve(FloatCurve=DataInterface() {propertyValues {Curve {keys [<at 0 0.8>,<at 1 1.2>]}}})
     Y=1
     Z=&FloatFromCurve(FloatCurve=DataInterface() {propertyValues {Curve {keys [<at 0 1.8>,<at 1 1>]}}})
    ))
   &ScaleMeshSize(Scale Factor=&VectorFromFloat(Value=&Multiply_Float(A=3, B=$System.Master Scale))) {module=ScaleMeshSize001}
   &ScaleColor(
    Scale Alpha=&Multiply_Float(
     A=&FloatFromCurve(
      FloatCurve=DataInterface() {propertyValues {Curve {keys [
          {interpMode=RCIM_Cubic, tangentMode=RCTM_User, time=0, value=0}
          {interpMode=RCIM_Cubic, tangentMode=RCTM_User, time=0.3, value=0.5}
          {tangentMode=RCTM_User, time=1, value=0}
         ]}}}
      Scale Curve=0.2
     )
     B=&RandomRangeFloat(Minimum=0.5, Maximum=1.7)
    )
    ScaleA=true
    ScaleRGB=true
   )
   &ScaleColor(
    Scale RGB=&MakeVectorFromLinearColorRGB(Color=&LerpLinearColors(
      StartColor=&RandomRangeLinearColor(Minimum=<rgba 1 0.3997 0.0521 4>, Maximum=<rgba 2 0.7993 0.1042 2>)
      EndColor=&LerpLinearColors(StartColor=$Particles.FriendlyColor, EndColor=$Particles.HostileColor, LerpFactor=$User.FriendlyHostileFloat)
      LerpFactor=0
     ))
    ScaleA=true
    ScaleRGB=true
   ) {module=ScaleColor002}
   &ScaleColor(
    Scale Alpha=&Multiply_Float(A=&Multiply_Float(A=$System.LightConeFade, B=$System.TOD), B=$System.Flicker)
    Scale RGB=&VectorFromFloat(Value=&Lerp_Float(A=1, B=$System.Flash, Alpha=2))
    ScaleA=true
    ScaleRGB=true
   ) {module=ScaleColor001}
  ]
  rendererClasses [&NiagaraMeshRendererProperties]
  renderers [{rendererClass=&NiagaraMeshRendererProperties, rendererIndex=0}]
 }
 Rays_flash002 {
  bEnabled=false
  emitterUpdateScript [
   &EmitterState(Inactive Response=::NewEnumerator1, Life Cycle Mode=::NewEnumerator1, Loop Duration=1.4, UseLoopDelay=false)
   &SpawnRate(SpawnRate=8, Use Spawn Probability=false)
  ]
  particleSpawnScript [
   &InitializeParticle(
    Color=<rgba 1.5 1.5 1.5 1>
    Lifetime Max=0.7
    Lifetime Min=0.5
    Lifetime Mode=::NewEnumerator1
    Mesh Scale Max=<0.45 0.2 0.06>
    Mesh Scale Min=<0.45 0.2 0.05>
    Mesh Scale Mode=::NewEnumerator2
    Position Offset=&Add_Vector(A=$System.Master Offset, B=<0 0 -5>)
    UsePositionOffset=true
    Write Material Random=true
   ) {module=InitializeParticle001}
   &DynamicMaterialParameters(
    Index 0 Param 1=&UniformRangedFloat(Minimum=-1, Maximum=1)
    Index 0 Param 2=1
    Module.Param0WriteEnabled=true
    Module.Param1WriteEnabled=true
    Module.Param2WriteEnabled=true
    Module.Param3WriteEnabled=true
   )
   /SetVariables(Particles.FriendlyColor:LinearColor=<rgba 0 0.3 2 3>, Particles.HostileColor:LinearColor=<rgba 2 0.1333 0 3>) {isSetParameters=true, module=SetVariables_#14}
   &InitialMeshOrientation(
    Mesh Orientation Mode=::NewEnumerator2
    Orientation Vector=&Lerp_Vector(
     A=&RandomVector(Vector Scale=1, Constrain to Hemisphere=true, Hemisphere Constraint Vector=<0 0 1>)
     B=<0 0 1>
     Alpha=0.6
    )
    Rotation=&RandomRangeVector(Minimum=<0 0 0>, Maximum=<0 0 0>)
    Use Rotation Vector=true
   ) {module=InitialMeshOrientation002}
  ]
  particleUpdateScript [
   &ParticleState()
   &ScaleMeshSize(Scale Factor=&MakeVector(
     X=&FloatFromCurve(FloatCurve=DataInterface() {propertyValues {Curve {keys [<at 0 0.8>,<at 1 1.2>]}}})
     Y=1
     Z=&FloatFromCurve(FloatCurve=DataInterface() {propertyValues {Curve {keys [<at 0 1.8>,<at 1 1>]}}})
    ))
   &ScaleMeshSize(Scale Factor=&VectorFromFloat(Value=&Multiply_Float(A=3, B=$System.Master Scale))) {module=ScaleMeshSize001}
   &ScaleColor(
    Scale Alpha=&Multiply_Float(
     A=&FloatFromCurve(
      FloatCurve=DataInterface() {propertyValues {Curve {keys [
          {interpMode=RCIM_Cubic, tangentMode=RCTM_User, time=0, value=0}
          {interpMode=RCIM_Cubic, tangentMode=RCTM_User, time=0.3, value=0.5}
          {tangentMode=RCTM_User, time=1, value=0}
         ]}}}
      Scale Curve=0.2
     )
     B=&RandomRangeFloat(Minimum=0.5, Maximum=1.7)
    )
    ScaleA=true
    ScaleRGB=true
   )
   &ScaleColor(
    Scale RGB=&Multiply_VectorByFloat(
     Vector=&MakeVectorFromLinearColorRGB(Color=&LerpLinearColors(
       StartColor=&SelectLinearColorFromArray(
        Array Sampling Mode=::NewEnumerator1
        Direct Array Index=$System.SysCurrentRarity
        Color Selection Array=DataInterface() {propertyValues {ColorData [
           <rgba 0.25 0.2237 0.2 1>
           <rgba 0.1347 0.5 0.0651 1>
           <rgba 0.0365 0.2131 1 1>
           <rgba 0.2376 0.0078 0.75 1>
           <rgba 1 0.526 0.0521 1>
           <rgba 0.775 0.3185 0.004 1>
           <rgba 0.0619 0.6071 0.75 1>
           <rgba 0.0622 0.4 0.0041 1>
          ]}}
        Random Seed=?
       )
       EndColor=&SelectLinearColorFromArray(
        Array Sampling Mode=::NewEnumerator1
        Direct Array Index=$System.SysNextRarity
        Color Selection Array=DataInterface() {propertyValues {ColorData [
           <rgba 0.25 0.2237 0.2 1>
           <rgba 0.1347 0.5 0.0651 1>
           <rgba 0.0365 0.2131 1 1>
           <rgba 0.2376 0.0078 0.75 1>
           <rgba 1 0.526 0.0521 1>
           <rgba 0.775 0.3185 0.004 1>
           <rgba 0.0619 0.6071 0.75 1>
           <rgba 0.0622 0.4 0.0041 1>
          ]}}
        Random Seed=?
       )
       LerpFactor=$System.LerpRarityColor
      ))
     Float=5
    )
    ScaleA=true
    ScaleRGB=true
   ) {module=ScaleColor002}
   &ScaleColor(
    Scale Alpha=&Multiply_Float(A=&Multiply_Float(A=$System.LightConeFade, B=$System.TOD), B=$System.Flicker)
    Scale RGB=&VectorFromFloat(Value=&Lerp_Float(A=1, B=$System.Flash, Alpha=2))
    ScaleA=true
    ScaleRGB=true
   ) {module=ScaleColor001}
   &ScaleColor(
    Scale Alpha=&FloatFromCurve(
     FloatCurve=DataInterface() {propertyValues {Curve {keys [{tangentMode=RCTM_User, time=0, value=1}, {tangentMode=RCTM_User, time=1, value=0}]}}}
     CurveIndex=$System.Timeline005.NormalizedTime
    )
    ScaleA=true
    ScaleRGB=true
   ) {module=ScaleColor003}
   &ScaleColor(
    Scale RGB=&Lerp_Vector(
     A=&VectorFromFloat(Value=1)
     B=&VectorFromFloat(Value=&Lerp_Float(
       A=6
       B=1
       Alpha=&Sine(Normalized Angle=&Multiply_Float(A=$User.RarityChangeFloat_01, B=10), Period=0.75, Scale=0.5, Bias=0.5)
      ))
     Alpha=&MakeCustomFloatFromBool(
      Bool=&Boolean_AndOperation(
       Bool1=&SetBoolByFloatComparison(A=$User.RarityChangeFloat_01, B=0.01)
       Bool2=&SetBoolByFloatComparison(A=$User.RarityChangeFloat_01, B=0.99, Comparison Type=::NewEnumerator6)
      )
      True Float=1
      False Float=0
     )
    )
    ScaleA=true
    ScaleRGB=true
   ) {module=ScaleColor004}
  ]
  rendererClasses [&NiagaraMeshRendererProperties]
  renderers [{rendererClass=&NiagaraMeshRendererProperties, rendererIndex=0}]
 }
}
summary {
 systemName=NS_Aug_Container_Lightcone_01
 emitters [
  {emitterName=Ray_Soft, rendererClasses [&NiagaraMeshRendererProperties]}
  {emitterName=Rays_flash, rendererClasses [&NiagaraMeshRendererProperties]}
  {emitterName=Glow, rendererClasses [&NiagaraSpriteRendererProperties]}
  {
   emitterName=Light_loop
   rendererClasses [
    &NiagaraLightRendererProperties
    &NiagaraComponentRendererProperties
    &NiagaraSpriteRendererProperties
   ]
  }
  {emitterName=Cone_Ray, rendererClasses [&NiagaraMeshRendererProperties]}
  {emitterName=Cone_pulse, rendererClasses [&NiagaraMeshRendererProperties]}
  {emitterName=Capture_Icon, rendererClasses [&NiagaraSpriteRendererProperties]}
  {bEnabled=false, emitterName=Ray_Soft001, rendererClasses [&NiagaraMeshRendererProperties]}
  {bEnabled=false, emitterName=Rays_flash001, rendererClasses [&NiagaraMeshRendererProperties]}
  {bEnabled=false, emitterName=Glow001, rendererClasses [&NiagaraSpriteRendererProperties]}
  {
   bEnabled=false
   emitterName=Light_loop001
   rendererClasses [&NiagaraLightRendererProperties, &NiagaraMeshRendererProperties]
  }
  {bEnabled=false, emitterName=Cone_Ray001, rendererClasses [&NiagaraMeshRendererProperties]}
  {bEnabled=false, emitterName=Cone_pulse001, rendererClasses [&NiagaraMeshRendererProperties]}
  {bEnabled=false, emitterName=Capture_Icon001, rendererClasses [&NiagaraSpriteRendererProperties]}
  {emitterName=Degrade_Glow, rendererClasses [&NiagaraSpriteRendererProperties]}
  {bEnabled=false, emitterName=Ray_Soft002, rendererClasses [&NiagaraMeshRendererProperties]}
  {bEnabled=false, emitterName=Rays_flash002, rendererClasses [&NiagaraMeshRendererProperties]}
  {bEnabled=false, emitterName=Glow002, rendererClasses [&NiagaraSpriteRendererProperties]}
  {
   bEnabled=false
   emitterName=Light_loop002
   rendererClasses [&NiagaraLightRendererProperties, &NiagaraMeshRendererProperties]
  }
  {bEnabled=false, emitterName=Cone_Ray002, rendererClasses [&NiagaraMeshRendererProperties]}
  {bEnabled=false, emitterName=Cone_pulse002, rendererClasses [&NiagaraMeshRendererProperties]}
  {bEnabled=false, emitterName=Capture_Icon002, rendererClasses [&NiagaraSpriteRendererProperties]}
  {bEnabled=false, emitterName=Degrade_Glow001, rendererClasses [&NiagaraSpriteRendererProperties]}
 ]
 userVariables [
  User.CapturedFloat {type=NiagaraFloat, defaultValue=0, description=""}
  User.CurrentRarityLevel {type=NiagaraInt32, defaultValue=4}
  User.FadeOutHoloGram {type=NiagaraBool}
  User.FriendlyHostileFloat {type=NiagaraFloat, defaultValue=0, description=""}
  User.LightConeFlash {type=NiagaraBool}
  User.LightConeFlicker {type=NiagaraBool}
  User.LightConeInstantOff {type=NiagaraBool}
  User.LightConeToggle {type=NiagaraBool, defaultValue=true}
  User.Master Scale {type=NiagaraFloat, defaultValue=0.65, description=""}
  User.NextRarityLevel {type=NiagaraInt32, defaultValue=3}
  User.ProgressAmount {type=NiagaraFloat, defaultValue=0, description=""}
  User.RarityChangeFloat_01 {type=NiagaraFloat, defaultValue=0, description=""}
  User.ShowProgressBar {type=NiagaraBool}
 ]
}
